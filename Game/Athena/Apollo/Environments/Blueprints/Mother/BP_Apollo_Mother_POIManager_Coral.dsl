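; Blueprint POI manager for the Mother event's Coral location, switching map foundations to match the progress of its abduction.
; CoralAbduction_Foundations loads the coral beam stream level, then GetAbductionPhase thresholds Progress into three phases.
; Each phase enables its own Coral foundation plus shared FieldTeam and AfterMath ones, with phase one's AfterMath marked temporary.
; MoveToCoral_Foundations is empty, and the location and spline-rotation helpers go uncalled here.
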(event Custom|MoveToCoral_Foundations
  (params (Progress :type Float%20%28double-precision%29) (Debug%20Check :type Integer)))

(event Custom|CoralAbduction_Foundations
  (params (Progress :type Float%20%28double-precision%29))
  (Game|LoadStreamLevel(byName) "Apollo_Mother_CoralBeam" true)
  (switch int (CallFunction|GetAbductionPhase Progress)
    (:0
      (CallFunction|EnableIslandFoundation (Variables|Coral|GetCoralPhase_01))
      (CallFunction|EnableIslandFoundation (Variables|Coral|GetFieldTeam_01))
      (CallFunction|EnableIslandFoundation (Variables|Coral|GetFieldTeam_02))
      (CallFunction|EnableIslandFoundation (Variables|Coral|GetFieldTeam_03))
      ;[1 1 1] To be removed when actual Coral calendar events are set up.
      (CallFunction|EnableIslandFoundation (Variables|Coral|GetAfterMath_01))
      (CallFunction|EnableIslandFoundation (Variables|Coral|GetAfterMath_02)))
    (:1
      (CallFunction|EnableIslandFoundation (Variables|Coral|GetCoralPhase_02))
      (CallFunction|EnableIslandFoundation (Variables|Coral|GetFieldTeam_01))
      (CallFunction|EnableIslandFoundation (Variables|Coral|GetFieldTeam_02))
      (CallFunction|EnableIslandFoundation (Variables|Coral|GetFieldTeam_03))
      (CallFunction|EnableIslandFoundation (Variables|Coral|GetAfterMath_01))
      (CallFunction|EnableIslandFoundation (Variables|Coral|GetAfterMath_02)))
    (:2
      (CallFunction|EnableIslandFoundation (Variables|Coral|GetCoralPhase_03))
      (CallFunction|EnableIslandFoundation (Variables|Coral|GetFieldTeam_01))
      (CallFunction|EnableIslandFoundation (Variables|Coral|GetFieldTeam_02))
      (CallFunction|EnableIslandFoundation (Variables|Coral|GetFieldTeam_03))
      (CallFunction|EnableIslandFoundation (Variables|Coral|GetAfterMath_01))
      (CallFunction|EnableIslandFoundation (Variables|Coral|GetAfterMath_02)))))

(fn EnableFoundationWithLocation
  (params (Foundation :type Building%20Foundation%20Object%20Reference))
  (Class|BuildingFoundation|SetDynamicFoundationTransform Foundation (Math|Transform|MakeTransform (Transformation|GetActorLocation Foundation)))
  (Class|BuildingFoundation|SetDynamicFoundationEnabled Foundation true))

(fn EnableIslandFoundation
  (params (Enable%20Foundation :type Building%20Foundation%20Object%20Reference) (Stage :type Integer))
  (Class|BuildingFoundation|SetDynamicFoundationEnabled Enable%20Foundation true))

(fn GetAbductionPhase
  (params (Progress :type Float%20%28double-precision%29))
  (if (< Progress 0.3)
    (Variables|Coral|SetAbduction%20Phase 0)
    (elif (< Progress 0.6)
      (Variables|Coral|SetAbduction%20Phase 1)
      (elif (>= Progress 0.6)
        (Variables|Coral|SetAbduction%20Phase 2)
        (else
          (Development|PrintString "NoAbductionPhaseSet")))))
  (return (Variables|Coral|GetAbduction%20Phase)))

(fn RotateOnSplinePoint
  (params (Spline :type Spline%20Component%20Object%20Reference))
  (return (Math|Rotator|MakeRotator 0.0 0.0 (Math|Float|SnapToGrid(Float) (.yaw (Math|Conversions|RotationFromXVector (Spline|GetLocationatSplinePoint Spline 0 "World"))) 9e+01))))

(fn ConstructionScript)
