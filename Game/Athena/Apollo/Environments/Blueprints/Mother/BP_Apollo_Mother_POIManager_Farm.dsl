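; POI manager for the Farm in the Apollo Mother environment, driving the farm abduction through dynamic building foundations.
; FarmAbduction_Foundations streams in the FarmBeam level, swaps the before foundation for the after one, and enables IO launch pad 01.
; GetAbductionPhase maps the Progress value to one of three phases, selecting which phase foundation to enable.
; Helper functions load and unload editor preview levels; MoveToFarm_Foundations and the construction script are empty.
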(event Custom|MoveToFarm_Foundations
  (params (Progress :type Float%20%28double-precision%29)))

(event Custom|FarmAbduction_Foundations
  (params (Progress :type Float%20%28double-precision%29))
  (Game|LoadStreamLevel(byName) "Apollo_Mother_FarmBeam" true)
  (Class|BuildingFoundation|SetDynamicFoundationEnabled (Variables|Foundations|GetFarmBefore) false)
  (Class|BuildingFoundation|SetDynamicFoundationEnabled (Variables|Foundations|GetFarmAfter) true)
  (switch int (CallFunction|GetAbductionPhase Progress)
    (:0
      (CallFunction|EnableIslandFoundation (Variables|Foundations|GetFarmPhase_01)))
    (:1
      (CallFunction|EnableIslandFoundation (Variables|Foundations|GetFarmPhase_02)))
    (:2
      (CallFunction|EnableIslandFoundation (Variables|Foundations|GetFarmPhase_03))))
  (Class|BuildingFoundation|SetDynamicFoundationEnabled (Variables|Foundations|GetIOLaunchPad_01) true))

(fn EnableFoundationWithLocation
  (params (Foundation :type Building%20Foundation%20Object%20Reference))
  (Class|BuildingFoundation|SetDynamicFoundationTransform Foundation (Math|Transform|MakeTransform (Transformation|GetActorLocation Foundation)))
  (Class|BuildingFoundation|SetDynamicFoundationEnabled Foundation true))

(fn EnableIslandFoundation
  (params (Foundation_Enable :type Building%20Foundation%20Object%20Reference))
  (Class|BuildingFoundation|SetDynamicFoundationEnabled Foundation_Enable true))

(fn GetAbductionPhase
  (params (Progress :type Float%20%28double-precision%29))
  (if (< Progress 0.3)
    (Variables|Default|SetAbductionPhase 0)
    (elif (< Progress 0.6)
      (Variables|Default|SetAbductionPhase 1)
      (elif (>= Progress 0.6)
        (Variables|Default|SetAbductionPhase 2)
        (else
          (Development|PrintString "NoAbductionPhaseSet")))))
  (return (Variables|Default|GetAbductionPhase)))

(fn InitializeUnpreviewFoundations
  (Utilities|Array|Add (Variables|Default|GetUnPreviewFoundations) (Variables|Foundations|GetFarmBefore))
  (Utilities|Array|Add (Variables|Default|GetUnPreviewFoundations) (Variables|Foundations|GetFarmAfter))
  (Utilities|Array|Add (Variables|Default|GetUnPreviewFoundations) (Variables|Foundations|GetFarmPhase_01))
  (Utilities|Array|Add (Variables|Default|GetUnPreviewFoundations) (Variables|Foundations|GetFarmPhase_02))
  (Utilities|Array|Add (Variables|Default|GetUnPreviewFoundations) (Variables|Foundations|GetFarmPhase_03))
  (Utilities|Array|Add (Variables|Default|GetUnPreviewFoundations) (Variables|Foundations|GetIOLaunchPad_01))
  (Utilities|Array|Add (Variables|Default|GetUnPreviewFoundations) (Variables|Foundations|GetIOLaunchPad_02))
  (Utilities|Array|Add (Variables|Default|GetUnPreviewFoundations) (Variables|Foundations|GetIOLaunchPad_03)))

(fn ResetAllPhase
  (params (UnpreviewFoundation_01 :type Building%20Foundation%20Object%20Reference) (UnpreviewFoundation_02 :type Building%20Foundation%20Object%20Reference) (UnpreviewFoundation_03 :type Building%20Foundation%20Object%20Reference))
  (Utilities|IsValid UnpreviewFoundation_01
    (:Is%20Valid
      (Class|BuildingFoundation|EditorOnlyUnloadPreviewLevel UnpreviewFoundation_01)
      (Utilities|IsValid UnpreviewFoundation_02
        (:Is%20Valid
          _)))
    (:Is%20Not%20Valid
      _))
  (Class|BuildingFoundation|EditorOnlyUnloadPreviewLevel UnpreviewFoundation_02)
  (Utilities|IsValid UnpreviewFoundation_03
    (:Is%20Valid
      (Class|BuildingFoundation|EditorOnlyUnloadPreviewLevel UnpreviewFoundation_03))))

(fn SetupPreviewFoundations
  (params (FarmBefore :type Building%20Foundation%20Object%20Reference) (FarmAfter :type Building%20Foundation%20Object%20Reference) (FarmPhase1 :type Building%20Foundation%20Object%20Reference) (FarmPhase2 :type Building%20Foundation%20Object%20Reference) (FarmPhase3 :type Building%20Foundation%20Object%20Reference) (FarmIOPad1 :type Building%20Foundation%20Object%20Reference) (FarmIOPad2 :type Building%20Foundation%20Object%20Reference) (FarmIOPad3 :type Building%20Foundation%20Object%20Reference))
  (bind _previewfoundations (Variables|Default|GetPreviewFoundations))
  (Utilities|Array|Clear _previewfoundations)
  (Utilities|IsValid FarmBefore
    (:Is%20Valid
      (Utilities|Array|Add _previewfoundations FarmBefore))
    (:Is%20Not%20Valid
      _))
  (Utilities|IsValid FarmAfter
    (:Is%20Valid
      (Utilities|Array|Add _previewfoundations FarmAfter))
    (:Is%20Not%20Valid
      _))
  (Utilities|IsValid FarmPhase1
    (:Is%20Valid
      (Utilities|Array|Add _previewfoundations FarmPhase1))
    (:Is%20Not%20Valid
      _))
  (Utilities|IsValid FarmPhase2
    (:Is%20Valid
      (Utilities|Array|Add _previewfoundations FarmPhase2))
    (:Is%20Not%20Valid
      _))
  (Utilities|IsValid FarmPhase3
    (:Is%20Valid
      (Utilities|Array|Add _previewfoundations FarmPhase3))
    (:Is%20Not%20Valid
      _))
  (Utilities|IsValid FarmIOPad1
    (:Is%20Valid
      (Utilities|Array|Add _previewfoundations FarmIOPad1))
    (:Is%20Not%20Valid
      _))
  (Utilities|IsValid FarmIOPad2
    (:Is%20Valid
      (Utilities|Array|Add _previewfoundations FarmIOPad2))
    (:Is%20Not%20Valid
      _))
  (Utilities|IsValid FarmIOPad3
    (:Is%20Valid
      (Utilities|Array|Add _previewfoundations FarmIOPad3))
    (:Is%20Not%20Valid
      _))
  (for _array_element (Variables|Default|GetPreviewFoundations)
    (Class|BuildingFoundation|EditorOnlyLoadPreviewLevel _array_element true)))

(fn SetupUnPreviewFoundation
  (params (Foundations :type Array%20of%20Building%20Foundation%20Object%20References))
  (for _array_element Foundations
    (Utilities|IsValid _array_element
      (:Is%20Valid
        _)
      (:Is%20Not%20Valid
        _))
    (Class|BuildingFoundation|EditorOnlyUnloadPreviewLevel _array_element)))

(fn ConstructionScript)
