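; Blueprint actor for a bent water pipe effect in the Mother environment, pairing a water particle effect with optional ambient audio.
; The construction script and BeginPlay both set the same fixed start-speed vector on the water pipe particle component.
; BeginPlay plays the audio component only when ambient audio is enabled, while EndPlay always calls a short linear fade-out on it.
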(event EventBeginPlay
  (Effects|Components|ParticleSystem|SetVectorParameter (Variables|Default|GetNS_Mother_POI_Chunk_WaterPipe) "Speed_Start" "300.000000,0.000000, 0")
  (if (Variables|Default|GetEnableAmbientAudio)
    (Audio|Components|Audio|Play (Variables|Default|GetAudio) 0.0)))

(event EventEndPlay
  (params (EndPlayReason :type EEndPlayReason%20Enum :default "Destroyed"))
  (Audio|Components|Audio|FadeOut (Variables|Default|GetAudio) 1.0 0.0 "Linear" :!class /Script/Engine.AudioComponent))

(fn ConstructionScript
  (Effects|Components|ParticleSystem|SetVectorParameter (Variables|Default|GetNS_Mother_POI_Chunk_WaterPipe) "Speed_Start" "300.000000,0.000000, 0"))
