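; Welcome spotlight prop with a configurable rotating, colored light beam.
; Construction Script feeds spotlight settings into the Niagara beam, spotlight mesh materials and editor-only light-shaft previs meshes.
; BeginPlay reapplies Niagara parameters and destroys the previs meshes.
; For the runtime beam, the server stores the Persistent FX component's actor, and SpotLightFXActor's OnRep copies rotation, size and color settings onto it.
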(event EventBeginPlay
  (bind _editoronly_sm_neo_tilted_light_shaft (Variables|Default|GetEditorOnly_sm_neo_tilted_light_shaft))
  (bind _editoronly_sm_neo_tilted_light_shaft_flares (Variables|Default|GetEditorOnly_sm_neo_tilted_light_shaft_Flares))
  (|CallParentFunction)
  (Utilities|FlowControl|Sequence
    (:then_0
      ;[1 1 1] Niagara parameters
      (Effects|Components|ParticleSystem|SetFloatParameter (Variables|Default|GetNS_Welcome_LightBeam) "ConeAngleFX" (Variables|SpotlightFXSettings|GetConeAngleFX))
      (Effects|Components|ParticleSystem|SetFloatParameter (Variables|Default|GetNS_Welcome_LightBeam) "ConeRotationFX" (Variables|SpotlightFXSettings|GetConeRotationFX))
      (Effects|Components|ParticleSystem|SetFloatParameter (Variables|Default|GetNS_Welcome_LightBeam) "HueFX" (Variables|SpotlightFXSettings|GetHueFX)))
    (:then_1
      (Utilities|IsValid _editoronly_sm_neo_tilted_light_shaft_flares
        (:Is%20Valid
          (Components|DestroyComponent _editoronly_sm_neo_tilted_light_shaft_flares))
        (:Is%20Not%20Valid
          _))
      (Utilities|IsValid _editoronly_sm_neo_tilted_light_shaft
        (:Is%20Valid
          (Components|DestroyComponent _editoronly_sm_neo_tilted_light_shaft))))))

(event EventinitPersistentFX
  (params (fxActor :type Actor%20Object%20Reference) (fxComponent :type Persistent%20FXComponent%20Object%20Reference))
  (Development|PrintString "InitPersitenFX" false false "(R=0.000000,G=0.660000,B=1.000000,A=1.000000)" 1e+01)
  (if (Networking|IsServer)
    (bind _asbp_welcome_spot_light_beam_01 (Utilities|Casting|CastToBP_Welcome_SpotLight_Beam_01 fxActor))
    (Variables|Default|SetSpotLightFXActor _asbp_welcome_spot_light_beam_01)))

(fn editorOnly_SpawnFXActor
  (params (actorClass :type Actor%20Class%20Reference) (fxComp :type Persistent%20FXComponent%20Object%20Reference))
  (return 0))

(fn OnRep_SpotLightFXActor
  (Utilities|IsValid (Variables|Default|GetSpotLightFXActor)
    (:Is%20Valid
      (Class|BPWelcomeSpotLightBeam01|SetSpotlightProperties (Variables|Default|GetSpotLightFXActor) (Variables|SpotlightFXSettings|GetRotation%20Max) (Variables|SpotlightFXSettings|GetRotation%20Speed) (Variables|SpotlightFXSettings|GetBeam%20Height) (Variables|SpotlightFXSettings|GetBeam%20Width) (Variables|SpotlightFXSettings|GetSpotlightColor)))))

(fn ConstructionScript
  (|CallParentFunction)
  (Utilities|FlowControl|Sequence
    (:then_2
      (Effects|Components|ParticleSystem|SetFloatParameter (Variables|Default|GetNS_Welcome_LightBeam) "ConeAngleFX" (Variables|SpotlightFXSettings|GetConeAngleFX))
      (Effects|Components|ParticleSystem|SetFloatParameter (Variables|Default|GetNS_Welcome_LightBeam) "ConeRotationFX" (Variables|SpotlightFXSettings|GetConeRotationFX))
      (Effects|Components|ParticleSystem|SetFloatParameter (Variables|Default|GetNS_Welcome_LightBeam) "HueFX" (Variables|SpotlightFXSettings|GetHueFX)))
    (:then_3
      (Rendering|Material|SetScalarParameterValueonMaterials (Variables|Default|GetS_Mother_Spotlight_01) "Rotation Max" (Variables|SpotlightFXSettings|GetRotation%20Max))
      (Rendering|Material|SetScalarParameterValueonMaterials (Variables|Default|GetS_Mother_Spotlight_01) "Rotation Speed" (Variables|SpotlightFXSettings|GetRotation%20Speed))
      (Rendering|Material|SetVectorParameterValueonMaterials (Variables|Default|GetS_Mother_Spotlight_01) "EmissiveColorMultiplier" (Math|Conversions|ToVector(LinearColor) (Variables|SpotlightFXSettings|GetSpotlightColor))))
    (:then_4
      ;[1 1 1] Editor Only previs, PersistentFXComponent does the real spawning and value copying is done in OnRep_SpotLightFxActor
      (Rendering|Material|SetScalarParameterValueonMaterials (Variables|Default|GetEditorOnly_sm_neo_tilted_light_shaft_Flares) "Rotation Max" (Variables|SpotlightFXSettings|GetRotation%20Max))
      (Rendering|Material|SetScalarParameterValueonMaterials (Variables|Default|GetEditorOnly_sm_neo_tilted_light_shaft) "Rotation Speed" (Variables|SpotlightFXSettings|GetRotation%20Speed))
      (Rendering|Material|SetScalarParameterValueonMaterials (Variables|Default|GetEditorOnly_sm_neo_tilted_light_shaft) "ScaleHeight" (Variables|SpotlightFXSettings|GetBeam%20Height))
      (Rendering|Material|SetVectorParameterValueonMaterials (Variables|Default|GetEditorOnly_sm_neo_tilted_light_shaft) "Color" (Math|Conversions|ToVector(LinearColor) (Variables|SpotlightFXSettings|GetSpotlightColor)))
      (Rendering|Material|SetScalarParameterValueonMaterials (Variables|Default|GetEditorOnly_sm_neo_tilted_light_shaft) "MainWidth" (Variables|SpotlightFXSettings|GetBeam%20Width)))))
