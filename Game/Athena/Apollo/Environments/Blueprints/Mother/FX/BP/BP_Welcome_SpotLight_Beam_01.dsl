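; Spotlight beam effect Blueprint built from a tilted light shaft mesh and a flares mesh.
; A custom event stores rotation, beam size and color settings, then pushes them to the meshes as material parameters.
; Rotation max goes only to the flares, while speed, height and width go only to the shaft.
; Both receive the color, with the flares at quarter intensity; the construction script is empty.
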(event Custom|setSpotlightProperties
  (params (Rotation%20Max :type Float%20%28double-precision%29) (Rotation%20Speed :type Float%20%28double-precision%29) (Beam%20Height :type Float%20%28double-precision%29) (Beam%20Width :type Float%20%28double-precision%29) (SpotlightColor :type Linear%20Color%20Structure))
  (Variables|SpotlightFXSettings|SetRotation%20Max Rotation%20Max)
  (Variables|SpotlightFXSettings|SetRotation%20Speed Rotation%20Speed)
  (Variables|SpotlightFXSettings|SetBeam%20Height Beam%20Height)
  (Variables|SpotlightFXSettings|SetBeam%20Width Beam%20Width)
  (Variables|SpotlightFXSettings|SetSpotlightColor SpotlightColor)
  (Rendering|Material|SetScalarParameterValueonMaterials (Variables|Default|Getsm_neo_tilted_light_shaft_Flares) "Rotation Max" (Variables|SpotlightFXSettings|GetRotation%20Max))
  (Rendering|Material|SetScalarParameterValueonMaterials (Variables|Default|Getsm_neo_tilted_light_shaft) "Rotation Speed" (Variables|SpotlightFXSettings|GetRotation%20Speed))
  (Rendering|Material|SetScalarParameterValueonMaterials (Variables|Default|Getsm_neo_tilted_light_shaft) "ScaleHeight" (Variables|SpotlightFXSettings|GetBeam%20Height))
  (Rendering|Material|SetVectorParameterValueonMaterials (Variables|Default|Getsm_neo_tilted_light_shaft) "Color" (Math|Conversions|ToVector(LinearColor) (Variables|SpotlightFXSettings|GetSpotlightColor)))
  (Rendering|Material|SetVectorParameterValueonMaterials (Variables|Default|Getsm_neo_tilted_light_shaft_Flares) "Color" (* (Math|Conversions|ToVector(LinearColor) (Variables|SpotlightFXSettings|GetSpotlightColor)) 0.25))
  (Rendering|Material|SetScalarParameterValueonMaterials (Variables|Default|Getsm_neo_tilted_light_shaft) "Main Width" (Variables|SpotlightFXSettings|GetBeam%20Width)))

(fn editorOnly_SpawnFXActor
  (params (actorClass :type Actor%20Class%20Reference) (fxComp :type Persistent%20FXComponent%20Object%20Reference))
  (return 0))

(fn ConstructionScript)
