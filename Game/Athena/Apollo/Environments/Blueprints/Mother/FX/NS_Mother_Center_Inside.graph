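# Niagara system for the Mother environment, apparently its center interior, layering color-shifting vertical streaks with looping, shockwave-named mesh effects.
# Tall sprites spawn on a wide cylinder surface and accelerate upward.
# Two active mesh emitters spawn 8000 units above the system origin, the larger using CameraOffset; a third is disabled.
# All emitters cycle color and fade in and out over particle lifetime, and feed dynamic material parameters.
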NiagaraSystem /Game/Athena/Apollo/Environments/Blueprints/Mother/FX/NS_Mother_Center_Inside
dependencies {
 usedDataInterfaces [&NiagaraDataInterfaceCurve, &NiagaraDataInterfaceColorCurve, &NiagaraDataInterfaceVectorCurve]
 usedDynamicInputs []
 usedModules [&SystemState]
}
emitters {
 FN_Sprite001 {
  emitterUpdateScript [
   &EmitterScalability(
    Module.bApplySpawnCountScale=false
    Module.bApplySpawnCountScaleByDistanceFraction=false
    Module.bUseMaxDistance=false
    Module.bUseMinDistance=false
   ) {enabled=false}
   &EmitterLifeCycle() {enabled=false}
   &EmitterState()
   &SpawnRate(Use Spawn Probability=false)
  ]
  particleSpawnScript [
   /Niagara/Modules/Spawn/Initialization/InitializeParticle(
    Color=<rgba 1 1 1 15>
    Lifetime=&UniformRangedFloat(Minimum=2, Maximum=3)
    Module.Write Color=true
    Module.Write Lifetime=true
    Module.Write Mass=true
    Module.Write Material Random=true
    Module.Write Position=false
    Module.Write Scale=false
    Module.Write SpriteRotation=false
    Module.Write SpriteSize=true
    Position=&SimulationPosition()
    Sprite Rotation=&UniformRangedFloat(Minimum=0, Maximum=360)
    Sprite Size=&UniformRangedVector2D(Minimum=<650 5000>, Maximum=<1500 25000>)
   ) {enabled=false}
   /Niagara/Modules/Spawn/Initialization/V2/InitializeParticle(
    Lifetime=5
    Sprite Size Max=<250 5000>
    Sprite Size Min=<250 5000>
    Sprite Size Mode=::NewEnumerator2
    UsePositionOffset=false
    Write Material Random=true
   ) {module=InitializeParticle001}
   &AddVelocity(Velocity=&UniformRangedVector(Minimum=<-10 -10 0>, Maximum=<10 10 -1500>))
   &AccelerationForce(Acceleration=<0 0 2000>)
   &ApplyInitialForces()
   &CylinderLocation(
    Cylinder Height=600
    Cylinder Radius=5000
    Lathe Profile=DataInterface() {propertyValues {Curve {keys []}}}
    MaskSpawnGroup=false
    Module.Enable Lathe Profile=false
    Module.Override Randomness=false
    Module.Surface Only=true
    Use Endcaps In Surface Only Mode=false
   )
  ]
  particleUpdateScript [
   &ParticleState()
   &SolveForcesAndVelocity(Clamp Velocity=false, Limit Acceleration=false)
   &DynamicMaterialParameters(
    Index 0 Param 1=&RandomRangeFloat(Minimum=0, Maximum=1)
    Index 0 Param 2=&RandomRangeFloat(Minimum=0, Maximum=1)
    Index 0 Param 3=&FloatFromCurve(FloatCurve=DataInterface() {propertyValues {Curve {keys [<at 0 0>,<at 1 -0.2>]}}})
    Module.Param0WriteEnabled=true
    Module.Param1WriteEnabled=true
    Module.Param2WriteEnabled=true
    Module.Param3WriteEnabled=true
   )
   &ScaleColor(
    Scale Alpha=&FloatFromCurve(FloatCurve=DataInterface() {propertyValues {Curve {keys [
         {interpMode=RCIM_Cubic, tangentMode=RCTM_User, time=0, value=0}
         {arriveTangent=0.002, interpMode=RCIM_Cubic, leaveTangent=0.002, tangentMode=RCTM_User, time=0.25, value=1}
         {interpMode=RCIM_Cubic, tangentMode=RCTM_User, time=1.0011, value=0.002}
        ]}}})
    ScaleA=true
    ScaleRGB=true
   ) {enabled=false}
   &ScaleSpriteSize(Scale Factor=<1 5>)
   &ScaleColor(
    Scale Alpha=&FloatFromCurve(
     FloatCurve=DataInterface() {propertyValues {Curve {keys [
         {interpMode=RCIM_Cubic, tangentMode=RCTM_User, time=0, value=0}
         {interpMode=RCIM_Cubic, tangentMode=RCTM_User, time=0.5, value=1}
         {interpMode=RCIM_Cubic, tangentMode=RCTM_User, time=1, value=0}
        ]}}}
     Scale Curve=0.25
    )
    Scale RGB=&Multiply_Vector(
     A=&MakeVectorFromLinearColorRGB(Color=&ColorFromCurve(ColorCurve=DataInterface() {propertyValues {
         AlphaCurve {keys [<at 0 1 RCIM_Cubic>,<at 0.9977 1 RCIM_Cubic>]}
         BlueCurve {keys [<at 0.1175 0.3372 RCIM_Cubic>,<at 0.2682 1>,<at 0.4507 0.7758>,<at 0.5951 0.7084>,<at 0.7691 1.5>,<at 0.9071 0.807 RCIM_Cubic>]}
         GreenCurve {keys [<at 0.1175 0.0169 RCIM_Cubic>,<at 0.2682 0.0673>,<at 0.4507 0.0314>,<at 0.5951 0>,<at 0.7691 0.4532>,<at 0.9071 0.0766 RCIM_Cubic>]}
         RedCurve {keys [<at 0.1175 0.0058 RCIM_Cubic>,<at 0.2682 0.8485>,<at 0.4507 0.026>,<at 0.5951 0.2275>,<at 0.7691 0>,<at 0.9071 0.5873 RCIM_Cubic>]}
        }}))
     B=&VectorFromFloat(Value=2)
    )
    ScaleA=true
    ScaleRGB=true
   ) {module=ScaleColor001}
  ]
  rendererClasses [&NiagaraSpriteRendererProperties]
  renderers [{rendererClass=&NiagaraSpriteRendererProperties}]
 }
 HT_LoopShockwave_CamF {
  emitterUpdateScript [
   &EmitterState()
   &SpawnRate(Use Spawn Probability=false)
  ]
  particleSpawnScript [
   &InitializeParticle(
    Color=<rgba 3 3 3 0.3>
    Lifetime=4
    Mesh Scale Mode=::NewEnumerator1
    Mesh Uniform Scale Max=50
    Mesh Uniform Scale Min=50
    UsePositionOffset=false
    Write Material Random=true
   ) {module=InitializeParticle001}
   &InitialMeshOrientation(Orientation Axis=<0 0 0>, Orientation Vector=<0 0 1>, Use Orientation Vector=true, Use Rotation Vector=false)
   &SystemLocation(MaskSpawnGroup=false, Offset=<0 0 8000>)
  ]
  particleUpdateScript [
   &ParticleState()
   &ScaleColor(
    Scale Alpha=&FloatFromCurve(FloatCurve=DataInterface() {propertyValues {Curve {keys [
         {interpMode=RCIM_Cubic, tangentMode=RCTM_User, time=0.0023, value=0}
         {arriveTangent=2.0824, interpMode=RCIM_Cubic, leaveTangent=0.0112, tangentMode=RCTM_Break, time=0.0552, value=0.9565}
         {arriveTangent=-0.1758, interpMode=RCIM_Cubic, leaveTangent=-0.1758, tangentMode=RCTM_User, time=1.0011, value=0.002}
        ]}}})
    Scale RGB=<150 150 150>
    ScaleA=true
    ScaleRGB=true
   ) {enabled=false}
   &ScaleColor(
    Scale Alpha=&FloatFromCurve(
     FloatCurve=DataInterface() {propertyValues {Curve {keys [
         {interpMode=RCIM_Cubic, tangentMode=RCTM_User, time=0, value=0}
         {interpMode=RCIM_Cubic, tangentMode=RCTM_User, time=0.5, value=1}
         {interpMode=RCIM_Cubic, tangentMode=RCTM_User, time=1, value=0}
        ]}}}
     Scale Curve=0.25
    )
    Scale RGB=&Multiply_Vector(
     A=&MakeVectorFromLinearColorRGB(Color=&ColorFromCurve(ColorCurve=DataInterface() {propertyValues {
         AlphaCurve {keys [<at 0 1 RCIM_Cubic>,<at 0.9977 1 RCIM_Cubic>]}
         BlueCurve {keys [<at 0.1175 0.3372 RCIM_Cubic>,<at 0.2682 1>,<at 0.4507 0.7758>,<at 0.5951 0.7084>,<at 0.7691 1.5>,<at 0.9071 0.807 RCIM_Cubic>]}
         GreenCurve {keys [<at 0.1175 0.0169 RCIM_Cubic>,<at 0.2682 0.0673>,<at 0.4507 0.0314>,<at 0.5951 0>,<at 0.7691 0.4532>,<at 0.9071 0.0766 RCIM_Cubic>]}
         RedCurve {keys [<at 0.1175 0.0058 RCIM_Cubic>,<at 0.2682 0.8485>,<at 0.4507 0.026>,<at 0.5951 0.2275>,<at 0.7691 0>,<at 0.9071 0.5873 RCIM_Cubic>]}
        }}))
     B=&VectorFromFloat(Value=1)
    )
    ScaleA=true
    ScaleRGB=true
   ) {module=ScaleColor001}
   &DynamicMaterialParameters(
    Index 0 Param 1=&FloatFromCurve(FloatCurve=DataInterface() {propertyValues {Curve {keys [<at 0 2.5>,<at 1 0>]}}})
    Index 0 Param 2=&FloatFromCurve(FloatCurve=DataInterface() {propertyValues {Curve {keys [<at 0 3>,<at 1 0>]}}}, Scale Curve=5)
    Index 0 Param 3=0.2
    Module.Param0WriteEnabled=true
    Module.Param1WriteEnabled=true
    Module.Param2WriteEnabled=true
    Module.Param3WriteEnabled=true
   )
  ]
  rendererClasses [&NiagaraMeshRendererProperties]
  renderers [{rendererClass=&NiagaraMeshRendererProperties, rendererIndex=0}]
 }
 HT_LoopShockwave_CamF001 {
  emitterUpdateScript [
   &EmitterState()
   &SpawnRate(SpawnRate=0.5, Use Spawn Probability=false)
  ]
  particleSpawnScript [
   &InitializeParticle(
    Color=<rgba 3 3 3 0.3>
    Lifetime=4
    Mesh Scale Mode=::NewEnumerator1
    Mesh Uniform Scale Max=200
    Mesh Uniform Scale Min=200
    UsePositionOffset=false
    Write Material Random=true
   ) {module=InitializeParticle001}
   &InitialMeshOrientation(Orientation Axis=<0 0 0>, Orientation Vector=<0 0 1>, Use Orientation Vector=true, Use Rotation Vector=false) {enabled=false}
   &SystemLocation(MaskSpawnGroup=false, Offset=<0 0 8000>)
   &CameraOffset()
  ]
  particleUpdateScript [
   &ParticleState()
   &ScaleMeshSize(Scale Factor=/Niagara/DynamicInputs/ScaleByCurve/ScaleVectorByCurve(
     InitialValue:Vector3f=<1 1 1>
     ScaleCurve:NiagaraDataInterfaceVectorCurve=DataInterface() {propertyValues {
       XCurve {keys [
         {arriveTangent=7.5627, interpMode=RCIM_Cubic, leaveTangent=3.2485, tangentMode=RCTM_User, time=0, value=0}
         {interpMode=RCIM_Cubic, tangentMode=RCTM_User, time=1, value=1}
        ]}
       YCurve {keys [
         {arriveTangent=9.2714, interpMode=RCIM_Cubic, leaveTangent=3.2485, tangentMode=RCTM_User, time=0, value=0}
         {interpMode=RCIM_Cubic, tangentMode=RCTM_User, time=1, value=1}
        ]}
       ZCurve {keys [
         {arriveTangent=8.0047, interpMode=RCIM_Cubic, leaveTangent=3.2485, tangentMode=RCTM_User, time=0, value=0}
         {interpMode=RCIM_Cubic, tangentMode=RCTM_User, time=1, value=1}
        ]}
      }}
     CurveIndex:NiagaraFloat=$Particles.NormalizedAge
    )) {enabled=false}
   &ScaleColor(
    Scale Alpha=&FloatFromCurve(FloatCurve=DataInterface() {propertyValues {Curve {keys [
         {interpMode=RCIM_Cubic, tangentMode=RCTM_User, time=0.0023, value=0}
         {arriveTangent=2.0824, interpMode=RCIM_Cubic, leaveTangent=0.0112, tangentMode=RCTM_Break, time=0.0552, value=0.9565}
         {arriveTangent=-0.1758, interpMode=RCIM_Cubic, leaveTangent=-0.1758, tangentMode=RCTM_User, time=1.0011, value=0.002}
        ]}}})
    Scale RGB=<150 150 150>
    ScaleA=true
    ScaleRGB=true
   ) {enabled=false}
   &ScaleColor(
    Scale Alpha=&FloatFromCurve(
     FloatCurve=DataInterface() {propertyValues {Curve {keys [
         {interpMode=RCIM_Cubic, tangentMode=RCTM_User, time=0, value=0}
         {interpMode=RCIM_Cubic, tangentMode=RCTM_User, time=0.5, value=1}
         {interpMode=RCIM_Cubic, tangentMode=RCTM_User, time=1, value=0}
        ]}}}
     Scale Curve=0.5
    )
    Scale RGB=&Multiply_Vector(
     A=&MakeVectorFromLinearColorRGB(Color=&ColorFromCurve(ColorCurve=DataInterface() {propertyValues {
         AlphaCurve {keys [<at 0 1 RCIM_Cubic>,<at 0.9977 1 RCIM_Cubic>]}
         BlueCurve {keys [<at 0.1175 0.3372 RCIM_Cubic>,<at 0.2682 1>,<at 0.4507 0.7758>,<at 0.5951 0.7084>,<at 0.7691 1.5>,<at 0.9071 0.807 RCIM_Cubic>]}
         GreenCurve {keys [<at 0.1175 0.0169 RCIM_Cubic>,<at 0.2682 0.7257>,<at 0.4507 0.0314>,<at 0.5951 0>,<at 0.7691 0.4532>,<at 0.9071 0.0766 RCIM_Cubic>]}
         RedCurve {keys [<at 0.1175 0.0058 RCIM_Cubic>,<at 0.2682 0.1892>,<at 0.4507 0.026>,<at 0.5951 0.2275>,<at 0.7691 0>,<at 0.9071 0.5873 RCIM_Cubic>]}
        }}))
     B=&VectorFromFloat(Value=0.2)
    )
    ScaleA=true
    ScaleRGB=true
   ) {module=ScaleColor001}
   &DynamicMaterialParameters(
    Index 0 Param 1=1
    Index 0 Param 2=-1
    Index 0 Param 3=1
    Module.Param0WriteEnabled=true
    Module.Param1WriteEnabled=true
    Module.Param2WriteEnabled=true
    Module.Param3WriteEnabled=true
   )
  ]
  rendererClasses [&NiagaraMeshRendererProperties]
  renderers [{rendererClass=&NiagaraMeshRendererProperties, rendererIndex=0}]
 }
 HT_LoopShockwave_CamF002 {
  bEnabled=false
  emitterUpdateScript [
   &EmitterState()
   &SpawnRate(SpawnRate=0.5, Use Spawn Probability=false)
  ]
  particleSpawnScript [
   /Niagara/Modules/Spawn/Initialization/InitializeParticle(
    Color=<rgba 3 3 3 0.3>
    Lifetime=2
    Mesh Scale=&VectorFromFloat(Value=250)
    Module.Write Color=true
    Module.Write Lifetime=true
    Module.Write Mass=true
    Module.Write Material Random=true
    Module.Write Position=false
    Module.Write Scale=true
    Module.Write SpriteRotation=false
    Module.Write SpriteSize=false
    Position=&SimulationPosition()
   ) {enabled=false}
   /Niagara/Modules/Spawn/Initialization/V2/InitializeParticle(
    Color=<rgba 3 3 3 0.3>
    Lifetime=4
    Mesh Scale=<25 25 25>
    Mesh Scale Mode=::NewEnumerator4
    UsePositionOffset=false
    Write Material Random=true
   ) {module=InitializeParticle001}
   &InitialMeshOrientation(Orientation Axis=<0 0 0>, Orientation Vector=<0 0 1>, Use Orientation Vector=true, Use Rotation Vector=false)
   &SystemLocation(MaskSpawnGroup=false, Offset=<0 0 8000>)
  ]
  particleUpdateScript [
   &ParticleState()
   &ScaleMeshSize(Scale Factor=&ScaleVectorByCurve(
     InitialValue:Vector3f=<1 1 1>
     ScaleCurve:NiagaraDataInterfaceVectorCurve=DataInterface() {propertyValues {
       XCurve {keys [
         {arriveTangent=7.5627, interpMode=RCIM_Cubic, leaveTangent=3.2485, tangentMode=RCTM_User, time=0, value=0}
         {interpMode=RCIM_Cubic, tangentMode=RCTM_User, time=1, value=1}
        ]}
       YCurve {keys [
         {arriveTangent=9.2714, interpMode=RCIM_Cubic, leaveTangent=3.2485, tangentMode=RCTM_User, time=0, value=0}
         {interpMode=RCIM_Cubic, tangentMode=RCTM_User, time=1, value=1}
        ]}
       ZCurve {keys [
         {arriveTangent=8.0047, interpMode=RCIM_Cubic, leaveTangent=3.2485, tangentMode=RCTM_User, time=0, value=0}
         {interpMode=RCIM_Cubic, tangentMode=RCTM_User, time=1, value=1}
        ]}
      }}
     CurveIndex:NiagaraFloat=$Particles.NormalizedAge
    )) {enabled=false}
   &ScaleColor(
    Scale Alpha=&FloatFromCurve(FloatCurve=DataInterface() {propertyValues {Curve {keys [
         {interpMode=RCIM_Cubic, tangentMode=RCTM_User, time=0.0023, value=0}
         {arriveTangent=2.0824, interpMode=RCIM_Cubic, leaveTangent=0.0112, tangentMode=RCTM_Break, time=0.0552, value=0.9565}
         {arriveTangent=-0.1758, interpMode=RCIM_Cubic, leaveTangent=-0.1758, tangentMode=RCTM_User, time=1.0011, value=0.002}
        ]}}})
    Scale RGB=<150 150 150>
    ScaleA=true
    ScaleRGB=true
   ) {enabled=false}
   &ScaleColor(
    Scale Alpha=&FloatFromCurve(
     FloatCurve=DataInterface() {propertyValues {Curve {keys [
         {interpMode=RCIM_Cubic, tangentMode=RCTM_User, time=0, value=0}
         {interpMode=RCIM_Cubic, tangentMode=RCTM_User, time=0.5, value=1}
         {interpMode=RCIM_Cubic, tangentMode=RCTM_User, time=1, value=0}
        ]}}}
     Scale Curve=100
    )
    Scale RGB=&Multiply_Vector(
     A=&MakeVectorFromLinearColorRGB(Color=&ColorFromCurve(ColorCurve=DataInterface() {propertyValues {
         AlphaCurve {keys [<at 0 1 RCIM_Cubic>,<at 0.9977 1 RCIM_Cubic>]}
         BlueCurve {keys [<at 0.1175 0.3372 RCIM_Cubic>,<at 0.2682 1>,<at 0.4507 0.7758>,<at 0.5951 0.7084>,<at 0.7691 1.5>,<at 0.9071 0.807 RCIM_Cubic>]}
         GreenCurve {keys [<at 0.1175 0.0169 RCIM_Cubic>,<at 0.2682 0.0673>,<at 0.4507 0.0314>,<at 0.5951 0>,<at 0.7691 0.4532>,<at 0.9071 0.0766 RCIM_Cubic>]}
         RedCurve {keys [<at 0.1175 0.0058 RCIM_Cubic>,<at 0.2682 0.8485>,<at 0.4507 0.026>,<at 0.5951 0.2275>,<at 0.7691 0>,<at 0.9071 0.5873 RCIM_Cubic>]}
        }}))
     B=&VectorFromFloat(Value=0.1)
    )
    ScaleA=true
    ScaleRGB=true
   ) {module=ScaleColor001}
   &DynamicMaterialParameters(
    Index 0 Param 1=2
    Index 0 Param 2=1
    Index 0 Param 3=0.1
    Module.Param0WriteEnabled=true
    Module.Param1WriteEnabled=true
    Module.Param2WriteEnabled=true
    Module.Param3WriteEnabled=true
   )
  ]
  rendererClasses [&NiagaraMeshRendererProperties]
  renderers [{rendererClass=&NiagaraMeshRendererProperties, rendererIndex=0}]
 }
}
summary {
 systemName=NS_Mother_Center_Inside
 emitters [
  {emitterName=FN_Sprite001, rendererClasses [&NiagaraSpriteRendererProperties]}
  {emitterName=HT_LoopShockwave_CamF, rendererClasses [&NiagaraMeshRendererProperties]}
  {emitterName=HT_LoopShockwave_CamF001, rendererClasses [&NiagaraMeshRendererProperties]}
  {bEnabled=false, emitterName=HT_LoopShockwave_CamF002, rendererClasses [&NiagaraMeshRendererProperties]}
 ]
 userVariables []
}
