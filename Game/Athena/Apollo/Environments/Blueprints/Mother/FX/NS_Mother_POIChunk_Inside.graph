NiagaraSystem /Game/Athena/Apollo/Environments/Blueprints/Mother/FX/NS_Mother_POIChunk_Inside
dependencies {
 usedDataInterfaces [
  &NiagaraDataInterfaceCurve
  &NiagaraDataInterfaceGBuffer
  &NiagaraDataInterfaceVectorCurve
  &NiagaraDataInterfaceColorCurve
  &NiagaraDataInterfaceParticleRead
  &NiagaraDataInterfaceCamera
 ]
 usedDynamicInputs []
 usedModules [
  &SystemState
  &self:SystemSpawnScript.SystemScriptSource.SystemScriptGraph.NiagaraNodeAssignment_0.SetVariables_#1
 ]
}
emitters {
 FN_Kiwi_Inside_foreground {
  simTarget=GPUComputeSim
  emitterUpdateScript [
   &EmitterState(
    Enable Distance Culling=true
    MaxDistance=50000
    MinDistance=1000
    Module.bUseMaxDistance=true
    Module.bUseMinDistance=false
    Scalability Mode=::NewEnumerator1
   )
   &SpawnRate(
    SpawnRate=&MakeCustomFloatFromBool(Bool=$User.IsKiwi, True Float=8, False Float=0)
    Use Spawn Probability=false
   )
  ]
  particleSpawnScript [
   /SetVariables(
    Particles.GrabSphere:NiagaraFloat=/Niagara/DynamicInputs/Math/CalculateParticleRadius(
     Radius Calculation Type:ENiagaraCollisionRadiusOptions=::NewEnumerator1
     Particle Radius:NiagaraFloat=1
     Mesh Dimensions:Vector3f=<10 10 10>
     Method for Calculating Particle Radius:ENiagaraCalculateRadiusOptions=::NewEnumerator2
    )
    Particles.MeshIndex=&RandomRangeInt(Minimum=0, Maximum=6)
    Particles.VisibilityTag=1
   ) {isSetParameters=true, module=SetVariables_#2}
   &InitializeParticle(
    Lifetime=&RandomRangeFloat(Minimum=22, Maximum=25)
    Mesh Scale Mode=::NewEnumerator1
    Mesh Uniform Scale Max:NiagaraFloat=&Multiply_Float(A=1, B=$System.ScaleSpawn)
    Mesh Uniform Scale Min:NiagaraFloat=&Multiply_Float(A=0.5, B=$System.ScaleSpawn)
    UsePositionOffset=false
    Write Material Random=true
   ) {module=InitializeParticle001}
   &BoxLocation(
    Box Offset=<0 -17600 -20500>
    Box Size=<23000 6000 5000>
    Coordinate Space:ENiagaraCoordinateSpace=::Simulation
    MaskSpawnGroup=false
    Module.Override Seed=false
    Rotate Box=false
    Surface Only=false
   )
   /SetVariables(Particles.InitialPos:NiagaraPosition=$Particles.Position) {isSetParameters=true, module=SetVariables_#3}
   &InitialMeshOrientation(
    Rotation=&RandomRangeVector(Minimum=<0 0 0>, Maximum=<180 180 180>)
    Use Orientation Vector=true
    Use Rotation Vector=true
   )
   &CameraOffset(Camera Offset Amount=100)
   &AddVelocityInCone(
    Cone Angle=50
    Cone Axis=<0 0 1>
    Module.Use Velocity Falloff On Cone Axis=false
    Velocity Strength=&Multiply_Float(
     A=&RandomRangeFloat(Minimum=200, Maximum=600)
     B=&MakeCustomFloatFromBool(Bool=$User.IsKiwi, True Float=$User.KiwiSpeed, False Float=1)
    )
   )
  ]
  particleUpdateScript [
   &ParticleState()
   /SetVariables(Particles.VisibilityTag=&MakeCustomIntFromBool(Bool=$Transient.FirstFrame, True INT=1, False INT=0)) {isSetParameters=true, module=SetVariables_#4}
   &MeshRotationForce(Rotation=/Niagara/DynamicInputs/ScaleByCurve/ScaleVectorByCurve(
     InitialValue:Vector3f=&RandomRangeVector(Minimum=<-0.2 -0.2 -0.02>, Maximum=<0.2 0.2 0.2>)
     ScaleCurve:NiagaraDataInterfaceVectorCurve=DataInterface() {propertyValues {
       XCurve {keys [<at 0.0049 0.4248>,<at 1.0003 1>]}
       YCurve {keys [<at 0.0063 0.4615>,<at 1.0003 1>]}
       ZCurve {keys [<at 0.0077 0.4432>,<at 1.0003 1>]}
      }}
     CurveIndex:NiagaraFloat=$Particles.NormalizedAge
    ))
   &VectorNoiseForce(Module.Use Random Noise Time Interval=false)
   &GravityForce(Gravity=&Multiply_Vector(
     A=&RandomRangeVector(Minimum=<0 0 50>, Maximum=<0 0 60>)
     B=&VectorFromFloat(Value=$User.KiwiSpeed)
    ))
   &CurlNoiseForce(Mask Curl Noise=true, Noise Frequency=1500, Noise Strength=250, Pan Noise=true, Pan Noise Field=<0 0 1>) {module=CurlNoiseForce001}
   &PointAttractionForce(
    Attraction Radius=15000
    Attraction Strength=&Multiply_Float(A=10, B=$User.KiwiSpeed)
    Attractor Position Offset:Vector3f=<0 0 5000>
    Falloff Exponent=1
    Kill Within Radius=false
    Module.Use Falloff=true
    UseAttractorOffset=true
   )
   &Drag(
    Drag=&FloatFromCurve(FloatCurve=DataInterface() {propertyValues {Curve {keys [<at 0 1>,<at 0.1429 0.4003>,<at 1 0.5>]}}}, Scale Curve=0)
    Use Linear Drag=true
    Use Rotational Drag=false
   )
   &SolveForcesAndVelocity(Clamp Velocity=true, Limit Acceleration=false, Speed Limit=1500)
   &DynamicMaterialParameters(
    Index 0 Param 1=&FloatFromCurve(FloatCurve=DataInterface() {propertyValues {Curve {keys [<at 0.6 -1>,<at 1 1>]}}})
    Index 0 Param 3=&RandomRangeFloat(Minimum=0.15, Maximum=1)
    Index 0 Param 4=0.1
    Index 1 Param 1=512
    Module.Param0WriteEnabled=true
    Module.Param0WriteEnabled1=true
    Module.Param1WriteEnabled=true
    Module.Param1WriteEnabled1=true
    Module.Param2WriteEnabled=true
    Module.Param2WriteEnabled1=true
    Module.Param3WriteEnabled=true
    Module.Param3WriteEnabled1=true
    Write Parameter Index 1=true
   )
   &SolveRotationalForcesAndVelocity()
   &ScaleMeshSize(Scale Factor=&VectorFromFloat(Value=&FloatFromCurve(FloatCurve=DataInterface() {propertyValues {Curve {keys [
          {arriveTangent=4.1785, interpMode=RCIM_Cubic, leaveTangent=4.1785, tangentMode=RCTM_User, time=0, value=0}
          {arriveTangent=0.9995, interpMode=RCIM_Cubic, leaveTangent=0.9995, tangentMode=RCTM_User, time=0.2223, value=0.6825}
          {arriveTangent=-0.0405, interpMode=RCIM_Cubic, leaveTangent=-0.0405, tangentMode=RCTM_User, time=1, value=1}
         ]}}})))
   /SetVariables(Particles.GetAge:NiagaraFloat=$Particles.NormalizedAge) {isSetParameters=true, module=SetVariables_#5}
   &CameraOffset(Camera Offset Amount=150) {module=CameraOffset001}
   &ScaleMeshSizeBySpeed(
    Max Scale Factor=<1 1 1>
    Min Scale Factor:Vector3f=<0 0 0>
    Sample Scale Factor By Curve=true
    Scale Factor Curve=DataInterface() {propertyValues {Curve {keys [<at 0 0>,<at 0.2 1>,<at 1 1>]}}}
   )
   /SetVariables(Particles.GetVelocityinZ:NiagaraFloat=/Game/Effects/Niagara/DynamicInputs/Math/FN_Fit(
     Clamp:NiagaraBool=false
     Value:NiagaraFloat=&MakeFloatFromVector(Vector=$Particles.Velocity, Channel=::NewEnumerator2)
     SourceMin:NiagaraFloat=0
     SourceMax:NiagaraFloat=1500
     DestMin:NiagaraFloat=0
     DestMax:NiagaraFloat=1
    )) {isSetParameters=true, module=SetVariables_#6}
   &ScaleColor(
    Scale Alpha=&FloatFromCurve(FloatCurve=DataInterface() {propertyValues {Curve {keys [<at 0 0>,<at 0.1 2>,<at 0.8 2>,<at 1 0>]}}})
    Scale RGB=&MakeVectorFromLinearColorRGB(Color=&ColorFromCurve(
      ColorCurve=DataInterface() {propertyValues {
        AlphaCurve {keys [<at 0 1 RCIM_Cubic>,<at 1 1 RCIM_Cubic>]}
        BlueCurve {keys [<at 0 1>,<at 0.8457 1 RCIM_Cubic>]}
        GreenCurve {keys [<at 0 1>,<at 0.8457 0.4196 RCIM_Cubic>]}
        RedCurve {keys [<at 0 1>,<at 0.8457 0 RCIM_Cubic>]}
       }}
      CurveIndex=$Particles.GetVelocityinZ
     ))
    ScaleA=true
    ScaleRGB=true
   )
   &KillParticles(Kill Particles=&SetBoolByFloatComparison(
     A=&MakeFloatFromVector(Vector=$Particles.Velocity, Channel=::NewEnumerator2)
     B=400
     Comparison Type=::NewEnumerator6
    )) {enabled=false, module=KillParticles002}
  ]
  rendererClasses [&NiagaraMeshRendererProperties]
  renderers [{rendererClass=&NiagaraMeshRendererProperties, rendererIndex=0}]
 }
 FN_Mesh {
  bEnabled=false
  simTarget=GPUComputeSim
  emitterUpdateScript [
   &EmitterState(
    Enable Distance Culling=true
    MaxDistance=25000
    Min Distance Response=::SleepAndLetParticlesFinish
    MinDistance=1000
    Module.bUseMaxDistance=true
    Module.bUseMinDistance=true
    Scalability Mode=::NewEnumerator1
   )
   &SpawnRate(SpawnRate=5000, Use Spawn Probability=false)
   &SpawnBurst_Instantaneous(Spawn Count=2500, Use Spawn Probability=false) {enabled=false}
  ]
  particleSpawnScript [
   /SetVariables(
    Particles.GrabSphere:NiagaraFloat=&CalculateParticleRadius(
     Radius Calculation Type:ENiagaraCollisionRadiusOptions=::NewEnumerator1
     Particle Radius:NiagaraFloat=1
     Mesh Dimensions:Vector3f=<10 10 10>
     Method for Calculating Particle Radius:ENiagaraCalculateRadiusOptions=::NewEnumerator2
    )
    Particles.MeshIndex=&RandomRangeInt(Minimum=0, Maximum=6)
    Particles.VisibilityTag=1
   ) {isSetParameters=true, module=SetVariables_#2}
   &InitializeParticle(
    Lifetime Max=8
    Lifetime Min=4
    Lifetime Mode=::NewEnumerator1
    Mesh Scale Mode=::NewEnumerator1
    Mesh Uniform Scale Max:NiagaraFloat=&Multiply_Float(A=1, B=$System.ScaleSpawn)
    Mesh Uniform Scale Min:NiagaraFloat=&Multiply_Float(A=0.25, B=$System.ScaleSpawn)
    UsePositionOffset=false
    Write Material Random=true
   ) {module=InitializeParticle001}
   /Niagara/Modules/Spawn/Location/ConeLocation(
    Cone Angle=1
    Cone Axis:Vector3f=<0 0 1>
    Cone Length=2500
    Enable Wedge=true
    MaskSpawnGroup=false
    Module.Override Randomness=false
    Offset:Vector3f=<0 0 -1250>
   ) {enabled=false}
   &SphereLocation(
    Hemisphere Z=true
    MaskSpawnGroup=false
    Override Seed=false
    Sphere Radius=500
    Surface Only=true
    UseNonUniformScale=true
    UseOffset=true
    UseSphereOrientation=false
   ) {enabled=false}
   &CylinderLocation(
    Cylinder Height=5000
    Cylinder Radius=5000
    MaskSpawnGroup=false
    Module.Enable Lathe Profile=false
    Module.Override Randomness=false
    Module.Surface Only=false
    Surface Only Band Thickness=2500
   )
   /SetVariables(Particles.InitialPos:NiagaraPosition=$Particles.Position) {isSetParameters=true, module=SetVariables_#3}
   /Niagara/Modules/Math/PlaceParticlesOnDepthBuffer_GPU(Max Z Depth=0, Module.Use Max Z Depth=false)
   /Game/Effects/Niagara/Modules/Spawn/FN_Spawn_GBuffer_Color()
   &self:DistanceCullingCheck(MaximumDistance=50000, MinimumDistance=1000, New Camera Query:NiagaraDataInterfaceCamera=$System.CameraPick) {module=DistanceCullingCheck}
   &KillParticlesInVolume(Box Size:Vector3f=<8000 8000 15000>, Invert Volume=true, Kill Shape=::NewEnumerator1)
   &InitialMeshOrientation(
    Rotation=&RandomRangeVector(Minimum=<0 0 0>, Maximum=<180 180 180>)
    Use Orientation Vector=true
    Use Rotation Vector=true
   )
   &AddVelocity(Velocity=<0 0 25>)
   &CameraOffset(Camera Offset Amount=100)
   &AddVelocityInCone(Cone Angle=50, Cone Axis=<0 0 1>, Module.Use Velocity Falloff On Cone Axis=false, Velocity Strength=2500)
  ]
  particleUpdateScript [
   &ParticleState()
   /SetVariables(Particles.VisibilityTag=&MakeCustomIntFromBool(Bool=$Transient.FirstFrame, True INT=1, False INT=0)) {isSetParameters=true, module=SetVariables_#4}
   &MeshRotationForce(Rotation=&ScaleVectorByCurve(
     InitialValue:Vector3f=&RandomRangeVector(Minimum=<-5 -5 -1>, Maximum=<5 5 3>)
     ScaleCurve:NiagaraDataInterfaceVectorCurve=DataInterface() {propertyValues {
       XCurve {keys [<at 0.0049 0.4248>,<at 1.0003 1>]}
       YCurve {keys [<at 0.0063 0.4615>,<at 1.0003 1>]}
       ZCurve {keys [<at 0.0077 0.4432>,<at 1.0003 1>]}
      }}
     CurveIndex:NiagaraFloat=$Particles.NormalizedAge
    ))
   &VectorNoiseForce(Force Amount=1000, Module.Use Random Noise Time Interval=false)
   &GravityForce(Gravity=&Multiply_VectorByFloat(Vector=<0 0 1000>, Float=1))
   &CurlNoiseForce(Mask Curl Noise=true, Noise Frequency=1500, Noise Strength=250, Pan Noise=true, Pan Noise Field=<0 0 1>) {module=CurlNoiseForce001}
   &Drag(
    Drag=&FloatFromCurve(FloatCurve=DataInterface() {propertyValues {Curve {keys [<at 0 1>,<at 0.1429 0.4003>,<at 1 0.5>]}}}, Scale Curve=0)
    Use Linear Drag=true
    Use Rotational Drag=false
   )
   &PointAttractionForce(Attraction Radius=5000, Attraction Strength=0, Kill Within Radius=false, Module.Use Falloff=false, UseAttractorOffset=true)
   &Collision(
    Age Colliding Particles=false
    Collision Enabled=&SetBoolByFloatComparison(A=2, B=$Particles.Age)
    Control Roll On Collision=true
    Enable Rest State=false
    GPU Collision Type=::NewEnumerator2
    Mesh Dimensions:Vector3f=<10 10 10>
    Module.EnableMaxCollisionCount=false
    Module.Randomize Collision Normal Vector=true
    Particle Radius Scale=25
    Radius Calculation Type=::NewEnumerator1
    Restitution=0
    ScaleRateOfRotation=true
   ) {enabled=false}
   &SolveForcesAndVelocity(Clamp Velocity=true, Limit Acceleration=false)
   &DynamicMaterialParameters(
    Index 0 Param 1=&FloatFromCurve(FloatCurve=DataInterface() {propertyValues {Curve {keys [<at 0.6 -1>,<at 1 1>]}}})
    Index 0 Param 3=&RandomRangeFloat(Minimum=0.15, Maximum=1)
    Index 0 Param 4=0.1
    Index 1 Param 1=512
    Module.Param0WriteEnabled=true
    Module.Param0WriteEnabled1=true
    Module.Param1WriteEnabled=true
    Module.Param1WriteEnabled1=true
    Module.Param2WriteEnabled=true
    Module.Param2WriteEnabled1=true
    Module.Param3WriteEnabled=true
    Module.Param3WriteEnabled1=true
    Write Parameter Index 1=true
   )
   &SolveRotationalForcesAndVelocity()
   &ScaleMeshSize(Scale Factor=&RandomRangeVector(Minimum=<0.5 0.5 0.5>, Maximum=<2 2 2>))
   /SetVariables(Particles.GetAge:NiagaraFloat=$Particles.NormalizedAge) {isSetParameters=true, module=SetVariables_#5}
   &CameraOffset(Camera Offset Amount=150) {module=CameraOffset001}
   &ScaleMeshSizeBySpeed(
    Max Scale Factor=<1 1 1>
    Min Scale Factor:Vector3f=<0 0 0>
    Sample Scale Factor By Curve=true
    Scale Factor Curve=DataInterface() {propertyValues {Curve {keys [<at 0 0>,<at 0.2 1>,<at 1 1>]}}}
   )
   /SetVariables(Particles.GetVelocityinZ:NiagaraFloat=&FN_Fit(
     Clamp:NiagaraBool=false
     Value:NiagaraFloat=&MakeFloatFromVector(Vector=$Particles.Velocity, Channel=::NewEnumerator2)
     SourceMin:NiagaraFloat=0
     SourceMax:NiagaraFloat=1500
     DestMin:NiagaraFloat=0
     DestMax:NiagaraFloat=1
    )) {isSetParameters=true, module=SetVariables_#6}
   &ScaleColor(
    Scale Alpha=&FloatFromCurve(FloatCurve=DataInterface() {propertyValues {Curve {keys [<at 0 0>,<at 0.1 0.95>,<at 1 0>]}}})
    Scale RGB=&MakeVectorFromLinearColorRGB(Color=&ColorFromCurve(
      ColorCurve=DataInterface() {propertyValues {
        AlphaCurve {keys [<at 0 1 RCIM_Cubic>,<at 1 1 RCIM_Cubic>]}
        BlueCurve {keys [<at 0 1>,<at 0.6208 1>,<at 0.8457 224.7108 RCIM_Cubic>]}
        GreenCurve {keys [<at 0 1>,<at 0.6208 1>,<at 0.8457 0 RCIM_Cubic>]}
        RedCurve {keys [<at 0 1>,<at 0.6208 1>,<at 0.8457 250 RCIM_Cubic>]}
       }}
      CurveIndex=$Particles.GetVelocityinZ
     ))
    ScaleA=true
    ScaleRGB=true
   )
   &KillParticles(Kill Particles=&SetBoolByFloatComparison(
     A=&MakeFloatFromVector(Vector=$Particles.Velocity, Channel=::NewEnumerator2)
     B=400
     Comparison Type=::NewEnumerator6
    )) {enabled=false, module=KillParticles002}
  ]
  rendererClasses [&NiagaraMeshRendererProperties]
  renderers [{rendererClass=&NiagaraMeshRendererProperties, rendererIndex=0}]
 }
 FN_Mesh002 {
  simTarget=GPUComputeSim
  emitterUpdateScript [
   &EmitterState(
    Enable Distance Culling=true
    MaxDistance=50000
    MinDistance=1000
    Module.bUseMaxDistance=true
    Module.bUseMinDistance=false
    Scalability Mode=::NewEnumerator1
   )
   &SpawnRate(
    SpawnRate=&MakeCustomFloatFromBool(Bool=$User.IsKiwi, True Float=50, False Float=100)
    Use Spawn Probability=false
   )
   &SpawnBurst_Instantaneous(Spawn Count=2500, Use Spawn Probability=false) {enabled=false}
  ]
  particleSpawnScript [
   /SetVariables(
    Particles.GrabSphere:NiagaraFloat=&CalculateParticleRadius(
     Radius Calculation Type:ENiagaraCollisionRadiusOptions=::NewEnumerator1
     Particle Radius:NiagaraFloat=1
     Mesh Dimensions:Vector3f=<10 10 10>
     Method for Calculating Particle Radius:ENiagaraCalculateRadiusOptions=::NewEnumerator2
    )
    Particles.MeshIndex=&RandomRangeInt(Minimum=0, Maximum=6)
    Particles.VisibilityTag=1
   ) {isSetParameters=true, module=SetVariables_#2}
   &InitializeParticle(
    Lifetime=&MakeCustomFloatFromBool(
     Bool=$User.IsKiwi
     True Float=&RandomRangeFloat(Minimum=8, Maximum=12)
     False Float=&RandomRangeFloat(Minimum=4, Maximum=8)
    )
    Mesh Scale Mode=::NewEnumerator1
    Mesh Uniform Scale Max:NiagaraFloat=&Multiply_Float(A=1, B=$System.ScaleSpawn)
    Mesh Uniform Scale Min:NiagaraFloat=&Multiply_Float(A=0.5, B=$System.ScaleSpawn)
    UsePositionOffset=false
    Write Material Random=true
   ) {module=InitializeParticle001}
   &ConeLocation(
    Cone Angle=1
    Cone Axis:Vector3f=<0 0 1>
    Cone Length=2500
    Enable Wedge=true
    MaskSpawnGroup=false
    Module.Override Randomness=false
    Offset:Vector3f=<0 0 -1250>
   ) {enabled=false}
   &SphereLocation(
    Hemisphere Z=true
    MaskSpawnGroup=false
    Override Seed=false
    Sphere Radius=500
    Surface Only=true
    UseNonUniformScale=true
    UseOffset=true
    UseSphereOrientation=false
   ) {enabled=false}
   &CylinderLocation(
    Cylinder Height=15000
    Cylinder Radius=15000
    MaskSpawnGroup=false
    Module.Enable Lathe Profile=false
    Module.Override Randomness=false
    Module.Surface Only=false
    Offset=<0 0 -5000>
    Surface Only Band Thickness=2500
   )
   /SetVariables(Particles.InitialPos:NiagaraPosition=$Particles.Position) {isSetParameters=true, module=SetVariables_#3}
   &FN_Spawn_GBuffer_Color() {enabled=false}
   &self:DistanceCullingCheck(MaximumDistance=50000, MinimumDistance=1000, New Camera Query:NiagaraDataInterfaceCamera=$System.CameraPick) {enabled=false, module=DistanceCullingCheck}
   &KillParticlesInVolume(Box Size:Vector3f=<8000 8000 15000>, Invert Volume=true, Kill Shape=::NewEnumerator1) {enabled=false}
   &InitialMeshOrientation(
    Rotation=&RandomRangeVector(Minimum=<0 0 0>, Maximum=<180 180 180>)
    Use Orientation Vector=true
    Use Rotation Vector=true
   )
   &CameraOffset(Camera Offset Amount=100)
   &AddVelocityInCone(
    Cone Angle=50
    Cone Axis=<0 0 1>
    Module.Use Velocity Falloff On Cone Axis=false
    Velocity Strength=&Multiply_Float(
     A=&RandomRangeFloat(Minimum=400, Maximum=1200)
     B=&MakeCustomFloatFromBool(Bool=$User.IsKiwi, True Float=$User.KiwiSpeed, False Float=1)
    )
   )
  ]
  particleUpdateScript [
   &ParticleState()
   /SetVariables(Particles.VisibilityTag=&MakeCustomIntFromBool(Bool=$Transient.FirstFrame, True INT=1, False INT=0)) {isSetParameters=true, module=SetVariables_#4}
   &MeshRotationForce(Rotation=&ScaleVectorByCurve(
     InitialValue:Vector3f=&MakeCustomVectorFromBool(
      Bool=$User.IsKiwi
      True Vector=&RandomRangeVector(Minimum=<-0.5 -0.5 -0.15>, Maximum=<0.5 0.5 0.5>)
      False Vector=&RandomRangeVector(Minimum=<-3 -3 -1>, Maximum=<3 3 3>)
     )
     ScaleCurve:NiagaraDataInterfaceVectorCurve=DataInterface() {propertyValues {
       XCurve {keys [<at 0.0049 0.4248>,<at 1.0003 1>]}
       YCurve {keys [<at 0.0063 0.4615>,<at 1.0003 1>]}
       ZCurve {keys [<at 0.0077 0.4432>,<at 1.0003 1>]}
      }}
     CurveIndex:NiagaraFloat=$Particles.NormalizedAge
    ))
   &VectorNoiseForce(Force Amount=1000, Module.Use Random Noise Time Interval=false)
   &GravityForce(Gravity=&Multiply_Vector(
     A=&RandomRangeVector(Minimum=<0 0 150>, Maximum=<0 0 200>)
     B=&VectorFromFloat(Value=&MakeCustomFloatFromBool(Bool=$User.IsKiwi, True Float=$User.KiwiSpeed, False Float=1))
    ))
   &CurlNoiseForce(Mask Curl Noise=true, Noise Frequency=1500, Noise Strength=250, Pan Noise=true, Pan Noise Field=<0 0 1>) {module=CurlNoiseForce001}
   &PointAttractionForce(
    Attraction Radius=15000
    Attraction Strength=&Multiply_Float(A=25, B=&MakeCustomFloatFromBool(Bool=$User.IsKiwi, True Float=$User.KiwiSpeed, False Float=1))
    Attractor Position Offset:Vector3f=<0 0 5000>
    Falloff Exponent=1
    Kill Within Radius=false
    Module.Use Falloff=true
    UseAttractorOffset=true
   )
   &Drag(
    Drag=&FloatFromCurve(FloatCurve=DataInterface() {propertyValues {Curve {keys [<at 0 1>,<at 0.1429 0.4003>,<at 1 0.5>]}}}, Scale Curve=0)
    Use Linear Drag=true
    Use Rotational Drag=false
   )
   &SolveForcesAndVelocity(Clamp Velocity=true, Limit Acceleration=false, Speed Limit=1500)
   &DynamicMaterialParameters(
    Index 0 Param 1=&FloatFromCurve(FloatCurve=DataInterface() {propertyValues {Curve {keys [<at 0.6 -1>,<at 1 1>]}}})
    Index 0 Param 3=&RandomRangeFloat(Minimum=0.15, Maximum=1)
    Index 0 Param 4=0.1
    Index 1 Param 1=512
    Module.Param0WriteEnabled=true
    Module.Param0WriteEnabled1=true
    Module.Param1WriteEnabled=true
    Module.Param1WriteEnabled1=true
    Module.Param2WriteEnabled=true
    Module.Param2WriteEnabled1=true
    Module.Param3WriteEnabled=true
    Module.Param3WriteEnabled1=true
    Write Parameter Index 1=true
   )
   &SolveRotationalForcesAndVelocity()
   &ScaleMeshSize(Scale Factor=&VectorFromFloat(Value=&FloatFromCurve(FloatCurve=DataInterface() {propertyValues {Curve {keys [
          {arriveTangent=4.1785, interpMode=RCIM_Cubic, leaveTangent=4.1785, tangentMode=RCTM_User, time=0, value=0}
          {arriveTangent=0.9995, interpMode=RCIM_Cubic, leaveTangent=0.9995, tangentMode=RCTM_User, time=0.2223, value=0.6825}
          {arriveTangent=-0.0405, interpMode=RCIM_Cubic, leaveTangent=-0.0405, tangentMode=RCTM_User, time=1, value=1}
         ]}}})))
   /SetVariables(Particles.GetAge:NiagaraFloat=$Particles.NormalizedAge) {isSetParameters=true, module=SetVariables_#5}
   &CameraOffset(Camera Offset Amount=150) {module=CameraOffset001}
   &ScaleMeshSizeBySpeed(
    Max Scale Factor=<1 1 1>
    Min Scale Factor:Vector3f=<0 0 0>
    Sample Scale Factor By Curve=true
    Scale Factor Curve=DataInterface() {propertyValues {Curve {keys [<at 0 0>,<at 0.2 1>,<at 1 1>]}}}
   )
   /SetVariables(Particles.GetVelocityinZ:NiagaraFloat=&FN_Fit(
     Clamp:NiagaraBool=false
     Value:NiagaraFloat=&MakeFloatFromVector(Vector=$Particles.Velocity, Channel=::NewEnumerator2)
     SourceMin:NiagaraFloat=0
     SourceMax:NiagaraFloat=1500
     DestMin:NiagaraFloat=0
     DestMax:NiagaraFloat=1
    )) {isSetParameters=true, module=SetVariables_#6}
   &ScaleColor(
    Scale Alpha=&FloatFromCurve(FloatCurve=DataInterface() {propertyValues {Curve {keys [<at 0 0>,<at 0.1 2>,<at 0.8 2>,<at 1 0>]}}})
    Scale RGB=&MakeVectorFromLinearColorRGB(Color=&ColorFromCurve(
      ColorCurve=DataInterface() {propertyValues {
        AlphaCurve {keys [<at 0 1 RCIM_Cubic>,<at 1 1 RCIM_Cubic>]}
        BlueCurve {keys [<at 0 1>,<at 0.8457 1 RCIM_Cubic>]}
        GreenCurve {keys [<at 0 1>,<at 0.8457 0.4196 RCIM_Cubic>]}
        RedCurve {keys [<at 0 1>,<at 0.8457 0 RCIM_Cubic>]}
       }}
      CurveIndex=$Particles.GetVelocityinZ
     ))
    ScaleA=true
    ScaleRGB=true
   )
   &KillParticles(Kill Particles=&SetBoolByFloatComparison(
     A=&MakeFloatFromVector(Vector=$Particles.Velocity, Channel=::NewEnumerator2)
     B=400
     Comparison Type=::NewEnumerator6
    )) {enabled=false, module=KillParticles002}
  ]
  rendererClasses [&NiagaraMeshRendererProperties]
  renderers [{rendererClass=&NiagaraMeshRendererProperties, rendererIndex=0}]
 }
 SpawnFrom {
  simTarget=GPUComputeSim
  emitterUpdateScript [
   &EmitterState(
    Enable Distance Culling=true
    MaxDistance=50000
    MinDistance=1000
    Module.bUseMaxDistance=true
    Module.bUseMinDistance=false
    Scalability Mode=::NewEnumerator1
   )
   &SpawnRate(
    SpawnRate=&MakeCustomFloatFromBool(Bool=$User.IsKiwi, True Float=45, False Float=100)
    Use Spawn Probability=false
   )
   &SpawnBurst_Instantaneous(Spawn Count=2500, Use Spawn Probability=false) {enabled=false}
  ]
  particleSpawnScript [
   /SetVariables(
    Particles.GrabSphere:NiagaraFloat=&CalculateParticleRadius(
     Radius Calculation Type:ENiagaraCollisionRadiusOptions=::NewEnumerator1
     Particle Radius:NiagaraFloat=1
     Mesh Dimensions:Vector3f=<10 10 10>
     Method for Calculating Particle Radius:ENiagaraCalculateRadiusOptions=::NewEnumerator2
    )
    Particles.MeshIndex=&RandomRangeInt(Minimum=0, Maximum=6)
    Particles.VisibilityTag=1
   ) {isSetParameters=true, module=SetVariables_#2}
   &InitializeParticle(
    Color=<rgba 0.2846 0.4283 1 1>
    Lifetime=&MakeCustomFloatFromBool(
     Bool=$User.IsKiwi
     True Float=&RandomRangeFloat(Minimum=8, Maximum=12)
     False Float=&RandomRangeFloat(Minimum=4, Maximum=8)
    )
    Mesh Scale Mode=::NewEnumerator1
    Mesh Uniform Scale Max:NiagaraFloat=&Multiply_Float(A=2, B=$System.ScaleSpawn)
    Mesh Uniform Scale Min:NiagaraFloat=&Multiply_Float(A=1, B=$System.ScaleSpawn)
    UsePositionOffset=false
    Write Material Random=true
   ) {module=InitializeParticle001}
   &ConeLocation(
    Cone Angle=1
    Cone Axis:Vector3f=<0 0 1>
    Cone Length=2500
    Enable Wedge=true
    MaskSpawnGroup=false
    Module.Override Randomness=false
    Offset:Vector3f=<0 0 -1250>
   ) {enabled=false}
   &SphereLocation(
    Hemisphere Z=true
    MaskSpawnGroup=false
    Override Seed=false
    Sphere Radius=500
    Surface Only=true
    UseNonUniformScale=true
    UseOffset=true
    UseSphereOrientation=false
   ) {enabled=false}
   &CylinderLocation(
    Cylinder Height=15000
    Cylinder Radius=15000
    MaskSpawnGroup=false
    Module.Enable Lathe Profile=false
    Module.Override Randomness=false
    Module.Surface Only=false
    Offset=<0 0 -5000>
    Surface Only Band Thickness=2500
   )
   /SetVariables(Particles.InitialPos:NiagaraPosition=$Particles.Position) {isSetParameters=true, module=SetVariables_#3}
   &FN_Spawn_GBuffer_Color() {enabled=false}
   &self:DistanceCullingCheck(MaximumDistance=50000, MinimumDistance=1000, New Camera Query:NiagaraDataInterfaceCamera=$System.CameraPick) {enabled=false, module=DistanceCullingCheck}
   &KillParticlesInVolume(Box Size:Vector3f=<8000 8000 15000>, Invert Volume=true, Kill Shape=::NewEnumerator1) {enabled=false}
   &InitialMeshOrientation(
    Rotation=&RandomRangeVector(Minimum=<0 0 0>, Maximum=<180 180 180>)
    Use Orientation Vector=true
    Use Rotation Vector=true
   )
   &CameraOffset(Camera Offset Amount=100)
   &AddVelocityInCone(
    Cone Angle=50
    Cone Axis=<0 0 1>
    Module.Use Velocity Falloff On Cone Axis=false
    Velocity Strength=&Multiply_Float(
     A=&RandomRangeFloat(Minimum=400, Maximum=1200)
     B=&MakeCustomFloatFromBool(Bool=$User.IsKiwi, True Float=$User.KiwiSpeed, False Float=1)
    )
   )
  ]
  particleUpdateScript [
   &ParticleState()
   /SetVariables(Particles.VisibilityTag=&MakeCustomIntFromBool(Bool=$Transient.FirstFrame, True INT=1, False INT=0)) {isSetParameters=true, module=SetVariables_#4}
   &MeshRotationForce(Rotation=&ScaleVectorByCurve(
     InitialValue:Vector3f=&MakeCustomVectorFromBool(
      Bool=$User.IsKiwi
      True Vector=&RandomRangeVector(Minimum=<-0.5 -0.5 -0.15>, Maximum=<0.5 0.5 0.5>)
      False Vector=&RandomRangeVector(Minimum=<-3 -3 -1>, Maximum=<3 3 3>)
     )
     ScaleCurve:NiagaraDataInterfaceVectorCurve=DataInterface() {propertyValues {
       XCurve {keys [<at 0.0049 0.4248>,<at 1.0003 1>]}
       YCurve {keys [<at 0.0063 0.4615>,<at 1.0003 1>]}
       ZCurve {keys [<at 0.0077 0.4432>,<at 1.0003 1>]}
      }}
     CurveIndex:NiagaraFloat=$Particles.NormalizedAge
    ))
   &VectorNoiseForce(Force Amount=1000, Module.Use Random Noise Time Interval=false)
   &GravityForce(Gravity=&Multiply_Vector(
     A=&RandomRangeVector(Minimum=<0 0 150>, Maximum=<0 0 200>)
     B=&VectorFromFloat(Value=&MakeCustomFloatFromBool(Bool=$User.IsKiwi, True Float=$User.KiwiSpeed, False Float=1))
    ))
   &CurlNoiseForce(Mask Curl Noise=true, Noise Frequency=1500, Noise Strength=250, Pan Noise=true, Pan Noise Field=<0 0 1>) {module=CurlNoiseForce001}
   &PointAttractionForce(
    Attraction Radius=15000
    Attraction Strength=&Multiply_Float(A=25, B=&MakeCustomFloatFromBool(Bool=$User.IsKiwi, True Float=$User.KiwiSpeed, False Float=1))
    Attractor Position Offset:Vector3f=<0 0 5000>
    Falloff Exponent=1
    Kill Within Radius=false
    Module.Use Falloff=true
    UseAttractorOffset=true
   )
   &Drag(
    Drag=&FloatFromCurve(FloatCurve=DataInterface() {propertyValues {Curve {keys [<at 0 1>,<at 0.1429 0.4003>,<at 1 0.5>]}}}, Scale Curve=0)
    Use Linear Drag=true
    Use Rotational Drag=false
   )
   &SolveForcesAndVelocity(Clamp Velocity=true, Limit Acceleration=false, Speed Limit=1500)
   &DynamicMaterialParameters(
    Index 0 Param 1=&FloatFromCurve(FloatCurve=DataInterface() {propertyValues {Curve {keys [<at 0.6 -1>,<at 1 1>]}}})
    Index 0 Param 3=&RandomRangeFloat(Minimum=0.15, Maximum=1)
    Index 0 Param 4=0.1
    Index 1 Param 1=512
    Module.Param0WriteEnabled=true
    Module.Param0WriteEnabled1=true
    Module.Param1WriteEnabled=true
    Module.Param1WriteEnabled1=true
    Module.Param2WriteEnabled=true
    Module.Param2WriteEnabled1=true
    Module.Param3WriteEnabled=true
    Module.Param3WriteEnabled1=true
    Write Parameter Index 1=true
   )
   &SolveRotationalForcesAndVelocity()
   &ScaleMeshSize(Scale Factor=&VectorFromFloat(Value=&FloatFromCurve(FloatCurve=DataInterface() {propertyValues {Curve {keys [
          {arriveTangent=4.1785, interpMode=RCIM_Cubic, leaveTangent=4.1785, tangentMode=RCTM_User, time=0, value=0}
          {arriveTangent=0.9995, interpMode=RCIM_Cubic, leaveTangent=0.9995, tangentMode=RCTM_User, time=0.2223, value=0.6825}
          {arriveTangent=-0.0405, interpMode=RCIM_Cubic, leaveTangent=-0.0405, tangentMode=RCTM_User, time=1, value=1}
         ]}}})))
   /SetVariables(Particles.GetAge:NiagaraFloat=$Particles.NormalizedAge) {isSetParameters=true, module=SetVariables_#5}
   &CameraOffset(Camera Offset Amount=150) {module=CameraOffset001}
   &ScaleMeshSizeBySpeed(
    Max Scale Factor=<1 1 1>
    Min Scale Factor:Vector3f=<0 0 0>
    Sample Scale Factor By Curve=true
    Scale Factor Curve=DataInterface() {propertyValues {Curve {keys [<at 0 0>,<at 0.2 1>,<at 1 1>]}}}
   )
   /SetVariables(Particles.GetVelocityinZ:NiagaraFloat=&FN_Fit(
     Clamp:NiagaraBool=false
     Value:NiagaraFloat=&MakeFloatFromVector(Vector=$Particles.Velocity, Channel=::NewEnumerator2)
     SourceMin:NiagaraFloat=0
     SourceMax:NiagaraFloat=1500
     DestMin:NiagaraFloat=0
     DestMax:NiagaraFloat=1
    )) {isSetParameters=true, module=SetVariables_#6}
   &ScaleColor(
    Scale Alpha=&FloatFromCurve(FloatCurve=DataInterface() {propertyValues {Curve {keys [<at 0 0>,<at 0.1 2>,<at 0.8 2>,<at 1 0>]}}})
    Scale RGB=&MakeVectorFromLinearColorRGB(Color=&ColorFromCurve(
      ColorCurve=DataInterface() {propertyValues {
        AlphaCurve {keys [<at 0 1 RCIM_Cubic>,<at 1 1 RCIM_Cubic>]}
        BlueCurve {keys [<at 0 1>,<at 0.8457 1 RCIM_Cubic>]}
        GreenCurve {keys [<at 0 1>,<at 0.8457 0.4196 RCIM_Cubic>]}
        RedCurve {keys [<at 0 1>,<at 0.8457 0 RCIM_Cubic>]}
       }}
      CurveIndex=$Particles.GetVelocityinZ
     ))
    ScaleA=true
    ScaleRGB=true
   )
   &KillParticles(Kill Particles=&SetBoolByFloatComparison(
     A=&MakeFloatFromVector(Vector=$Particles.Velocity, Channel=::NewEnumerator2)
     B=400
     Comparison Type=::NewEnumerator6
    )) {enabled=false, module=KillParticles002}
  ]
  rendererClasses [&NiagaraMeshRendererProperties]
  renderers [{rendererClass=&NiagaraMeshRendererProperties, rendererIndex=0}]
 }
 Sprite {
  simTarget=GPUComputeSim
  emitterUpdateScript [
   &EmitterState(
    Enable Distance Culling=true
    MaxDistance=50000
    MinDistance=1000
    Module.bUseMaxDistance=true
    Module.bUseMinDistance=false
    Scalability Mode=::NewEnumerator1
   )
   &SpawnRate(SpawnRate=5000, Use Spawn Probability=false)
   &SpawnBurst_Instantaneous(Spawn Count=2500, Use Spawn Probability=false) {enabled=false}
  ]
  particleSpawnScript [
   &InitializeParticle(
    Lifetime Max=8
    Lifetime Min=4
    Lifetime Mode=::NewEnumerator1
    Sprite Rotation Mode=::NewEnumerator2
    Sprite Size Mode=::NewEnumerator1
    Uniform Sprite Size Max=&Multiply_Float(A=50, B=$System.ScaleSpawn)
    Uniform Sprite Size Min=&Multiply_Float(A=15, B=$System.ScaleSpawn)
    UsePositionOffset=false
    Write Material Random=true
   ) {module=InitializeParticle001}
   &CylinderLocation(
    Cylinder Height=15000
    Cylinder Radius=15000
    MaskSpawnGroup=false
    Module.Enable Lathe Profile=false
    Module.Override Randomness=false
    Module.Surface Only=false
    Offset=<0 0 -5000>
    Surface Only Band Thickness=2500
   )
   /SetVariables(Particles.InitialPos:NiagaraPosition=$Particles.Position) {isSetParameters=true, module=SetVariables_#3}
   &CameraOffset(Camera Offset Amount=100)
   &AddVelocityInCone(
    Cone Angle=50
    Cone Axis=<0 0 1>
    Module.Use Velocity Falloff On Cone Axis=false
    Velocity Strength=&RandomRangeFloat(Minimum=200, Maximum=400)
   )
   &SubUVAnimation(End Frame=3, SubUV Animation Mode:ENiagara_SubUVLookupMode=::NewEnumerator2)
  ]
  particleUpdateScript [
   &ParticleState()
   /SetVariables(Particles.VisibilityTag=&MakeCustomIntFromBool(Bool=$Transient.FirstFrame, True INT=1, False INT=0)) {isSetParameters=true, module=SetVariables_#4}
   &VectorNoiseForce(Force Amount=1000, Module.Use Random Noise Time Interval=false)
   &GravityForce(Gravity=&Multiply_VectorByFloat(Vector=&RandomRangeVector(Minimum=<0 0 150>, Maximum=<0 0 250>), Float=1))
   &CurlNoiseForce(Mask Curl Noise=true, Noise Frequency=1500, Noise Strength=250, Pan Noise=true, Pan Noise Field=<0 0 1>) {module=CurlNoiseForce001}
   &PointAttractionForce(
    Attraction Radius=15000
    Attraction Strength=25
    Attractor Position Offset:Vector3f=<0 0 5000>
    Falloff Exponent=1
    Kill Within Radius=false
    Module.Use Falloff=true
    UseAttractorOffset=true
   )
   &Drag(
    Drag=&FloatFromCurve(FloatCurve=DataInterface() {propertyValues {Curve {keys [<at 0 1>,<at 0.1429 0.4003>,<at 1 0.5>]}}}, Scale Curve=0)
    Use Linear Drag=true
    Use Rotational Drag=false
   )
   &Collision(
    Age Colliding Particles=false
    Collision Enabled=&SetBoolByFloatComparison(A=2, B=$Particles.Age)
    Control Roll On Collision=true
    Enable Rest State=false
    GPU Collision Type=::NewEnumerator2
    Mesh Dimensions:Vector3f=<10 10 10>
    Module.EnableMaxCollisionCount=false
    Module.Randomize Collision Normal Vector=true
    Particle Radius Scale=25
    Radius Calculation Type=::NewEnumerator1
    Restitution=0
    ScaleRateOfRotation=true
   ) {enabled=false}
   &SolveForcesAndVelocity(Clamp Velocity=true, Limit Acceleration=false, Speed Limit=1500)
   &DynamicMaterialParameters(
    Index 0 Param 1=&FloatFromCurve(FloatCurve=DataInterface() {propertyValues {Curve {keys [<at 0.6 -1>,<at 1 1>]}}})
    Index 0 Param 3=&RandomRangeFloat(Minimum=0.15, Maximum=1)
    Index 0 Param 4=0.1
    Index 1 Param 1=512
    Module.Param0WriteEnabled=true
    Module.Param0WriteEnabled1=true
    Module.Param1WriteEnabled=true
    Module.Param1WriteEnabled1=true
    Module.Param2WriteEnabled=true
    Module.Param2WriteEnabled1=true
    Module.Param3WriteEnabled=true
    Module.Param3WriteEnabled1=true
    Write Parameter Index 1=true
   )
   &SolveRotationalForcesAndVelocity()
   /SetVariables(Particles.GetAge:NiagaraFloat=$Particles.NormalizedAge) {isSetParameters=true, module=SetVariables_#5}
   &CameraOffset(Camera Offset Amount=150) {module=CameraOffset001}
   /SetVariables(Particles.GetVelocityinZ:NiagaraFloat=&FN_Fit(
     Clamp:NiagaraBool=false
     Value:NiagaraFloat=&MakeFloatFromVector(Vector=$Particles.Velocity, Channel=::NewEnumerator2)
     SourceMin:NiagaraFloat=0
     SourceMax:NiagaraFloat=1500
     DestMin:NiagaraFloat=0
     DestMax:NiagaraFloat=1
    )) {isSetParameters=true, module=SetVariables_#6}
   &ScaleColor(
    Scale Alpha=&FloatFromCurve(FloatCurve=DataInterface() {propertyValues {Curve {keys [<at 0 0>,<at 0.1 0.95>,<at 1 0>]}}})
    Scale RGB=&MakeVectorFromLinearColorRGB(Color=&ColorFromCurve(
      ColorCurve=DataInterface() {propertyValues {
        AlphaCurve {keys [<at 0 1 RCIM_Cubic>,<at 1 1 RCIM_Cubic>]}
        BlueCurve {keys [<at 0 1>,<at 0.8457 1 RCIM_Cubic>]}
        GreenCurve {keys [<at 0 1>,<at 0.8457 0.4196 RCIM_Cubic>]}
        RedCurve {keys [<at 0 1>,<at 0.8457 0 RCIM_Cubic>]}
       }}
      CurveIndex=$Particles.GetVelocityinZ
     ))
    ScaleA=true
    ScaleRGB=true
   )
   &KillParticles(Kill Particles=&SetBoolByFloatComparison(
     A=&MakeFloatFromVector(Vector=$Particles.Velocity, Channel=::NewEnumerator2)
     B=400
     Comparison Type=::NewEnumerator6
    )) {enabled=false, module=KillParticles002}
   &ScaleSpriteSize(Scale Factor=&Vector2DFromFloat(Value=&FloatFromCurve(FloatCurve=DataInterface() {propertyValues {Curve {keys [
          {interpMode=RCIM_Cubic, tangentMode=RCTM_User, time=0, value=0}
          {interpMode=RCIM_Cubic, tangentMode=RCTM_User, time=0.2, value=1}
          {tangentMode=RCTM_User, time=1, value=0}
         ]}}})))
   &SpriteRotationRate(Rotation Rate=&RandomRangeFloat(Minimum=0, Maximum=360))
  ]
  rendererClasses [&NiagaraSpriteRendererProperties]
  renderers [{rendererClass=&NiagaraSpriteRendererProperties}]
 }
 Sprite001 {
  simTarget=GPUComputeSim
  emitterUpdateScript [
   &EmitterState(
    Enable Distance Culling=true
    MaxDistance=50000
    MinDistance=1000
    Module.bUseMaxDistance=true
    Module.bUseMinDistance=false
    Scalability Mode=::NewEnumerator1
   )
   &SpawnRate(SpawnRate=0, Use Spawn Probability=false)
   &SpawnBurst_Instantaneous(Spawn Count=2500, Use Spawn Probability=false) {enabled=false}
   &SpawnParticlesFromOtherEmitter(
    Attribute Reader=DataInterface() {propertyValues {EmitterBinding {bindingMode=Other, emitterName=SpawnFrom}}}
    Maximum Age Spawn Threshold=3
    Spawn Rate=&MakeCustomFloatFromBool(Bool=$User.IsKiwi, True Float=10, False Float=25)
    Spawn Rate Per Particle Cap:NiagaraFloat=&MakeCustomFloatFromBool(Bool=$User.IsKiwi, True Float=10, False Float=25)
    Use Maximum Age=true
    Use Minimum Age=true
   )
  ]
  particleSpawnScript [
   &InitializeParticle(
    Color=<rgba 0.2846 0.4283 1 1>
    Lifetime Max=8
    Lifetime Min=4
    Lifetime Mode=::NewEnumerator1
    Sprite Rotation Mode=::NewEnumerator2
    Sprite Size Mode=::NewEnumerator1
    Uniform Sprite Size Max=&Multiply_Float(A=50, B=$System.ScaleSpawn)
    Uniform Sprite Size Min=&Multiply_Float(A=15, B=$System.ScaleSpawn)
    UsePositionOffset=false
    Write Material Random=true
   ) {module=InitializeParticle001}
   &SampleParticlesFromOtherEmitter(
    Applied Position Offset:Vector3f=&RandomRangeVector(Minimum=<0 0 0>, Maximum=<1 1 1>)
    Applied Velocity Scale:NiagaraFloat=&MakeCustomFloatFromBool(Bool=$User.IsKiwi, True Float=0.05, False Float=0.5)
    Velocity Application:ENiagara_DirectReadApplicationMode=::NewEnumerator0
    Velocity Sampling=::NewEnumerator0
   )
   &CylinderLocation(
    Cylinder Height=5
    Cylinder Radius=5
    MaskSpawnGroup=false
    Module.Enable Lathe Profile=false
    Module.Override Randomness=false
    Module.Surface Only=false
    Surface Only Band Thickness=2500
   )
   /SetVariables(Particles.InitialPos:NiagaraPosition=$Particles.Position) {isSetParameters=true, module=SetVariables_#3}
   &CameraOffset(Camera Offset Amount=100)
   &AddVelocityInCone(
    Cone Angle=50
    Cone Axis=<0 0 1>
    Module.Use Velocity Falloff On Cone Axis=false
    Velocity Strength=&RandomRangeFloat(Minimum=0, Maximum=0)
   ) {enabled=false}
   &SubUVAnimation(End Frame=3, SubUV Animation Mode:ENiagara_SubUVLookupMode=::NewEnumerator2)
  ]
  particleUpdateScript [
   &ParticleState()
   /SetVariables(Particles.VisibilityTag=&MakeCustomIntFromBool(Bool=$Transient.FirstFrame, True INT=1, False INT=0)) {isSetParameters=true, module=SetVariables_#4}
   &VectorNoiseForce(Force Amount=1000, Module.Use Random Noise Time Interval=false)
   &GravityForce(Gravity=&Multiply_VectorByFloat(Vector=&RandomRangeVector(Minimum=<0 0 0>, Maximum=<0 0 0>), Float=1)) {enabled=false}
   &CurlNoiseForce(
    Mask Curl Noise=true
    Noise Frequency=1
    Noise Strength=&MakeCustomFloatFromBool(Bool=$User.IsKiwi, True Float=50, False Float=500)
    Pan Noise=true
    Pan Noise Field=<0 0 1>
   ) {module=CurlNoiseForce001}
   &PointAttractionForce(
    Attraction Radius=15000
    Attraction Strength=25
    Attractor Position Offset:Vector3f=<0 0 5000>
    Falloff Exponent=1
    Kill Within Radius=false
    Module.Use Falloff=true
    UseAttractorOffset=true
   )
   &Drag(
    Drag=&FloatFromCurve(FloatCurve=DataInterface() {propertyValues {Curve {keys [<at 0 1>,<at 0.1429 0.4003>,<at 1 0.5>]}}}, Scale Curve=0)
    Use Linear Drag=true
    Use Rotational Drag=false
   )
   &Collision(
    Age Colliding Particles=false
    Collision Enabled=&SetBoolByFloatComparison(A=2, B=$Particles.Age)
    Control Roll On Collision=true
    Enable Rest State=false
    GPU Collision Type=::NewEnumerator2
    Mesh Dimensions:Vector3f=<10 10 10>
    Module.EnableMaxCollisionCount=false
    Module.Randomize Collision Normal Vector=true
    Particle Radius Scale=25
    Radius Calculation Type=::NewEnumerator1
    Restitution=0
    ScaleRateOfRotation=true
   ) {enabled=false}
   &SolveForcesAndVelocity(Clamp Velocity=true, Limit Acceleration=false, Speed Limit=1500)
   &DynamicMaterialParameters(
    Index 0 Param 1=&FloatFromCurve(FloatCurve=DataInterface() {propertyValues {Curve {keys [<at 0.6 -1>,<at 1 1>]}}})
    Index 0 Param 3=&RandomRangeFloat(Minimum=0.15, Maximum=1)
    Index 0 Param 4=0.1
    Index 1 Param 1=512
    Module.Param0WriteEnabled=true
    Module.Param0WriteEnabled1=true
    Module.Param1WriteEnabled=true
    Module.Param1WriteEnabled1=true
    Module.Param2WriteEnabled=true
    Module.Param2WriteEnabled1=true
    Module.Param3WriteEnabled=true
    Module.Param3WriteEnabled1=true
    Write Parameter Index 1=true
   )
   &SolveRotationalForcesAndVelocity()
   /SetVariables(Particles.GetAge:NiagaraFloat=$Particles.NormalizedAge) {isSetParameters=true, module=SetVariables_#5}
   &CameraOffset(Camera Offset Amount=150) {module=CameraOffset001}
   /SetVariables(Particles.GetVelocityinZ:NiagaraFloat=&FN_Fit(
     Clamp:NiagaraBool=false
     Value:NiagaraFloat=&MakeFloatFromVector(Vector=$Particles.Velocity, Channel=::NewEnumerator2)
     SourceMin:NiagaraFloat=0
     SourceMax:NiagaraFloat=1500
     DestMin:NiagaraFloat=0
     DestMax:NiagaraFloat=1
    )) {isSetParameters=true, module=SetVariables_#6}
   &ScaleColor(
    Scale Alpha=&FloatFromCurve(FloatCurve=DataInterface() {propertyValues {Curve {keys [<at 0 0>,<at 0.1 0.95>,<at 1 0>]}}})
    Scale RGB=&MakeVectorFromLinearColorRGB(Color=&ColorFromCurve(
      ColorCurve=DataInterface() {propertyValues {
        AlphaCurve {keys [<at 0 1 RCIM_Cubic>,<at 1 1 RCIM_Cubic>]}
        BlueCurve {keys [<at 0 1>,<at 0.8457 1 RCIM_Cubic>]}
        GreenCurve {keys [<at 0 1>,<at 0.8457 0.4196 RCIM_Cubic>]}
        RedCurve {keys [<at 0 1>,<at 0.8457 0 RCIM_Cubic>]}
       }}
      CurveIndex=$Particles.GetVelocityinZ
     ))
    ScaleA=true
    ScaleRGB=true
   )
   &KillParticles(Kill Particles=&SetBoolByFloatComparison(
     A=&MakeFloatFromVector(Vector=$Particles.Velocity, Channel=::NewEnumerator2)
     B=400
     Comparison Type=::NewEnumerator6
    )) {enabled=false, module=KillParticles002}
   &ScaleSpriteSize(Scale Factor=&Vector2DFromFloat(Value=&FloatFromCurve(FloatCurve=DataInterface() {propertyValues {Curve {keys [
          {interpMode=RCIM_Cubic, tangentMode=RCTM_User, time=0, value=0}
          {interpMode=RCIM_Cubic, tangentMode=RCTM_User, time=0.2, value=1}
          {tangentMode=RCTM_User, time=1, value=0}
         ]}}})))
   &SpriteRotationRate(Rotation Rate=&RandomRangeFloat(Minimum=0, Maximum=360))
  ]
  rendererClasses [&NiagaraSpriteRendererProperties]
  renderers [{rendererClass=&NiagaraSpriteRendererProperties}]
 }
}
summary {
 systemName=NS_Mother_POIChunk_Inside
 emitters [
  {bEnabled=false, emitterName=FN_Mesh, simTarget=GPUComputeSim, rendererClasses [&NiagaraMeshRendererProperties]}
  {emitterName=FN_Mesh002, simTarget=GPUComputeSim, rendererClasses [&NiagaraMeshRendererProperties]}
  {emitterName=Sprite, simTarget=GPUComputeSim, rendererClasses [&NiagaraSpriteRendererProperties]}
  {emitterName=SpawnFrom, simTarget=GPUComputeSim, rendererClasses [&NiagaraMeshRendererProperties]}
  {emitterName=Sprite001, simTarget=GPUComputeSim, rendererClasses [&NiagaraSpriteRendererProperties]}
  {emitterName=FN_Kiwi_Inside_foreground, simTarget=GPUComputeSim, rendererClasses [&NiagaraMeshRendererProperties]}
 ]
 userVariables [
  User.IsKiwi {type=NiagaraBool}
  User.KiwiSpeed {type=NiagaraFloat, defaultValue=0.14, description=""}
 ]
}
