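# Niagara effect for the Mother POI chunk underside, streaming tumbling, randomly chosen mesh debris upward.
# Two GPU mesh emitters spawn around a User.SpawnSphere-sized volume, with a central kill box and camera-distance culling; FN_Mesh also snaps particles to the depth buffer.
# Upward gravity, noise, drag and collision drive motion, while age and speed scale size.
# Color curves read vertical speed, though FN_Mesh001 writes that value to NewNiagaraFloat instead of the GetVelocityinZ it reads.
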NiagaraSystem /Game/Athena/Apollo/Environments/Blueprints/Mother/FX/NS_Mother_POIChunk_Under
dependencies {
 usedDataInterfaces [
  &NiagaraDataInterfaceCurve
  &NiagaraDataInterfaceGBuffer
  &NiagaraDataInterfaceVectorCurve
  &NiagaraDataInterfaceColorCurve
  &NiagaraDataInterfaceCamera
 ]
 usedDynamicInputs []
 usedModules [
  &SystemState
  &self:SystemSpawnScript.SystemScriptSource.SystemScriptGraph.NiagaraNodeAssignment_0.SetVariables_#1
 ]
}
emitters {
 FN_Mesh {
  simTarget=GPUComputeSim
  emitterUpdateScript [
   &EmitterState(
    Enable Distance Culling=true
    Enable Visibility Culling=true
    MaxDistance=35000
    Min Distance Response=::SleepAndClearParticles
    MinDistance=500
    Module.bApplySpawnCountScaleByDistance=true
    Module.bUseMaxDistance=true
    Module.bUseMinDistance=true
    Scalability Mode=::NewEnumerator1
    Scale Spawn Count=true
    Spawn Count Scale By Distance=DataInterface() {propertyValues {Curve {keys [<at 0 0>,<at 0.2 1>,<at 0.8 1>,<at 1 0>]}}}
    Visibility Cull Response:ENiagaraExecutionStateManagement=::SleepAndClearParticles
   )
   &SpawnRate(SpawnRate=&Multiply_Float(A=1, B=100), Use Spawn Probability=false)
   &SpawnBurst_Instantaneous(Spawn Count=2500, Use Spawn Probability=false) {enabled=false}
  ]
  particleSpawnScript [
   /SetVariables(
    Particles.GrabSphere:NiagaraFloat=/Niagara/DynamicInputs/Math/CalculateParticleRadius(
     Radius Calculation Type:ENiagaraCollisionRadiusOptions=::NewEnumerator1
     Particle Radius:NiagaraFloat=1
     Mesh Dimensions:Vector3f=<10 10 10>
     Method for Calculating Particle Radius:ENiagaraCalculateRadiusOptions=::NewEnumerator2
    )
    Particles.MeshIndex=&RandomRangeInt(Minimum=0, Maximum=3)
    Particles.VisibilityTag=1
   ) {isSetParameters=true, module=SetVariables_#2}
   &InitializeParticle(
    Lifetime Max=8
    Lifetime Min=4
    Lifetime Mode=::NewEnumerator1
    Mesh Scale Mode=::NewEnumerator1
    Mesh Uniform Scale Max:NiagaraFloat=&Multiply_Float(A=1, B=$System.ScaleSpawn)
    Mesh Uniform Scale Min:NiagaraFloat=&Multiply_Float(A=0.35, B=$System.ScaleSpawn)
    UsePositionOffset=false
    Write Material Random=true
   ) {module=InitializeParticle001}
   /Niagara/Modules/Spawn/Location/ConeLocation(
    Cone Angle=1
    Cone Axis:Vector3f=<0 0 1>
    Cone Length=2500
    Enable Wedge=true
    MaskSpawnGroup=false
    Module.Override Randomness=false
    Offset:Vector3f=<0 0 -1250>
   )
   &SphereLocation(
    Hemisphere Z=true
    MaskSpawnGroup=false
    Non Uniform Scale=<0.5 0.5 0.5>
    Override Seed=false
    Sphere Radius=$User.SpawnSphere
    Surface Only=true
    UseNonUniformScale=true
    UseOffset=true
    UseSphereOrientation=false
   )
   &CylinderLocation(
    Cylinder Radius=1
    MaskSpawnGroup=false
    Module.Enable Lathe Profile=false
    Module.Override Randomness=false
    Module.Surface Only=true
    Use Endcaps In Surface Only Mode=false
   )
   /SetVariables(Particles.InitialPos:NiagaraPosition=$Particles.Position) {isSetParameters=true, module=SetVariables_#3}
   /Niagara/Modules/Math/PlaceParticlesOnDepthBuffer_GPU(Max Z Depth=15000, Module.Use Max Z Depth=true)
   /Game/Effects/Niagara/Modules/Spawn/FN_Spawn_GBuffer_Color() {enabled=false}
   &self:DistanceCullingCheck(MaximumDistance=15000, MinimumDistance=3000, New Camera Query:NiagaraDataInterfaceCamera=$System.CameraPick) {module=DistanceCullingCheck}
   &KillParticlesInVolume(Box Size:Vector3f=<1500 1500 15000>, Kill Shape=::NewEnumerator1)
   &InitialMeshOrientation(
    Rotation=&RandomRangeVector(Minimum=<0 0 0>, Maximum=<180 180 180>)
    Use Orientation Vector=true
    Use Rotation Vector=true
   )
   &AddVelocity(Velocity=<0 0 25>)
   &CameraOffset(Camera Offset Amount=100)
   &AddVelocityInCone(Cone Angle=50, Cone Axis=<0 0 1>, Module.Use Velocity Falloff On Cone Axis=false, Velocity Strength=2500)
  ]
  particleUpdateScript [
   &ParticleState()
   /SetVariables(Particles.VisibilityTag=&MakeCustomIntFromBool(Bool=$Transient.FirstFrame, True INT=1, False INT=0)) {isSetParameters=true, module=SetVariables_#4}
   &MeshRotationForce(Rotation=/Niagara/DynamicInputs/ScaleByCurve/ScaleVectorByCurve(
     InitialValue:Vector3f=&RandomRangeVector(Minimum=<-2 -2 -1>, Maximum=<2 2 2>)
     ScaleCurve:NiagaraDataInterfaceVectorCurve=DataInterface() {propertyValues {
       XCurve {keys [<at 0.0049 0.4248>,<at 1.0003 1>]}
       YCurve {keys [<at 0.0063 0.4615>,<at 1.0003 1>]}
       ZCurve {keys [<at 0.0077 0.4432>,<at 1.0003 1>]}
      }}
     CurveIndex:NiagaraFloat=$Particles.NormalizedAge
    ))
   &VectorNoiseForce(Force Amount=1000, Module.Use Random Noise Time Interval=false)
   &GravityForce(Gravity=&Multiply_VectorByFloat(Vector=<0 0 1000>, Float=1))
   &CurlNoiseForce(Mask Curl Noise=true, Noise Frequency=1500, Noise Strength=250, Pan Noise=true, Pan Noise Field=<0 0 1>) {module=CurlNoiseForce001}
   &Drag(
    Drag=&FloatFromCurve(FloatCurve=DataInterface() {propertyValues {Curve {keys [<at 0 1>,<at 0.1429 0.4003>,<at 1 0.5>]}}})
    Use Linear Drag=true
    Use Rotational Drag=false
   )
   &PointAttractionForce(Attraction Radius=5000, Attraction Strength=0, Kill Within Radius=false, Module.Use Falloff=false, UseAttractorOffset=true)
   &Collision(
    Age Colliding Particles=false
    Collision Enabled=&SetBoolByFloatComparison(A=2, B=$Particles.Age)
    Control Roll On Collision=true
    Enable Rest State=false
    GPU Collision Type=::NewEnumerator2
    Mesh Dimensions:Vector3f=<10 10 10>
    Module.EnableMaxCollisionCount=false
    Module.Randomize Collision Normal Vector=true
    Particle Radius Scale=25
    Radius Calculation Type=::NewEnumerator1
    Restitution=0
    ScaleRateOfRotation=true
   )
   &SolveForcesAndVelocity(Clamp Velocity=true, Limit Acceleration=false, Speed Limit=500)
   &DynamicMaterialParameters(
    Index 0 Param 1=&FloatFromCurve(FloatCurve=DataInterface() {propertyValues {Curve {keys [<at 0.6 -1>,<at 1 1>]}}})
    Index 0 Param 3=&RandomRangeFloat(Minimum=0.15, Maximum=1)
    Index 0 Param 4=0.1
    Index 1 Param 1=512
    Module.Param0WriteEnabled=true
    Module.Param0WriteEnabled1=true
    Module.Param1WriteEnabled=true
    Module.Param1WriteEnabled1=true
    Module.Param2WriteEnabled=true
    Module.Param2WriteEnabled1=true
    Module.Param3WriteEnabled=true
    Module.Param3WriteEnabled1=true
    Write Parameter Index 1=true
   )
   &SolveRotationalForcesAndVelocity()
   &ScaleMeshSize(Scale Factor=&VectorFromFloat(Value=&FloatFromCurve(FloatCurve=DataInterface() {propertyValues {Curve {keys [<at 0 0>,<at 0.1 1>,<at 0.5 1>,<at 1 0>]}}})))
   /SetVariables(Particles.GetAge:NiagaraFloat=$Particles.NormalizedAge) {isSetParameters=true, module=SetVariables_#5}
   &CameraOffset(Camera Offset Amount=150) {module=CameraOffset001}
   &ScaleMeshSizeBySpeed(
    Max Scale Factor=<1 1 1>
    Min Scale Factor:Vector3f=<0 0 0>
    Sample Scale Factor By Curve=true
    Scale Factor Curve=DataInterface() {propertyValues {Curve {keys [<at 0 0>,<at 0.2 1>,<at 1 1>]}}}
   )
   /SetVariables(Particles.GetVelocityinZ:NiagaraFloat=/Game/Effects/Niagara/DynamicInputs/Math/FN_Fit(
     Clamp:NiagaraBool=false
     Value:NiagaraFloat=&MakeFloatFromVector(Vector=$Particles.Velocity, Channel=::NewEnumerator2)
     SourceMin:NiagaraFloat=0
     SourceMax:NiagaraFloat=550
     DestMin:NiagaraFloat=0
     DestMax:NiagaraFloat=1
    )) {isSetParameters=true, module=SetVariables_#6}
   &ScaleColor(
    Scale Alpha=&FloatFromCurve(FloatCurve=DataInterface() {propertyValues {Curve {keys [<at 0 0>,<at 0.1 0.95>,<at 1 0>]}}})
    Scale RGB=&MakeVectorFromLinearColorRGB(Color=&ColorFromCurve(
      ColorCurve=DataInterface() {propertyValues {
        AlphaCurve {keys [<at 0 1 RCIM_Cubic>,<at 1 1 RCIM_Cubic>]}
        BlueCurve {keys [<at 0 0.0065>,<at 0.8061 0.0176>,<at 0.8768 0.25 RCIM_Cubic>]}
        GreenCurve {keys [<at 0 0.0065>,<at 0.8061 0.0116>,<at 0.8768 0.1784 RCIM_Cubic>]}
        RedCurve {keys [<at 0 0.0048>,<at 0.8061 0.0065>,<at 0.8768 0 RCIM_Cubic>]}
       }}
      CurveIndex=$Particles.GetVelocityinZ
     ))
    ScaleA=true
    ScaleRGB=true
   )
   &KillParticles(Kill Particles=&SetBoolByFloatComparison(
     A=&MakeFloatFromVector(Vector=$Particles.Velocity, Channel=::NewEnumerator2)
     B=400
     Comparison Type=::NewEnumerator6
    )) {enabled=false, module=KillParticles002}
  ]
  rendererClasses [&NiagaraMeshRendererProperties]
  renderers [{rendererClass=&NiagaraMeshRendererProperties, rendererIndex=0}]
 }
 FN_Mesh001 {
  simTarget=GPUComputeSim
  emitterUpdateScript [
   &EmitterState(
    Enable Distance Culling=true
    Enable Visibility Culling=true
    MaxDistance=35000
    Min Distance Response=::SleepAndClearParticles
    MinDistance=500
    Module.bApplySpawnCountScaleByDistance=true
    Module.bUseMaxDistance=true
    Module.bUseMinDistance=true
    Scalability Mode=::NewEnumerator1
    Scale Spawn Count=true
    Spawn Count Scale By Distance=DataInterface() {propertyValues {Curve {keys [<at 0 0>,<at 0.1 1>,<at 0.9 1>,<at 1 0>]}}}
    Visibility Cull Response:ENiagaraExecutionStateManagement=::SleepAndClearParticles
   )
   &SpawnRate(SpawnRate=100, Use Spawn Probability=false)
   &SpawnBurst_Instantaneous(Spawn Count=2500, Use Spawn Probability=false) {enabled=false}
  ]
  particleSpawnScript [
   /SetVariables(
    Particles.GrabSphere:NiagaraFloat=&CalculateParticleRadius(
     Radius Calculation Type:ENiagaraCollisionRadiusOptions=::NewEnumerator1
     Particle Radius:NiagaraFloat=1
     Mesh Dimensions:Vector3f=<10 10 10>
     Method for Calculating Particle Radius:ENiagaraCalculateRadiusOptions=::NewEnumerator2
    )
    Particles.MeshIndex=&RandomRangeInt(Minimum=0, Maximum=3)
    Particles.VisibilityTag=1
   ) {isSetParameters=true, module=SetVariables_#2}
   &InitializeParticle(
    Lifetime Max=8
    Lifetime Min=4
    Lifetime Mode=::NewEnumerator1
    Mesh Scale Mode=::NewEnumerator1
    Mesh Uniform Scale Max:NiagaraFloat=&Multiply_Float(A=0.75, B=$System.ScaleSpawn)
    Mesh Uniform Scale Min:NiagaraFloat=&Multiply_Float(A=0.25, B=$System.ScaleSpawn)
    UsePositionOffset=false
    Write Material Random=true
   ) {module=InitializeParticle001}
   &ConeLocation(
    Cone Axis:Vector3f=<0 0 1>
    Cone Length=2500
    Enable Wedge=true
    MaskSpawnGroup=false
    Module.Override Randomness=false
    Offset:Vector3f=<0 0 -1250>
   )
   &SphereLocation(
    Hemisphere Z=true
    MaskSpawnGroup=false
    Non Uniform Scale=<0.5 0.5 0.5>
    Override Seed=false
    Sphere Radius=$User.SpawnSphere
    Surface Only=true
    UseNonUniformScale=true
    UseOffset=true
    UseSphereOrientation=false
   )
   &CylinderLocation(
    Cylinder Radius=500
    MaskSpawnGroup=false
    Module.Enable Lathe Profile=false
    Module.Override Randomness=false
    Module.Surface Only=true
    Use Endcaps In Surface Only Mode=false
   )
   /SetVariables(Particles.InitialPos:NiagaraPosition=$Particles.Position) {isSetParameters=true, module=SetVariables_#3}
   &FN_Spawn_GBuffer_Color() {enabled=false}
   &KillParticlesInVolume(Box Size:Vector3f=<1500 1500 15000>, Kill Shape=::NewEnumerator1)
   &self:DistanceCullingCheck(MaximumDistance=15000, MinimumDistance=3000, New Camera Query:NiagaraDataInterfaceCamera=$System.CameraPick) {module=DistanceCullingCheck}
   &InitialMeshOrientation(
    Rotation=&RandomRangeVector(Minimum=<0 0 0>, Maximum=<180 180 180>)
    Use Orientation Vector=true
    Use Rotation Vector=true
   )
   &AddVelocity(Velocity=<0 0 25>)
   &CameraOffset(Camera Offset Amount=100)
   &AddVelocityInCone(Cone Angle=50, Cone Axis=<0 0 1>, Module.Use Velocity Falloff On Cone Axis=false, Velocity Strength=2500)
  ]
  particleUpdateScript [
   &ParticleState()
   /SetVariables(Particles.VisibilityTag=&MakeCustomIntFromBool(Bool=$Transient.FirstFrame, True INT=1, False INT=0)) {isSetParameters=true, module=SetVariables_#4}
   &MeshRotationForce(Rotation=&ScaleVectorByCurve(
     InitialValue:Vector3f=&RandomRangeVector(Minimum=<-2 -2 -1>, Maximum=<2 2 2>)
     ScaleCurve:NiagaraDataInterfaceVectorCurve=DataInterface() {propertyValues {
       XCurve {keys [<at 0.0049 0.4248>,<at 1.0003 1>]}
       YCurve {keys [<at 0.0063 0.4615>,<at 1.0003 1>]}
       ZCurve {keys [<at 0.0077 0.4432>,<at 1.0003 1>]}
      }}
     CurveIndex:NiagaraFloat=$Particles.NormalizedAge
    ))
   &VectorNoiseForce(Force Amount=1000, Module.Use Random Noise Time Interval=false)
   &GravityForce(Gravity=&Multiply_VectorByFloat(Vector=<0 0 1000>, Float=1))
   &CurlNoiseForce(Mask Curl Noise=true, Noise Frequency=1500, Noise Strength=250, Pan Noise=true, Pan Noise Field=<0 0 1>) {module=CurlNoiseForce001}
   &Drag(
    Drag=&FloatFromCurve(FloatCurve=DataInterface() {propertyValues {Curve {keys [<at 0 1>,<at 0.1429 0.4003>,<at 1 0.5>]}}})
    Use Linear Drag=true
    Use Rotational Drag=false
   )
   &PointAttractionForce(Attraction Radius=5000, Attraction Strength=0, Kill Within Radius=false, Module.Use Falloff=false, UseAttractorOffset=true)
   &Collision(
    Age Colliding Particles=false
    Control Roll On Collision=true
    Enable Rest State=false
    GPU Collision Type=::NewEnumerator2
    Mesh Dimensions:Vector3f=<10 10 10>
    Module.EnableMaxCollisionCount=false
    Module.Randomize Collision Normal Vector=true
    Particle Radius Scale=25
    Radius Calculation Type=::NewEnumerator1
    Restitution=0
    ScaleRateOfRotation=true
   )
   &SolveForcesAndVelocity(Clamp Velocity=true, Limit Acceleration=false, Speed Limit=500)
   &DynamicMaterialParameters(
    Index 0 Param 1=&FloatFromCurve(FloatCurve=DataInterface() {propertyValues {Curve {keys [<at 0 -1>,<at 1 1>]}}})
    Index 0 Param 3=&RandomRangeFloat(Minimum=0.15, Maximum=1)
    Index 0 Param 4=0.1
    Index 1 Param 1=512
    Module.Param0WriteEnabled=true
    Module.Param0WriteEnabled1=true
    Module.Param1WriteEnabled=true
    Module.Param1WriteEnabled1=true
    Module.Param2WriteEnabled=true
    Module.Param2WriteEnabled1=true
    Module.Param3WriteEnabled=true
    Module.Param3WriteEnabled1=true
    Write Parameter Index 1=true
   )
   &SolveRotationalForcesAndVelocity()
   &ScaleMeshSize(Scale Factor=&VectorFromFloat(Value=&FloatFromCurve(FloatCurve=DataInterface() {propertyValues {Curve {keys [<at 0 0>,<at 0.1 1>,<at 0.5 1>,<at 1 0>]}}})))
   /SetVariables(Particles.GetAge:NiagaraFloat=$Particles.NormalizedAge) {isSetParameters=true, module=SetVariables_#5}
   &CameraOffset(Camera Offset Amount=150) {module=CameraOffset001}
   &ScaleMeshSizeBySpeed(
    Max Scale Factor=<1 1 1>
    Min Scale Factor:Vector3f=<0 0 0>
    Sample Scale Factor By Curve=true
    Scale Factor Curve=DataInterface() {propertyValues {Curve {keys [<at 0.1 0>,<at 0.2 1>,<at 1 1>]}}}
   )
   /SetVariables(Particles.NewNiagaraFloat:NiagaraFloat=&FN_Fit(
     Clamp:NiagaraBool=false
     Value:NiagaraFloat=&MakeFloatFromVector(Vector=$Particles.Velocity, Channel=::NewEnumerator2)
     SourceMin:NiagaraFloat=0
     SourceMax:NiagaraFloat=550
     DestMin:NiagaraFloat=0
     DestMax:NiagaraFloat=1
    )) {isSetParameters=true, module=SetVariables_#6}
   &ScaleColor(
    Scale Alpha=&FloatFromCurve(FloatCurve=DataInterface() {propertyValues {Curve {keys [<at 0 0>,<at 0.1 0.95>,<at 1 0>]}}})
    Scale RGB=&MakeVectorFromLinearColorRGB(Color=&ColorFromCurve(
      ColorCurve=DataInterface() {propertyValues {
        AlphaCurve {keys [<at 0 1 RCIM_Cubic>,<at 1 1 RCIM_Cubic>]}
        BlueCurve {keys [<at 0 1>,<at 0.8061 1>,<at 0.8768 5 RCIM_Cubic>]}
        GreenCurve {keys [<at 0 1>,<at 0.8061 1>,<at 0.8768 3.6951 RCIM_Cubic>]}
        RedCurve {keys [<at 0 1>,<at 0.8061 1>,<at 0.8768 0 RCIM_Cubic>]}
       }}
      CurveIndex=$Particles.GetVelocityinZ
     ))
    ScaleA=true
    ScaleRGB=true
   )
  ]
  rendererClasses [&NiagaraMeshRendererProperties]
  renderers [{rendererClass=&NiagaraMeshRendererProperties, rendererIndex=0}]
 }
}
summary {
 systemName=NS_Mother_POIChunk_Under
 emitters [
  {emitterName=FN_Mesh, simTarget=GPUComputeSim, rendererClasses [&NiagaraMeshRendererProperties]}
  {emitterName=FN_Mesh001, simTarget=GPUComputeSim, rendererClasses [&NiagaraMeshRendererProperties]}
 ]
 userVariables [User.SpawnSphere {type=NiagaraFloat, defaultValue=0, description=""}]
}
