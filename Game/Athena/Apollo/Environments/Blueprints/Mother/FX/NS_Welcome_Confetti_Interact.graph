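# Niagara system for welcome confetti that lies on the landscape and scatters as the view moves through it.
# A GPU emitter continuously spawns multicolored flipbook pieces, using landscape samples for terrain height and normal.
# A negative-strength attraction at a camera-relative point repels pieces, while active-state and terrain checks gate wind, curl noise, drag and sprite orientation between ground normal and camera-facing.
# A second, mesh-based variant emitter is disabled.
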NiagaraSystem /Game/Athena/Apollo/Environments/Blueprints/Mother/FX/NS_Welcome_Confetti_Interact
dependencies {
 usedDataInterfaces [&NiagaraDataInterfaceCamera, &NiagaraDataInterfaceCurve, &NiagaraDataInterfaceColorCurve]
 usedDynamicInputs []
 usedModules [&SystemState]
}
emitters {
 InteractiveConfetti {
  simTarget=GPUComputeSim
  emitterUpdateScript [
   /SetVariables() {isSetParameters=true, module=SetVariables_#1}
   /SetVariables(
    Emitter.CameraForward:Vector3f=&ReturnCameraVector(Camera Property=::NewEnumerator1, Coordinate Space=::Simulation, CameraQuery=$Emitter.CameraRead)
    Emitter.CameraPosition:Vector3f=&ReturnCameraVector(Camera Property=::NewEnumerator0, Coordinate Space=::World, CameraQuery=$Emitter.CameraRead)
    Emitter.CameraRight:Vector3f=&ReturnCameraVector(Camera Property=::NewEnumerator3, Coordinate Space=::Simulation, CameraQuery=$Emitter.CameraRead)
    Emitter.CameraUp:Vector3f=&ReturnCameraVector(Camera Property=::NewEnumerator2, Coordinate Space=::Simulation, CameraQuery=$Emitter.CameraRead)
   ) {isSetParameters=true, module=SetVariables_#2}
   &EmitterState(
    Enable Distance Culling=true
    Life Cycle Mode=::NewEnumerator1
    Loop Behavior=::NewEnumerator1
    Loop Duration=1
    Max Distance Response=::SleepAndClearParticles
    Module.bApplySpawnCountScaleByDistance=false
    Module.bUseMaxDistance=true
    Module.bUseMinDistance=false
    Reset Age On Awaken=true
    Scalability Mode=::NewEnumerator1
    Scale Spawn Count=true
    Spawn Count Scale By Distance=DataInterface() {propertyValues {Curve {keys [<at 0 0>,<at 0.3 0.95>,<at 0.75 1>,<at 1 0>]}}}
    UseLoopDelay=false
   )
   &SpawnBurst_Instantaneous(Spawn Count=7500, Use Spawn Probability=false) {enabled=false}
   &SpawnRate(SpawnRate=5000, Use Spawn Probability=false)
  ]
  particleSpawnScript [
   /Niagara/Modules/Spawn/Initialization/InitializeParticle(
    Color=&RandomRangeLinearColor(Minimum=<rgba 1 0.1164 0 1>, Maximum=<rgba 1 0.2736 0 1>)
    Lifetime=&RandomRangeFloat(Minimum=3, Maximum=5)
    Mass=5
    Module.Write Color=true
    Module.Write Lifetime=true
    Module.Write Mass=true
    Module.Write Material Random=true
    Module.Write Position=false
    Module.Write Scale=true
    Module.Write SpriteRotation=true
    Module.Write SpriteSize=true
    Position=&SimulationPosition()
    Sprite Rotation=&RandomRangeFloat(Minimum=0, Maximum=360)
    Sprite Size=&Vector2DFromFloat(Value=&RandomRangeFloat(Minimum=10, Maximum=15))
   )
   /SetVariables(Particles.Terrain Offset=5) {isSetParameters=true, module=SetVariables_#3}
   /Niagara/Modules/Spawn/Location/SphereLocation(MaskSpawnGroup=false, Module.Override Randomness=false, Sphere Radius=1000, Surface Only=false)
   &self:Sample_Landscape() {module=Sample_Landscape}
   /Niagara/Modules/Spawn/Orientation/V2/InitialMeshOrientation(
    Orientation Axis:Vector3f=<-1 0 0>
    Orientation Vector=<0 0 1>
    Rotation=&RandomRangeVector(Minimum=<0 0 0>, Maximum=<1 0 0>)
    Use Orientation Vector=true
    Use Rotation Vector=true
   ) {enabled=false}
   /Niagara/Modules/Spawn/Orientation/V3/InitialMeshOrientation(
    Orientation Axis=<-1 0 0>
    Orientation Vector=<0 0 1>
    Rotation=&RandomRangeVector(Minimum=<0 0 0>, Maximum=<1 0 0>)
    Use Orientation Vector=true
    Use Rotation Vector=true
   ) {module=InitialMeshOrientation001}
   /SetVariables(Particles.Position=&Add_Vector(
     A=Linked() {linkedVariable {name=Particles.Position, type=NiagaraPosition}}
     B=&MakeVector(X=0, Y=0, Z=$Particles.Terrain Offset)
    )) {isSetParameters=true, module=SetVariables_#4}
   &AddVelocity()
   &CameraOffset(Camera Offset Amount=15)
   &SubUVAnimation(End Frame=36, SubUV Animation Mode:ENiagara_SubUVLookupMode=::NewEnumerator2) {module=SubUVAnimation001}
   &Color(Color=&ColorFromCurve(
     ColorCurve=DataInterface() {propertyValues {
       AlphaCurve {keys [<at 0 1>,<at 1 1>]}
       BlueCurve {keys [<at 0 0.8714>,<at 0.2389 0.8714>,<at 0.2977 0.7011>,<at 0.4603 0.7011>,<at 0.5416 0.1624>,<at 1 0.1624>]}
       GreenCurve {keys [<at 0 0.5149>,<at 0.2389 0.5149>,<at 0.2977 0.063>,<at 0.4603 0.063>,<at 0.5416 0.9583>,<at 1 0.9583>]}
       RedCurve {keys [<at 0 0.0144>,<at 0.2389 0.0144>,<at 0.2977 1>,<at 0.4603 1>,<at 0.5416 0>,<at 1 0>]}
      }}
     CurveIndex=$Particles.MaterialRandom
    )) {module=Color001}
   /SetVariables(Particles.SubImageIndex=&RandomRangeFloat(Minimum=0, Maximum=35)) {isSetParameters=true}
   /SetVariables() {isSetParameters=true}
  ]
  particleUpdateScript [
   &self:Sample_Landscape_UPdate(Bias:NiagaraFloat=$Particles.Terrain Offset, New Landscape Sample:NiagaraDataInterfaceLandscape=$User.Landscape) {module=Sample_Landscape_UPdate}
   /SetVariables(Particles.Has Collided With Terrain:NiagaraBool=&SetBoolByFloatComparison(
     A=&Add_Float(A=$Output.Sample_Landscape_UPdate.Terrain Height, B=20)
     B=&MakeFloatFromVector(Vector=Linked() {linkedVariable {name=Particles.Position, type=NiagaraPosition}}, Channel=::NewEnumerator2)
    )) {isSetParameters=true, module=SetVariables_#5}
   /Niagara/Modules/Update/Forces/PointAttractionForce(
    Attraction Radius=80
    AttractionStrength=-85
    AttractorPosition:Vector3f=&Add_Vector(
     A=$Emitter.CameraPosition
     B=&Add_Vector(
      A=&Add_Vector(
       A=&Multiply_VectorByFloat(Vector=$Particles.Landscape Normal, Float=-130)
       B=&Multiply_VectorByFloat(Vector=$Emitter.CameraForward, Float=300)
      )
      B=&Multiply_VectorByFloat(Vector=$Emitter.CameraRight, Float=-50)
     )
    )
    Falloff Exponent=1
    Module.Kill Within Radius=false
    Module.Use Falloff=false
   )
   /SetVariables(Particles.Normalized Velocity:Vector3f=&NormalizeVector(Vector To Normalize=$Particles.Velocity)) {isSetParameters=true, module=SetVariables_#6}
   /SetVariables(Particles.Is Active:NiagaraBool=&Boolean_OrOperation(
     Bool1=$Particles.Is Active
     Bool2=&SetBoolByFloatComparison(
      A=&MakeFloatFromVector(Vector=Linked() {linkedVariable {name=Particles.Position, type=NiagaraPosition}}, Channel=::NewEnumerator2)
      B=$Output.Sample_Landscape_UPdate.Terrain Height
      Comparison Type=::NewEnumerator5
     )
    )) {isSetParameters=true, module=SetVariables_#7}
   &WindForce(
    Wind Speed=<0 250 200>
    Wind Speed Scale=&MakeCustomFloatFromBool(Bool=$Particles.Has Collided With Terrain, True Float=0, False Float=2)
   )
   &CurlNoiseForce(
    Noise Frequency=10
    Noise Strength=&MakeCustomFloatFromBool(
     Bool=$Particles.Is Active
     True Float=&FloatFromCurve(FloatCurve=DataInterface() {propertyValues {Curve {keys [<at 0.25 1>,<at 0.5 0>]}}}, Scale Curve=1000)
     False Float=0
    )
    Pan Noise=false
   ) {module=CurlNoiseForce001}
   &Drag(
    Drag=&MakeCustomFloatFromBool(
     Bool=$Particles.Is Active
     True Float=&FloatFromCurve(FloatCurve=DataInterface() {propertyValues {Curve {keys [<at 0.1 0>,<at 0.25 0.95>,<at 0.5 1>,<at 0.7 0>]}}})
     False Float=0
    )
    Use Linear Drag=true
    Use Rotational Drag=false
   )
   &GravityForce(Gravity=&MakeVector(
     X=0
     Y=0
     Z=&FloatFromCurve(
      FloatCurve=DataInterface() {propertyValues {Curve {keys [
          <at 0 0>
          {interpMode=RCIM_Cubic, tangentMode=RCTM_User, time=0.02, value=1}
          <at 0.1 0>
          {interpMode=RCIM_Cubic, tangentMode=RCTM_User, time=0.2, value=-0.9}
          <at 0.35 -0.25>
          <at 0.8 0.1>
          <at 1 0.1>
         ]}}}
      Scale Curve=400
     )
    )) {module=GravityForce001}
   /SetVariables(Particles.Sprite Orientation:Vector3f=&Lerp_Vector(
     A=$Particles.Landscape Normal
     B=&ReturnCameraVector(Camera Property=::NewEnumerator4, Coordinate Space=::Simulation, Origin Position=$Particles.Position, CameraQuery=$Emitter.CameraRead)
     Alpha=&MakeCustomFloatFromBool(Bool=$Particles.Is Active, True Float=1, False Float=0)
    )) {isSetParameters=true, module=SetVariables_#8}
   &SolveForcesAndVelocity(Clamp Velocity=true, Limit Acceleration=false)
   &self:ScratchModule_03(Terrain Height:NiagaraFloat=$Output.Sample_Landscape_UPdate.Terrain Height, Threshold=20) {module=ScratchModule_03}
   &SubUVAnimation(End Frame=35, SubUV Animation Mode:ENiagara_SubUVLookupMode=::NewEnumerator1, SubUV Loop Count=4)
   &ParticleState(DeltaTime:NiagaraFloat=&MakeCustomFloatFromBool(
     Bool=&SetBoolByFloatComparison(A=&VectorLength(Vector=$Particles.Velocity), B=1)
     True Float=$Engine.DeltaTime
     False Float=0
    ))
   &ParticleState(Kill Particles When Lifetime Has Elapsed=false) {enabled=false, module=ParticleState001}
   /SetVariables(Particles.MeshOrientation=&LerpQuaternion(
     Quaternion A:Quat4f=$Particles.Initial.MeshOrientation
     Quaternion B:Quat4f=&FindQuatBetweenTwoVectors(From Vector=<1 0 0>, To Vector=$Particles.Normalized Velocity)
     Lerp Factor:NiagaraFloat=&ClampFloat(Float=&MakeFloatFromVector(Vector=$Particles.Normalized Velocity, Channel=::NewEnumerator0))
     Enable Spherical Linear Interpolation:NiagaraBool=true
    )) {enabled=false, isSetParameters=true, module=SetVariables_#9}
   &DynamicMaterialParameters(
    Index 0 Param 1=&Lerp_Float(
     A=0
     B=1
     Alpha=&ClampFloat(Float=&MakeFloatFromVector(Vector=$Particles.Normalized Velocity, Channel=::NewEnumerator0))
    )
    Module.Param0WriteEnabled=true
    Module.Param1WriteEnabled=true
    Module.Param2WriteEnabled=true
    Module.Param3WriteEnabled=true
   ) {enabled=false}
   &Color(Color=&MakeCustomLinearColorFromBool(Bool=$Particles.Is Close to Ground, True Color=<rgba 0 0.2682 1 1>, False Color=<rgba 0 1 0.2985 1>)) {enabled=false}
   &ScaleColor(
    Scale Alpha=&FloatFromCurve(FloatCurve=DataInterface() {propertyValues {Curve {keys [<at 0.9 1>,<at 1 0>]}}})
    ScaleA=true
    ScaleRGB=true
   )
  ]
  rendererClasses [&NiagaraSpriteRendererProperties, &NiagaraMeshRendererProperties]
  renderers [
   {rendererClass=&NiagaraSpriteRendererProperties}
   {rendererClass=&NiagaraMeshRendererProperties, rendererIndex=1}
  ]
 }
 InteractiveConfetti001 {
  bEnabled=false
  simTarget=GPUComputeSim
  emitterUpdateScript [
   /SetVariables() {isSetParameters=true, module=SetVariables_#1}
   /SetVariables(
    Emitter.CameraForward:Vector3f=&ReturnCameraVector(Camera Property=::NewEnumerator1, Coordinate Space=::Simulation, CameraQuery=$Emitter.CameraRead)
    Emitter.CameraPosition:Vector3f=&ReturnCameraVector(Camera Property=::NewEnumerator0, Coordinate Space=::World, CameraQuery=$Emitter.CameraRead)
    Emitter.CameraUp:Vector3f=&ReturnCameraVector(Camera Property=::NewEnumerator2, Coordinate Space=::Simulation, CameraQuery=$Emitter.CameraRead)
   ) {isSetParameters=true, module=SetVariables_#2}
   &EmitterState(
    Enable Distance Culling=true
    Life Cycle Mode=::NewEnumerator1
    Loop Behavior=::NewEnumerator1
    Loop Duration=1
    Max Distance Response=::SleepAndClearParticles
    Module.bApplySpawnCountScaleByDistance=true
    Module.bUseMaxDistance=true
    Module.bUseMinDistance=false
    Reset Age On Awaken=true
    Scalability Mode=::NewEnumerator1
    Scale Spawn Count=true
    Spawn Count Scale By Distance=DataInterface() {propertyValues {Curve {keys [<at 0 0>,<at 0.3 0.95>,<at 0.75 1>,<at 1 0>]}}}
    UseLoopDelay=false
   )
   &SpawnBurst_Instantaneous(Spawn Count=7500, Use Spawn Probability=false) {enabled=false}
   &SpawnRate(SpawnRate=5000, Use Spawn Probability=false)
  ]
  particleSpawnScript [
   &InitializeParticle(
    Color=&RandomRangeLinearColor(Minimum=<rgba 1 0.1164 0 1>, Maximum=<rgba 1 0.2736 0 1>)
    Lifetime=&RandomRangeFloat(Minimum=3, Maximum=5)
    Mass=5
    Mesh Scale=<0.2 0.5 2>
    Module.Write Color=true
    Module.Write Lifetime=true
    Module.Write Mass=true
    Module.Write Material Random=true
    Module.Write Position=false
    Module.Write Scale=true
    Module.Write SpriteRotation=true
    Module.Write SpriteSize=true
    Position=&SimulationPosition()
    Sprite Rotation=&RandomRangeFloat(Minimum=0, Maximum=360)
    Sprite Size=&Vector2DFromFloat(Value=&RandomRangeFloat(Minimum=10, Maximum=20))
   ) {enabled=false}
   /Niagara/Modules/Spawn/Initialization/V2/InitializeParticle(
    Lifetime Max=5
    Lifetime Min=3
    Lifetime Mode=::NewEnumerator1
    Mesh Scale Max=<1 1 3>
    Mesh Scale Min=<0.6 0.6 0.6>
    Mesh Scale Mode=::NewEnumerator2
    UsePositionOffset=false
    Write Material Random=true
   ) {module=InitializeParticle001}
   /SetVariables(Particles.Terrain Offset=5) {isSetParameters=true, module=SetVariables_#3}
   &SphereLocation(MaskSpawnGroup=false, Module.Override Randomness=false, Sphere Radius=1000, Surface Only=false)
   &self:Sample_Landscape() {module=Sample_Landscape}
   /Niagara/Modules/Spawn/Orientation/V2/InitialMeshOrientation(
    Orientation Axis:Vector3f=<-1 0 0>
    Orientation Vector=<0 0 1>
    Rotation=&RandomRangeVector(Minimum=<0 0 0>, Maximum=<1 0 0>)
    Use Orientation Vector=true
    Use Rotation Vector=true
   ) {enabled=false}
   /Niagara/Modules/Spawn/Orientation/V3/InitialMeshOrientation(
    Orientation Axis=<-1 0 0>
    Orientation Vector=<0 0 1>
    Rotation=&RandomRangeVector(Minimum=<0 0 0>, Maximum=<1 0 0>)
    Use Orientation Vector=true
    Use Rotation Vector=true
   ) {module=InitialMeshOrientation001}
   /SetVariables(Particles.Position=&Add_Vector(
     A=Linked() {linkedVariable {name=Particles.Position, type=NiagaraPosition}}
     B=&MakeVector(X=0, Y=0, Z=$Particles.Terrain Offset)
    )) {isSetParameters=true, module=SetVariables_#4}
   &AddVelocity() {enabled=false}
   &CameraOffset(Camera Offset Amount=15)
   &Color(Color=&ColorFromCurve(
     ColorCurve=DataInterface() {propertyValues {
       AlphaCurve {keys [<at 0 1>,<at 1 1>]}
       BlueCurve {keys [<at 0 0.8714>,<at 0.2389 0.8714>,<at 0.2977 0.7011>,<at 0.4603 0.7011>,<at 0.5416 0.1624>,<at 1 0.1624>]}
       GreenCurve {keys [<at 0 0.5149>,<at 0.2389 0.5149>,<at 0.2977 0.063>,<at 0.4603 0.063>,<at 0.5416 0.9583>,<at 1 0.9583>]}
       RedCurve {keys [<at 0 0.0144>,<at 0.2389 0.0144>,<at 0.2977 1>,<at 0.4603 1>,<at 0.5416 0>,<at 1 0>]}
      }}
     CurveIndex=$Particles.MaterialRandom
    )) {module=Color001}
   &DynamicMaterialParameters(
    Index 0 Param 1=&RandomRangeFloat(Minimum=0, Maximum=0)
    Index 0 Param 2=&RandomRangeFloat(Minimum=0.2, Maximum=1)
    Index 0 Param 3=&RandomRangeFloat(Minimum=0.2, Maximum=1)
    Index 0 Param 4=0.3
    Module.Param0WriteEnabled=true
    Module.Param1WriteEnabled=true
    Module.Param2WriteEnabled=true
    Module.Param3WriteEnabled=true
   ) {enabled=false, module=DynamicMaterialParameters001}
  ]
  particleUpdateScript [
   &self:Sample_Landscape_UPdate(Bias:NiagaraFloat=$Particles.Terrain Offset, New Landscape Sample:NiagaraDataInterfaceLandscape=$User.Landscape) {module=Sample_Landscape_UPdate}
   /SetVariables(Particles.Has Collided With Terrain:NiagaraBool=&SetBoolByFloatComparison(
     A=&Add_Float(A=$Output.Sample_Landscape_UPdate.Terrain Height, B=20)
     B=&MakeFloatFromVector(Vector=Linked() {linkedVariable {name=Particles.Position, type=NiagaraPosition}}, Channel=::NewEnumerator2)
    )) {isSetParameters=true, module=SetVariables_#5}
   &PointAttractionForce(
    Attraction Radius=250
    AttractionStrength=-2
    AttractorPosition:Vector3f=&Add_Vector(
     A=$Emitter.CameraPosition
     B=&Add_Vector(
      A=&Multiply_VectorByFloat(Vector=$Emitter.CameraUp, Float=-250)
      B=&Multiply_VectorByFloat(Vector=$Emitter.CameraForward, Float=250)
     )
    )
    Falloff Exponent=1
    Module.Kill Within Radius=false
    Module.Use Falloff=true
   )
   /Niagara/Modules/Update/Forces/CurlNoiseForce(Noise Frequency=0, Noise Strength=0, Pan Noise=false)
   /SetVariables(Particles.Normalized Velocity:Vector3f=&NormalizeVector(Vector To Normalize=$Particles.Velocity)) {isSetParameters=true, module=SetVariables_#6}
   /SetVariables(Particles.Is Active:NiagaraBool=&Boolean_AndOperation(
     Bool1=&SetBoolByFloatComparison(
      A=&MakeFloatFromVector(Vector=Linked() {linkedVariable {name=Particles.Position, type=NiagaraPosition}}, Channel=::NewEnumerator2)
      B=$Output.Sample_Landscape_UPdate.Terrain Height
      Comparison Type=::NewEnumerator5
     )
     Bool2=&SetBoolByFloatComparison(A=&MakeFloatFromVector(Vector=$Particles.Normalized Velocity, Channel=::NewEnumerator2), B=0.01)
    )) {isSetParameters=true, module=SetVariables_#7}
   &WindForce(
    Wind Speed=<0 100 25>
    Wind Speed Scale=&MakeCustomFloatFromBool(Bool=$Particles.Is Active, True Float=5, False Float=0)
   )
   /Niagara/Modules/Update/Forces/V2/CurlNoiseForce(
    Noise Frequency=5
    Noise Strength=&MakeCustomFloatFromBool(Bool=$Particles.Is Active, True Float=250, False Float=0)
    Pan Noise=false
    Pan Noise Field=<0 1 1>
   ) {module=CurlNoiseForce001}
   &GravityForce(Gravity=&MakeCustomVectorFromBool(Bool=$Particles.Is Active, True Vector=<0 0 -25>, False Vector=<0 0 0>))
   &Drag(
    Drag=&MakeCustomFloatFromBool(Bool=$Particles.Is Active, True Float=0, False Float=0)
    Use Linear Drag=true
    Use Rotational Drag=false
   )
   &SolveForcesAndVelocity(Clamp Velocity=true, Limit Acceleration=false)
   /SetVariables(Particles.Sprite Orientation:Vector3f=&Lerp_Vector(
     A=$Particles.Landscape Normal
     B=&ReturnCameraVector(Camera Property=::NewEnumerator4, Coordinate Space=::Simulation, Origin Position=$Particles.Position, CameraQuery=$Emitter.CameraRead)
     Alpha=&MakeCustomFloatFromBool(Bool=$Particles.Is Active, True Float=1, False Float=0)
    )) {enabled=false, isSetParameters=true, module=SetVariables_#8}
   &self:ScratchModule_03(Terrain Height:NiagaraFloat=$Output.Sample_Landscape_UPdate.Terrain Height, Threshold=10) {module=ScratchModule_03}
   &ParticleState(DeltaTime:NiagaraFloat=&MakeCustomFloatFromBool(
     Bool=&SetBoolByFloatComparison(A=&VectorLength(Vector=$Particles.Velocity), B=1)
     True Float=$Engine.DeltaTime
     False Float=0
    ))
   &ParticleState(Kill Particles When Lifetime Has Elapsed=false) {enabled=false, module=ParticleState001}
   /SetVariables(Particles.MeshOrientation=&LerpQuaternion(
     Quaternion A:Quat4f=$Particles.Initial.MeshOrientation
     Quaternion B:Quat4f=&FindQuatBetweenTwoVectors(From Vector=<1 0 0>, To Vector=$Particles.Normalized Velocity)
     Lerp Factor:NiagaraFloat=&ClampFloat(Float=&MakeFloatFromVector(Vector=$Particles.Normalized Velocity, Channel=::NewEnumerator0))
     Enable Spherical Linear Interpolation:NiagaraBool=true
    )) {isSetParameters=true}
   /SetVariables(Particles.MeshOrientation=&LerpQuaternion(
     Quaternion A:Quat4f=$Particles.Initial.MeshOrientation
     Quaternion B:Quat4f=&FindQuatBetweenTwoVectors(From Vector=$Particles.Landscape Normal, To Vector=$Particles.Normalized Velocity)
     Lerp Factor:NiagaraFloat=&ClampFloat(Float=&MakeFloatFromVector(Vector=$Particles.Normalized Velocity, Channel=::NewEnumerator0))
     Enable Spherical Linear Interpolation:NiagaraBool=true
    )) {enabled=false, isSetParameters=true, module=SetVariables_#9}
   /SetVariables(Particles.MeshOrientation=&OrientMeshToVector(Look At Direction=$Particles.Landscape Normal, Reference Axis Vector=<0 1 0>)) {enabled=false, isSetParameters=true}
   &Color(Color=&MakeCustomLinearColorFromBool(Bool=$Particles.Is Close to Ground, True Color=<rgba 0 0.2682 1 1>, False Color=<rgba 0 1 0.2985 1>)) {enabled=false}
   &ScaleColor(
    Scale Alpha=&FloatFromCurve(FloatCurve=DataInterface() {propertyValues {Curve {keys [<at 0.9 1>,<at 1 0>]}}})
    ScaleA=true
    ScaleRGB=true
   )
   &DynamicMaterialParameters(
    Index 0 Param 1=&AngleConversion(
     Angle:NiagaraFloat=&Lerp_Float(
      A=0
      B=&Multiply_Float(
       A=&RandomRangeFloat(Minimum=15, Maximum=35)
       B=&MakeCustomFloatFromBool(Bool=$Particles.Is Active, True Float=1, False Float=0)
      )
      Alpha=&Waveform(
       "Waveform [1]"=::NewEnumerator0
       "[1] Frequency"=&RandomRangeFloat(Minimum=1, Maximum=2)
       "[1] Amplitude Min/Max"=<-1 1>
       "[2] Phase"=$Emitter.Age
       "[3] Phase"=$Emitter.Age
      )
     )
     Angle Input:ENiagara_AngleInput=::NewEnumerator0
     Angle Output:ENiagara_AngleInput=::NewEnumerator1
    )
    Index 0 Param 2=&RandomRangeFloat(Minimum=0, Maximum=0.2)
    Index 0 Param 3=&RandomRangeFloat(Minimum=0, Maximum=1)
    Index 0 Param 4=&RandomRangeFloat(Minimum=-0.25, Maximum=1)
    Module.Param0WriteEnabled=true
    Module.Param1WriteEnabled=true
    Module.Param2WriteEnabled=true
    Module.Param3WriteEnabled=true
   )
   &SolveRotationalForcesAndVelocity()
  ]
  rendererClasses [&NiagaraMeshRendererProperties]
  renderers [{rendererClass=&NiagaraMeshRendererProperties, rendererIndex=0}]
 }
}
summary {
 systemName=NS_Welcome_Confetti_Interact
 emitters [
  {
   emitterName=InteractiveConfetti
   simTarget=GPUComputeSim
   rendererClasses [&NiagaraSpriteRendererProperties, &NiagaraMeshRendererProperties]
  }
  {bEnabled=false, emitterName=InteractiveConfetti001, simTarget=GPUComputeSim, rendererClasses [&NiagaraMeshRendererProperties]}
 ]
 userVariables [
  User.Alpha_AttractionStrength {type=NiagaraFloat, defaultValue=0, description=""}
  User.Landscape {
   type=NiagaraDataInterfaceLandscape
   description=""
   defaultValue {
    struct=NiagaraExt_VariableValue_DataInterface
    value {
     dataInterface=&self:NiagaraDataInterfaceLandscape_0
     dataInterfaceClass=&NiagaraDataInterfaceLandscape
    }
   }
  }
  User.Player Position {type=Vector3f, description="", defaultValue {struct=Vector3f, value=<0 0 0>}}
 ]
}
