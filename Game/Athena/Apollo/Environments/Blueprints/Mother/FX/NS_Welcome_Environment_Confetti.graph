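# Niagara GPU sprite system that scatters confetti across the environment, apparently for a welcome celebration in the Apollo Mother environment.
# Particles spawn at a constant rate in a growing, drifting sphere, pass through a terrain-near-player placement module, and randomly take one of three main colors.
# Curve-driven bursts of speed feed launch velocity, curl noise and wind, alongside drag and collision.
# Sprites spin, animate via SubUV, fade, and scale by owner LOD distance.
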NiagaraSystem /Game/Athena/Apollo/Environments/Blueprints/Mother/FX/NS_Welcome_Environment_Confetti
dependencies {
 usedDataInterfaces [
  &NiagaraDataInterfaceCamera
  &NiagaraDataInterfaceCurve
  /Script/Niagara.NiagaraDataInterfaceTexture
  &NiagaraDataInterfaceColorCurve
 ]
 usedDynamicInputs []
 usedModules [
  &SystemState
  &self:SystemSpawnScript.SystemScriptSource.SystemScriptGraph.NiagaraNodeAssignment_0.SetVariables_#1
 ]
}
emitters {Confetti {
  simTarget=GPUComputeSim
  emitterUpdateScript [
   /SetVariables(
    Emitter.CameraPostion:Vector3f=&ReturnCameraVector(Camera Property=::NewEnumerator0, Coordinate Space=::Simulation)
    Emitter.Emitter_HeightOffset:NiagaraFloat=$System.System_HeightOffset
    Emitter.Overall_SpeedMult=0.25
    Emitter.SpawnRate_Curve:NiagaraFloat=&FloatFromCurve(
     FloatCurve=DataInterface() {propertyValues {Curve {keys [
         {arriveTangent=9.7155, interpMode=RCIM_Cubic, leaveTangent=9.7155, tangentMode=RCTM_User, time=0.0009, value=0.1}
         {arriveTangent=-0.0619, interpMode=RCIM_Cubic, leaveTangent=-0.0619, tangentMode=RCTM_User, time=0.0989, value=0}
         <at 0.297 0>
         {interpMode=RCIM_Cubic, tangentMode=RCTM_User, time=0.3179, value=0.1027}
         {interpMode=RCIM_Cubic, tangentMode=RCTM_User, time=0.3525, value=0}
         <at 0.4303 0.0992>
         <at 0.5102 0.0235>
         {interpMode=RCIM_Cubic, tangentMode=RCTM_User, time=0.8355, value=0.0227}
         <at 0.9972 0.1061>
        ]}}}
     CurveIndex=&Multiply_Float(A=$Emitter.NormalizedLoopAge, B=&RandomRangeFloat(Minimum=0.25, Maximum=0.5))
     Scale Curve=&Lerp_Float(A=5000, B=25000, Alpha=&UniformRangedFloat(Minimum=0.25, Maximum=1))
    )
    Emitter.SpawnSpeed_Curve:NiagaraFloat=&FloatFromCurve(
     FloatCurve=DataInterface() {propertyValues {Curve {keys [
         {arriveTangent=30.2685, interpMode=RCIM_Cubic, leaveTangent=30.2685, tangentMode=RCTM_User, time=0.0009, value=0.0215}
         {arriveTangent=-0.0619, interpMode=RCIM_Cubic, leaveTangent=-0.0619, tangentMode=RCTM_User, time=0.0989, value=0.3515}
         <at 0.297 0.0032>
         <at 0.3192 0.3332>
         <at 0.3525 0.0398>
         <at 0.6171 0.3148>
         <at 0.9939 0.3698>
        ]}}}
     CurveIndex=&Multiply_Float(A=$Emitter.NormalizedLoopAge, B=&RandomRangeFloat(Minimum=1, Maximum=2))
     Scale Curve=&Multiply_Float(
      A=&Lerp_Float(A=5, B=10, Alpha=&UniformRangedFloat(Minimum=0.25, Maximum=1))
      B=1
     )
    )
   ) {isSetParameters=true}
   &EmitterState()
   &SpawnRate(SpawnRate=500, Use Spawn Probability=false)
  ]
  particleSpawnScript [
   /SetVariables(Particles.SpawnGroup=&RandomRangeInt(Minimum=0, Maximum=3)) {enabled=false, isSetParameters=true}
   /SetVariables(
    Particles.HeightOffset:NiagaraFloat=$System.System_HeightOffset
    Particles.UporDown:NiagaraFloat=&UniformAOrBFloat(
     A=&RandomRangeFloat(Minimum=-200, Maximum=200)
     B=&RandomRangeFloat(Minimum=0, Maximum=0)
     Distribution Weight=0.8
    )
   ) {isSetParameters=true}
   &InitializeParticle(
    Color=&MaskLinearColorBySpawnGroup(
     Masked Linear Color=&HueShiftLinearColor(
      Hue=0
      Saturation=&RandomRangeFloat(Minimum=5, Maximum=10)
      Value=&RandomRangeFloat(Minimum=1, Maximum=10)
      Linear Color=&RandomRangeLinearColor(Minimum=<rgba 0 0 0 1>, Maximum=<rgba 1 0 0.841 1>, Color Channel Mode=::NewEnumerator2)
     )
     Passthrough Linear Color=<rgba 1 1 1 1>
    )
    Lifetime Max=4
    Lifetime Min=3
    Lifetime Mode=::NewEnumerator1
    Sprite Rotation Mode=::NewEnumerator2
    Sprite Size Mode=::NewEnumerator1
    Uniform Sprite Size Max=12
    Uniform Sprite Size Min=8
    UsePositionOffset=false
    Write Material Random=true
   ) {module=InitializeParticle001}
   &AddVelocity(Velocity=&Multiply_VectorByFloat(
     Vector=&Multiply_VectorByFloat(
      Vector=&Multiply_VectorByFloat(Vector=&UniformRangedVector(Minimum=<60 128 100>, Maximum=<120 500 150>), Float=$Emitter.SpawnSpeed_Curve)
      Float=$Emitter.Overall_SpeedMult
     )
     Float=$Particles.HeightOffset
    ))
   /NiagaraFN/Modules/Spawn/FN_SpawnParticlesOnTerrainNearPlayer_ParticleType(
    ColorTexture=DataInterface() {propertyValues {
      Texture=/Game/Effects/Fort_Effects/Textures/Environments/Apollo/T_WholeMap_Color_D_01
      TextureUserParameter {parameter {name=None, typeDefHandle {registeredTypeIndex=2}, varData []}}
     }}
    Height Map Max=25600
    Height Map Min=-25600
    Height Texture -x -y:Vector3f=<-130048 -130048 0>
    HeightOffset:NiagaraFloat=&RandomRangeFloat(Minimum=500, Maximum=1500)
    HeightTexture=DataInterface() {propertyValues {
      Texture=/Game/Effects/Fort_Effects/Textures/Environments/Apollo/T_Heightmap_1530
      TextureUserParameter {parameter {name=None, typeDefHandle {registeredTypeIndex=2}, varData []}}
     }}
    HeightTexture +x +y:Vector3f=<130048 130048 0>
    Player Position:NiagaraPosition=$Particles.Position
    Player Relative Particle Location:Vector2f=&UniformRangedVector2D(Minimum=<0 0>, Maximum=<0 0>)
   ) {enabled=false}
   &SphereLocation(
    MaskSpawnGroup=false
    Non Uniform Scale=&RandomRangeVector(Minimum=<-1 -1 0>, Maximum=<1 1 0.1>)
    Offset=&Multiply_VectorByFloat(
     Vector=<1 0 0>
     Float=&FloatFromCurve(FloatCurve=DataInterface() {propertyValues {Curve {keys [<at 0 0>,<at 2 1>]}}}, CurveIndex=$System.LoopedAge, Scale Curve=500)
    )
    Override Seed=false
    Sphere Radius=&Multiply_Float(
     A=5000
     B=&FloatFromCurve(
      FloatCurve=DataInterface() {propertyValues {Curve {keys [<at 0 0.1>,<at 0.157 0.9198>]}}}
      CurveIndex=$System.NormalizedLoopAge
      Scale Curve=5
     )
    )
    Surface Only=false
    UseNonUniformScale=true
    UseOffset=true
    UseSphereOrientation=false
   )
   /NiagaraFN/Modules/Spawn/FN_SpawnParticlesOnTerrainNearPlayer(
    HeightOffset:NiagaraFloat=&RandomRangeFloat(Minimum=0, Maximum=800)
    Player Position:NiagaraPosition=$Particles.Position
   )
   &CurlNoiseForce(
    Noise Strength=&Multiply_Float(
     A=150
     B=&Multiply_Float(A=&Multiply_Float(A=$Particles.HeightOffset, B=$Emitter.Overall_SpeedMult), B=1)
    )
    Pan Noise=false
   ) {module=CurlNoiseForce001}
   &ApplyInitialForces()
   /Niagara/Modules/Update/Camera/MaintainInCameraParticleScale(Render Target Width=2)
   &CameraOffset()
   &SubUVAnimation(End Frame=36, SubUV Animation Mode:ENiagara_SubUVLookupMode=::NewEnumerator2) {module=SubUVAnimation001}
   &Color(Color=&ColorFromCurve(
     ColorCurve=DataInterface() {propertyValues {
       AlphaCurve {keys [<at 0 1>,<at 1 1>]}
       BlueCurve {keys [<at 0 0.8714>,<at 0.2389 0.8714>,<at 0.2977 0.7011>,<at 0.4603 0.7011>,<at 0.5416 0.1624>,<at 1 0.1624>]}
       GreenCurve {keys [<at 0 0.5149>,<at 0.2389 0.5149>,<at 0.2977 0.063>,<at 0.4603 0.063>,<at 0.5416 0.9583>,<at 1 0.9583>]}
       RedCurve {keys [<at 0 0.0144>,<at 0.2389 0.0144>,<at 0.2977 1>,<at 0.4603 1>,<at 0.5416 0>,<at 1 0>]}
      }}
     CurveIndex=$Particles.MaterialRandom
    ))
  ]
  particleUpdateScript [
   &ParticleState()
   &GravityForce(Gravity=&Multiply_VectorByFloat(
     Vector=<0 0 1>
     Float=/Game/Effects/Niagara/DynamicInputs/Math/FN_Fit(
      Clamp:NiagaraBool=false
      Value:NiagaraFloat=$Particles.UporDown
      SourceMin:NiagaraFloat=0
      SourceMax:NiagaraFloat=1
      DestMin:NiagaraFloat=0
      DestMax:NiagaraFloat=0
     )
    ))
   &CurlNoiseForce(
    Curl Noise Cone Mask Angle=120
    Mask Curl Noise=true
    Noise Frequency=5
    Noise Strength=&Multiply_Float(A=$Emitter.SpawnSpeed_Curve, B=100)
    Pan Noise=true
    Pan Noise Field=<0 0 1>
   )
   &WindForce(Air Resistance=1, Wind Speed=<0 150 0>, Wind Speed Scale=$Emitter.SpawnSpeed_Curve)
   &Drag(
    Drag=&FloatFromCurve(
     FloatCurve=DataInterface() {propertyValues {Curve {keys [<at -0.0085 0.0857>,<at 0.0429 0.6046>,<at 0.1003 0.0765>,<at 1 0>]}}}
     Scale Curve=2
    )
    Use Linear Drag=true
    Use Rotational Drag=false
   )
   &Collision(
    Age Colliding Particles=false
    Collision Enabled=&SetBoolByFloatComparison(A=1, B=$Particles.Age, Comparison Type=::NewEnumerator2)
    Enable Rest State=false
    Friction=0
    GPU Collision Type=::NewEnumerator2
    Module.EnableMaxCollisionCount=false
    Module.Randomize Collision Normal Vector=true
    Particle Radius Scale=2
    Randomize Collision Normal=0.25
    Restitution=0.005
   )
   &SolveForcesAndVelocity(Clamp Velocity=true, Limit Acceleration=false, Speed Limit=800)
   &SpriteRotationRate(Rotation Rate=&FloatFromCurve(
     FloatCurve=DataInterface() {propertyValues {Curve {keys [<at 0 1>,<at 1 0>]}}}
     Scale Curve=&NormalizedAngleToDegrees(NormalizedAngle=&Lerp_Float(A=-3, B=3, Alpha=$Particles.Initial.MaterialRandom))
    ))
   &SubUVAnimation(End Frame=16, SubUV Animation Mode:ENiagara_SubUVLookupMode=::NewEnumerator1, SubUV Loop Count=2)
   &ScaleSpriteSize(Scale Factor=&Vector2DFromFloat(Value=&FloatFromCurve(
      FloatCurve=DataInterface() {propertyValues {Curve {keys [<at 0.6918 1>,<at 1 0>]}}}
      Scale Curve=&FloatFromCurve(
       FloatCurve=DataInterface() {propertyValues {Curve {keys [<at 2500 0.875>,<at 4776.3999 1.625>,<at 12000 2.5>,<at 15000 0>]}}}
       CurveIndex=$Engine.Owner.LODDistance
      )
     )))
   &DynamicMaterialParameters(Module.Param0WriteEnabled=true, Module.Param1WriteEnabled=true, Module.Param2WriteEnabled=true, Module.Param3WriteEnabled=true)
   &ScaleColor(
    Scale Alpha=&FloatFromCurve(FloatCurve=DataInterface() {propertyValues {Curve {keys [
         {interpMode=RCIM_Cubic, tangentMode=RCTM_User, time=0, value=0}
         {interpMode=RCIM_Cubic, tangentMode=RCTM_User, time=0.2, value=1}
         {interpMode=RCIM_Cubic, tangentMode=RCTM_User, time=0.9, value=1}
         {tangentMode=RCTM_User, time=1, value=0}
        ]}}})
    ScaleA=true
    ScaleRGB=true
   )
  ]
  rendererClasses [&NiagaraSpriteRendererProperties]
  renderers [{rendererClass=&NiagaraSpriteRendererProperties}]
 }}
summary {
 systemName=NS_Welcome_Environment_Confetti
 emitters [{emitterName=Confetti, simTarget=GPUComputeSim, rendererClasses [&NiagaraSpriteRendererProperties]}]
 userVariables [User.Landscape {
   type=NiagaraDataInterfaceLandscape
   description=""
   defaultValue {
    struct=NiagaraExt_VariableValue_DataInterface
    value {
     dataInterface=&self:NiagaraDataInterfaceLandscape_0
     dataInterfaceClass=&NiagaraDataInterfaceLandscape
    }
   }
  }]
}
