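# Niagara system for a swaying "Welcome" light beam effect in the Mother environment, built from one GPU sprite emitter.
# Sprites spawn inside a long cone whose axis oscillates in X over emitter age, with user parameters controlling cone angle and sway amount.
# Each takes a random color from a gradient, hue-shifted by a user parameter.
# Particles get early curl-noise drift, swell from zero size, and fade out over their lifetime.
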NiagaraSystem /Game/Athena/Apollo/Environments/Blueprints/Mother/FX/NS_Welcome_LightBeam
dependencies {
 usedDataInterfaces [&NiagaraDataInterfaceColorCurve, &NiagaraDataInterfaceCurve]
 usedDynamicInputs []
 usedModules [&SystemState]
}
emitters {FN_Sprite {
  simTarget=GPUComputeSim
  emitterUpdateScript [
   &EmitterState(Enable Distance Culling=true, Module.bUseMaxDistance=true, Module.bUseMinDistance=false, Scalability Mode=::NewEnumerator1)
   &SpawnRate(SpawnRate=10, Use Spawn Probability=false)
  ]
  particleSpawnScript [
   &InitializeParticle(
    Color=&ColorFromCurve(
     ColorCurve=DataInterface() {propertyValues {
       AlphaCurve {keys [<at 0 1 RCIM_Cubic>,<at 0.9977 1 RCIM_Cubic>]}
       BlueCurve {keys [<at 0.1175 0.0288 RCIM_Cubic>,<at 0.2682 0.7775>,<at 0.4507 0.4148>,<at 0.5951 1>,<at 0.6953 0.1725>,<at 0.9071 0.521 RCIM_Cubic>]}
       GreenCurve {keys [<at 0.1175 0.1034 RCIM_Cubic>,<at 0.2682 0.1382>,<at 0.4507 0.315>,<at 0.5951 0.0671>,<at 0.6953 0.3216>,<at 0.9071 0.4852 RCIM_Cubic>]}
       RedCurve {keys [<at 0.1175 0.3372 RCIM_Cubic>,<at 0.2682 1>,<at 0.4507 1>,<at 0.5951 0.9911>,<at 0.6953 1>,<at 0.9071 0.807 RCIM_Cubic>]}
      }}
     CurveIndex=&RandomRangeFloat(Minimum=0, Maximum=1)
    )
    Lifetime Mode=::NewEnumerator1
    Sprite Size Mode=::NewEnumerator1
    Uniform Sprite Size Max=100
    Uniform Sprite Size Min=50
    UsePositionOffset=false
    Write Material Random=true
   ) {module=InitializeParticle001}
   /SetVariables(Particles.Color=&HueShiftLinearColor(Hue=$User.HueFX, Saturation=1, Value=1, Linear Color=$Particles.Color)) {isSetParameters=true}
   /Niagara/Modules/Spawn/Location/ConeLocation(
    Cone Angle:NiagaraFloat=&Multiply_Float(A=$User.ConeAngleFX, B=1)
    Cone Axis:Vector3f=&MakeVector(
     X=&Waveform(
      "Waveform [1]"=::NewEnumerator0
      "[1] Frequency"=1
      "[1] Amplitude Min/Max"=&Multiply_Vector2(A=<-0.1 0.1>, B=&Vector2DFromFloat(Value=$User.ConeRotationFX))
      "[2] Phase"=$Emitter.Age
      "[3] Phase"=$Emitter.Age
     )
     Y=0
     Z=1
    )
    Cone Length=1500
    MaskSpawnGroup=false
    Module.Override Randomness=false
   )
   &AddVelocityFromPoint(
    Module.Use Falloff Distance=false
    Origin Offset=&UniformRangedVector(Minimum=<0 0 -100>, Maximum=<0 0 -100>)
    Origin Offset Coordinate Space=::Simulation
    Velocity Strength=&UniformRangedFloat(Minimum=10, Maximum=15)
   ) {enabled=false}
   /Niagara/Modules/Update/Forces/CurlNoiseForce(Noise Frequency=0.25, Noise Strength=5, Pan Noise=true, Pan Noise Field=<0 0 -5>) {enabled=false, module=CurlNoiseForce001}
   &ApplyInitialForces() {enabled=false}
  ]
  particleUpdateScript [
   &ParticleState()
   &CurlNoiseForce(
    Noise Frequency=1
    Noise Quality / Cost=::NewEnumerator0
    Noise Strength=&FloatFromCurve(FloatCurve=DataInterface() {propertyValues {Curve {keys [<at 0 1>,<at 0.25 0>]}}})
    Pan Noise=true
    Pan Noise Field=<0 10 0>
   )
   &ScaleSpriteSize(Scale Factor=&Vector2DFromFloat(Value=&FloatFromCurve(FloatCurve=DataInterface() {propertyValues {Curve {keys [{arriveTangent=4.5137, interpMode=RCIM_Cubic, leaveTangent=4.5137, tangentMode=RCTM_User, time=0, value=0}, <at 1 0.5 RCIM_Cubic>]}}})))
   &SolveForcesAndVelocity(Clamp Velocity=false, Limit Acceleration=false)
   &ScaleColor(
    Scale Alpha=&FloatFromCurve(FloatCurve=DataInterface() {propertyValues {Curve {keys [<at 0 1>,<at 1 0>]}}}, Scale Curve=0.5)
    ScaleA=true
    ScaleRGB=true
   )
  ]
  rendererClasses [&NiagaraSpriteRendererProperties]
  renderers [{rendererClass=&NiagaraSpriteRendererProperties}]
 }}
summary {
 systemName=NS_Welcome_LightBeam
 emitters [{emitterName=FN_Sprite, simTarget=GPUComputeSim, rendererClasses [&NiagaraSpriteRendererProperties]}]
 userVariables [
  User.ConeAngleFX {type=NiagaraFloat, defaultValue=2, description=""}
  User.ConeRotationFX {type=NiagaraFloat, defaultValue=0.1, description=""}
  User.HueFX {type=NiagaraFloat, defaultValue=0, description=""}
 ]
}
