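# Translucent two-sided material for the Apollo Mothership beam, built from a tinted emissive color with a scalar intensity.
# A remapped, inverted vertical UV gradient can fade opacity via the UseGradient static switch and is also multiplied by a Glow parameter.
# With UseWPO enabled, red vertex color masks a Z-axis vertex offset whose length interpolates between GlowLengthMin and GlowLength by Glow, stretching the beam.
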Material /Game/Athena/Apollo/Environments/Blueprints/Mother/M_Apollo_Mothership_Beam
graph {
 properties {blendMode=BLEND_Translucent, pixelDepthOffsetMode=PDOM_Legacy, twoSided=true, usedShadingModels=Default Lit}
 nodes [
  Multiply(
   A=VectorParameter() {parameterName=Color, defaultValue=<rgba 0.1232 0.8601 1 0>}.RGB
   B=ScalarParameter() {defaultValue=1, parameterName=Emissive}
  )
  n1 = ScalarParameter() {parameterName=GradMin}
  n2 = ScalarParameter() {defaultValue=1, parameterName=GradMax}
  n3 = MaterialFunctionCall(UV Channel=TextureCoordinate()) {materialFunction=&LinearGradient}
  n4 = OneMinus(None=Saturate(None=LinearInterpolate(A=@n1, B=@n2, Alpha=@n3.VGradient)))
  Multiply(A=@n4, B=ScalarParameter() {parameterName=Glow})
  n5 = ScalarParameter() {defaultValue=0.2, parameterName=Opacity}
  StaticSwitchParameter(True=LinearInterpolate(B=@n5, Alpha=@n4), False=@n5) {parameterName=UseGradient}
  n6 = VertexColor()
  n7 = LinearInterpolate(
   A=ScalarParameter() {defaultValue=9000, parameterName=GlowLengthMin}
   B=ScalarParameter() {defaultValue=-15000, parameterName=GlowLength}
   Alpha=ScalarParameter() {parameterName=Glow}
  )
  StaticSwitchParameter(True=AppendVector(A=Constant2Vector(), B=Multiply(A=@n6.R, B=@n7)), False=Constant3Vector()) {parameterName=UseWPO}
 ]
}
