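# Translucent, two-sided lit material for a cloudy atmosphere on the Skyfire Mothership, with cloud fogging applied.
# Two independently panned and scaled samples of an infected-map texture drive blended normals and a combined cloud value for a three-color blend.
# That value, minus an atmosphere gradient, forms opacity, shaped by a sphere-masked inverted Fresnel on vertex normals.
# TwoSidedSign selects separate opacity multipliers for inner and outer faces.
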Material /Game/Athena/Apollo/Environments/Blueprints/Mother/Materials/M_Mother_Atmosphere
graph {
 properties {
  bApplyCloudFogging=true
  bUseTranslucencyVertexFog=false
  blendMode=BLEND_Translucent
  pixelDepthOffsetMode=PDOM_Legacy
  translucencyLightingMode=TLM_VolumetricPerVertexDirectional
  twoSided=true
  usedShadingModels=Default Lit
 }
 nodes [
  n1 = ScalarParameter() {parameterName=CloudsU_Pan_A}
  n2 = ScalarParameter() {defaultValue=0.05, parameterName=CloudsV_Pan_A}
  n3 = Multiply(
   A=Panner(Coordinate=TextureCoordinate(), Speed=AppendVector(A=@n1, B=@n2))
   B=AppendVector(A=ScalarParameter() {defaultValue=5, parameterName=CloudsU_Scale_A}, B=ScalarParameter() {defaultValue=0.1, parameterName=CloudsV_Scale_A})
  )
  n4 = ScalarParameter() {parameterName=CloudsU_Pan_B}
  n5 = ScalarParameter() {defaultValue=0.011, parameterName=CloudsV_Pan_B}
  n6 = Multiply(
   A=Panner(Coordinate=TextureCoordinate(), Speed=AppendVector(A=@n4, B=@n5))
   B=AppendVector(A=ScalarParameter() {defaultValue=6, parameterName=CloudsU_Scale_B}, B=ScalarParameter() {defaultValue=0.06, parameterName=CloudsV_Scale_B})
  )
  MaterialFunctionCall(
   BaseNormal=TextureSample(UVs=NamedRerouteDeclaration(None=@n3) {name=UVs_A, nodeColor=<rgba 0 0.7422 1 1>}) {samplerType=SAMPLERTYPE_Normal, texture=/Game/Global/Textures/Default/Tileable_Patterns/2ndInfectedMapNormal}.RGB
   AdditionalNormal=TextureSample(UVs=NamedRerouteDeclaration(None=@n6) {name=UVs_B, nodeColor=<rgba 1 0 0.5625 1>}) {samplerType=SAMPLERTYPE_Normal, texture=&2ndInfectedMapNormal}.RGB
  ) {materialFunction=&BlendAngleCorrectedNormals}
  n7 = NamedRerouteUsage()
  n8 = NamedRerouteUsage()
  n9 = Divide(A=Add(
    A=TextureSample(UVs=@n7) {texture=&2ndInfectedMapLinear}.R
    B=TextureSample(UVs=@n8) {texture=&2ndInfectedMapLinear}.R
   ))
  MaterialFunctionCall(
   A=VectorParameter() {parameterName=ColorA, defaultValue=<rgba 0.4686 0.5424 0.5885 0>}.RGB
   B=VectorParameter() {parameterName=ColorB, defaultValue=<rgba 0.1238 0.1887 0.3142 0>}.RGB
   C=VectorParameter() {parameterName=ColorC, defaultValue=<rgba 0.0621 0.1697 0.2691 0>}.RGB
   Alpha=@n9
  ) {materialFunction=&3ColorBlend}
  n10 = TextureSample() {
   samplerType=SAMPLERTYPE_LinearColor
   texture=&/T_Skyfire_Mothership_AtmosphereGradient
  }
  n11 = Subtract(A=Subtract(A=@n9) {constB=0.2}, B=@n10.R)
  n12 = ScalarParameter() {defaultValue=3, parameterName=CloudOpacityScale}
  n13 = MaterialFunctionCall(Exponent=ScalarParameter() {defaultValue=1.5, parameterName=FresnelExponent}, Vertex / Pixel Normal=VertexNormalWS()) {materialFunction=&MF_Fresnel_FX}
  n14 = Constant() {r=0.2}
  n15 = Constant() {r=0.2}
  n16 = Constant()
  n17 = TwoSidedSign()
  Multiply(
   A=Multiply(A=Multiply(A=@n11, B=@n12), B=SphereMask(A=@n13.Fresnel Inverted Double Sided, B=@n14, Radius=@n15, Hardness=@n16))
   B=LinearInterpolate(
    A=ScalarParameter() {defaultValue=0.1, parameterName=FinalOpacityInnerFaces}
    B=ScalarParameter() {defaultValue=0.3, parameterName=FinalOpacityOuterFaces}
    Alpha=Saturate(None=@n17)
   )
  )
 ]
}
