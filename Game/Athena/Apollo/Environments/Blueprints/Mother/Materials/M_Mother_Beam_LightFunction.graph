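# Light function material for the Mother beam in the Apollo environment, producing animated turbulent light variation.
# Two turbulence textures with world-XY noise-distorted UVs either pan or rotate by static switch, then are multiplied, square-rooted and remapped.
# Mobile and low feature levels substitute constant 1 for the noise.
# A parameterized world-space sphere mask is added and the result is scaled; an older MakeFloat3 noise branch appears unconnected.
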Material /Game/Athena/Apollo/Environments/Blueprints/Mother/Materials/M_Mother_Beam_LightFunction
graph {
 properties {materialDomain=MD_LightFunction, pixelDepthOffsetMode=PDOM_Legacy, shadingModel=MSM_Unlit, usedShadingModels=Unlit}
 nodes [
  LinearInterpolate()
  n1 = Time()
  n2 = Add(A=@n1) {constB=0.1}
  n3 = TextureSample(UVs=Rotator(Coordinate=TextureCoordinate() {uTiling=4, vTiling=4}, Time=@n2) {centerX=2, centerY=2}) {texture=/Game/Global/Textures/Default/Caustics/T_Turbulence_Seamless}
  n4 = TextureSample(UVs=Rotator(Coordinate=TextureCoordinate() {uTiling=8, vTiling=8}, Time=@n2) {centerX=4, centerY=4, speed=-0.1}) {texture=&T_Turbulence_Seamless}
  n5 = Add(A=@n1) {constB=0.2}
  n6 = TextureSample(UVs=Rotator(Coordinate=TextureCoordinate() {uTiling=4, vTiling=4}, Time=@n5) {centerX=2, centerY=2}) {texture=&T_Turbulence_Seamless}
  n7 = TextureSample(UVs=Rotator(Coordinate=TextureCoordinate() {uTiling=8, vTiling=8}, Time=@n5) {centerX=4, centerY=4, speed=-0.1}) {texture=&T_Turbulence_Seamless}
  MaterialFunctionCall(Y=LinearInterpolate(Alpha=Add(A=@n3.R, B=@n4.R)) {constA=0.25}, Z=LinearInterpolate(Alpha=Add(A=@n6.R, B=@n7.R)) {constA=0.25}) {materialFunction=&MakeFloat3}
  n8 = WorldPosition()
  n9 = TextureSampleParameter2D(UVs=Divide(A=ComponentMask(None=@n8.XYZ) {g=true, r=true}, B=ScalarParameter() {defaultValue=4096, parameterName=Divide}) {constB=4096}) {parameterName=Param, texture=&T_Turbulence_Seamless}
  n10 = Multiply(A=Subtract(A=AppendVector(A=@n9.R, B=@n9.R)) {constB=0.5}) {constB=0.2}
  n11 = Add(A=@n10, B=Multiply(A=TextureCoordinate(), B=ScalarParameter() {defaultValue=4, parameterName=Scale UV A}))
  n12 = Time()
  n13 = Multiply(A=ScalarParameter() {defaultValue=0.25, parameterName=Scale Time A}, B=@n12)
  n14 = StaticSwitchParameter(True=Panner(Coordinate=@n11, Time=@n13) {speedX=1, speedY=1}, False=Rotator(Coordinate=@n11, Time=@n13) {centerX=2, centerY=2, speed=1}) {desc="T=Panner; F=Rotator", parameterName=Panner/Rotator}
  n15 = Add(A=Multiply(A=TextureCoordinate(), B=ScalarParameter() {defaultValue=8, parameterName=Scale UV B}), B=@n10)
  n16 = Multiply(A=@n12, B=ScalarParameter() {defaultValue=-0.1, parameterName=Scale Time B})
  n17 = StaticSwitchParameter(True=Panner(Coordinate=@n15, Time=@n16) {speedX=1, speedY=1}, False=Rotator(Coordinate=@n15, Time=@n16) {centerX=4, centerY=4, speed=1}) {desc="T=Panner; F=Rotator", parameterName=Panner/Rotator}
  n18 = Multiply(A=Multiply(
    A=TextureSampleParameter2D(UVs=@n14) {parameterName=Texture_01, texture=&T_Turbulence_Seamless}.R
    B=TextureSampleParameter2D(UVs=@n17) {parameterName=Texture_02, texture=&T_Turbulence_Seamless}.R
   )) {constB=2}
  n19 = Constant() {r=1}
  n20 = ShadingPathSwitch(Default=LinearInterpolate(Alpha=SquareRoot(None=@n18)) {constA=0.01}, Mobile=@n19)
  n21 = WorldPosition()
  n22 = VectorParameter() {parameterName=Sphere, defaultValue=<rgba 0 0 0 1>}
  n23 = ScalarParameter() {defaultValue=100, group=EmissiveColors, parameterName=Radius}
  n24 = ScalarParameter() {defaultValue=0.2, group=EmissiveColors, parameterName=Hardness}
  Multiply(
   A=Add(A=FeatureLevelSwitch(Default=@n20, ES2=@n19, ES3_1=@n19, SM4_REMOVED=@n20), B=SphereMask(A=@n21.XYZ, B=@n22.RGB, Radius=@n23, Hardness=@n24))
   B=ScalarParameter()
  )
 ]
}
