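# Additive, unlit, two-sided Niagara mesh-particle material for the Mother low-gravity beam, drawing a purple, shimmering energy effect.
# Noise-distorted UVs drive panning smoke and caustic layers, alongside color blends, Fresnel shaping and a reflection-vector chroma layer.
# Particle alpha combines with near-camera and depth fades, plus an optional sine edge gradient.
# An optional lightray switch yaws the mesh toward the camera, compensates UVs and bypasses Fresnel; mobile feature levels simplify noise and fades.
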Material /Game/Athena/Apollo/Environments/Blueprints/Mother/Materials/M_Mother_Beam_Lowgrav
graph {
 properties {
  bEnableResponsiveAA=true
  bUsedWithNiagaraMeshParticles=true
  blendMode=BLEND_Additive
  pixelDepthOffsetMode=PDOM_Legacy
  shadingModel=MSM_Unlit
  translucencyLightingMode=TLM_VolumetricPerVertexDirectional
  twoSided=true
  usedShadingModels=Unlit
 }
 nodes [
  Constant() {r=1}
  Constant()
  n1 = ParticleColor()
  n2 = MaterialFunctionCall(Distance=ScalarParameter() {defaultValue=500, parameterName=NearCameraFadeDistance}) {materialFunction=&Near_Camera_Fade}
  n3 = Constant() {r=1}
  n4 = ScalarParameter() {parameterName=Depth_FadeDistance}
  n5 = Constant() {r=1}
  n6 = StaticSwitchParameter(True=FeatureLevelSwitch(Default=DepthFade(FadeDistance=@n4), ES2=@n5, ES3_1=@n5), False=Constant() {r=1}) {defaultValue=true, parameterName=UseDepthFade}
  n7 = TextureCoordinate() {vTiling=0.5}
  n8 = Multiply(A=@n6, B=Sine(None=ComponentMask(None=@n7) {g=true}))
  Multiply(
   A=@n1.A
   B=Multiply(A=FeatureLevelSwitch(Default=@n2.Result, ES2=@n3, ES3_1=@n3), B=StaticSwitchParameter(True=@n8, False=@n6) {parameterName=UseEdgeGradient})
  )
  n9 = ScalarParameter() {defaultValue=2, parameterName=FresnelExponent}
  n10 = LinearInterpolate(
   A=ScalarParameter() {defaultValue=0.15, parameterName=FresnelLow}
   Alpha=StaticSwitchParameter(
    True=Constant()
    False=MaterialFunctionCall(Exponent=@n9) {materialFunction=&MF_Fresnel_FX}.Fresnel Inverted Double Sided
   ) {parameterName=UseOptimizedLightrayWPO}
  )
  n11 = FeatureLevelSwitch(Default=@n10, ES2=@n10, ES3_1=@n10, SM5=@n10, SM6=@n10)
  n12 = Multiply(A=@n10, B=NamedRerouteUsage())
  n13 = LinearInterpolate(
   A=VectorParameter() {parameterName=ColorBaseA, defaultValue=<rgba 0.1184 0.0469 0.3038 0>}.RGB
   B=VectorParameter() {parameterName=ColorBaseB, defaultValue=<rgba 0.5646 0.0961 0.6163 0>}.RGB
   Alpha=Multiply(A=NamedRerouteUsage(), B=NamedRerouteUsage())
  )
  n14 = VectorParameter() {parameterName=ColorSparkles, defaultValue=<rgba 0.7778 0.226 0.8 0>}
  n15 = NamedRerouteUsage()
  n16 = Multiply(A=Multiply(A=Add(A=@n11, B=@n12), B=LinearInterpolate(A=@n13, B=@n14.RGB, Alpha=@n15)), B=VertexColor().A)
  n17 = ScalarParameter() {defaultValue=0.12, parameterName=EmissiveBoost}
  Multiply(A=ScalarParameter() {defaultValue=1, parameterName=ColorIntensity}, B=Multiply(A=Multiply(A=@n16, B=@n17), B=@n1.RGB))
  n18 = Constant3Vector() {constant=<rgba 0 0 1 0>}
  n19 = Constant3Vector() {constant=<rgba 0 1 0 0>}
  n20 = Normalize(VectorInput=MaterialFunctionCall() {materialFunction=&CameraDirectionVector}.Result)
  n21 = Normalize(VectorInput=Subtract(
    A=MaterialFunctionCall() {materialFunction=&ObjectPivotPoint}.Object Pivot Location
    B=CameraPositionWS()
   ))
  n22 = MaterialFunctionCall() {materialFunction=&ObjectScale}
  n23 = Max(A=@n22.Scale X, B=@n22.Scale X)
  n24 = CameraPositionWS()
  n25 = MaterialFunctionCall() {materialFunction=&ObjectPivotPoint}
  n26 = MaterialFunctionCall!(
   min=Add(A=ScalarParameter() {defaultValue=-6000, parameterName=CameraDirectionalVectorLerpBias}, B=@n23)
   max=@n23
   value=Distance(A=ComponentMask(None=@n24) {g=true, r=true}, B=ComponentMask(None=@n25.Object Pivot Location) {g=true, r=true})
  ) {materialFunction=/Game/Effects/Fort_Effects/Materials/_Functions/InverseLerp_MF}
  n27 = Normalize(VectorInput=Multiply(A=LinearInterpolate(A=@n20, B=@n21, Alpha=@n26.Clamped), B=Constant3Vector() {constant=<rgba 1 1 0 0>}))
  n28 = Constant3Vector() {constant=<rgba 1 0 0 0>}
  n29 = NamedRerouteDeclaration(None=Arctangent2Fast(Y=DotProduct(A=@n19, B=@n27), X=DotProduct(A=@n28, B=@n27))) {name=OptimizedLightrayAngleInRadians, nodeColor=<rgba 1 0 0.8906 1>}
  n30 = MaterialFunctionCall() {materialFunction=&ObjectPivotPoint}
  n31 = WorldPosition()
  NamedRerouteDeclaration(None=StaticSwitchParameter(
    True=RotateAboutAxis(NormalizedRotationAxis=@n18, RotationAngle=@n29, PivotPoint=@n30.Object Pivot Location, Position=@n31.XYZ) {desc=2pi, period=6.2832}
    False=Constant()
   ) {parameterName=UseOptimizedLightrayWPO}) {name=WPO, nodeColor=<rgba 0 0.2734 1 1>}
  n32 = ScalarParameter() {parameterName=FadeDetailDistance}
  NamedRerouteDeclaration(None=Clamp(None=MaterialFunctionCall(Distance=@n32) {materialFunction=&Near_Camera_Fade}.Result)) {name=FadeDetailDistance, nodeColor=<rgba 0.5 0.5 1 1>}
  n33 = TextureCoordinate()
  n34 = AppendVector(A=Divide(A=NamedRerouteUsage()) {constB=3.1416}, B=Constant())
  NamedRerouteDeclaration(None=StaticSwitchParameter(True=Subtract(A=@n33, B=@n34), False=@n33) {parameterName=UseOptimizedLightrayWPO}) {name=UVs, nodeColor=<rgba 0 0.7656 1 1>}
  n35 = VectorParameter() {parameterName=NoiseA_TilePan, defaultValue=<rgba 1 1 0 0.0194>}
  n36 = Panner(
   Coordinate=TextureCoordinate()
   Speed=AppendVector(
    A=ScalarParameter() {group=Distort, parameterName=Distort_Main_PanSpeed_X}
    B=ScalarParameter() {group=Distort, parameterName=Distort_Main_PanSpeed_Y}
   )
  ) {speedY=-0.4}
  n37 = Divide(
   A=ComponentMask(None=TextureSampleParameter2D(UVs=@n36) {parameterName=NoiseTexture, texture=&T_NoisePack_01}.RGB) {g=true, r=true}
   B=ScalarParameter() {defaultValue=8, group=Distort, parameterName=Distort_Amount}
  ) {constB=8}
  n38 = Constant()
  n39 = Add(
   A=FeatureLevelSwitch(Default=@n37, ES2=@n38, ES3_1=@n38, SM4_REMOVED=@n37)
   B=Multiply(A=TextureCoordinate(), B=ScalarParameter() {defaultValue=1, group=Distort, parameterName=Distort_NoiseTile})
  )
  n40 = Add(A=Multiply(A=NamedRerouteUsage(), B=AppendVector(A=@n35.R, B=@n35.R)), B=@n39)
  n41 = AppendVector(A=@n35.B, B=@n35.B)
  n42 = VectorParameter() {parameterName=NoiseB_TilePan, defaultValue=<rgba -2 -2 0.014 0>}
  n43 = Add(A=Multiply(A=NamedRerouteUsage(), B=AppendVector(A=@n42.R, B=@n42.R)), B=@n39)
  n44 = AppendVector(A=@n42.B, B=@n42.B)
  n45 = Add(
   A=TextureSample(UVs=Panner(Coordinate=@n40, Speed=@n41) {bFractionalPart=true}) {texture=&T_Inky_Smoke_Tile}.G
   B=TextureSample(UVs=Panner(Coordinate=@n43, Speed=@n44) {bFractionalPart=true}) {texture=&T_Inky_Smoke_Tile}.G
  )
  n46 = Multiply(A=@n45, B=ScalarParameter() {defaultValue=4.5, parameterName=NoiseBrightScale})
  n47 = ScalarParameter() {defaultValue=5.5, parameterName=NoiseExponent}
  n48 = VectorParameter() {parameterName=NoiseC_TilePan, defaultValue=<rgba 1 1 0.0875 -0.077>}
  n49 = Multiply(A=NamedRerouteUsage(), B=AppendVector(A=@n48.R, B=@n48.R))
  n50 = TextureSample(UVs=Panner(Coordinate=Add(A=@n49, B=@n39), Speed=AppendVector(A=@n48.B, B=@n48.B)) {bFractionalPart=true}) {texture=&T_ShieldCaustics}
  n51 = VectorParameter() {parameterName=NoiseD_TilePan, defaultValue=<rgba -2 -2 0.075 0.0605>}
  n52 = Multiply(A=NamedRerouteUsage(), B=AppendVector(A=@n51.R, B=@n51.R))
  n53 = TextureSample(UVs=Panner(Coordinate=Add(A=@n52, B=@n39), Speed=AppendVector(A=@n51.B, B=@n51.B)) {bFractionalPart=true}) {texture=&T_ShieldCaustics}
  n54 = NamedRerouteDeclaration(None=Add(
    A=NamedRerouteDeclaration(None=@n50.G) {name=NoiseC, nodeColor=<rgba 0.7109 1 0 1>}
    B=NamedRerouteDeclaration(None=@n53.G) {name=NoiseD, nodeColor=<rgba 0.3359 1 0 1>}
   )) {name=NoiseC+D, nodeColor=<rgba 1 0 0.6562 1>}
  n55 = Add(A=Multiply(A=@n45, B=ScalarParameter() {defaultValue=2, parameterName=NoiseBaseScale}), B=Multiply(A=Power(Base=@n46, Exp=@n47), B=@n54))
  n56 = AppendVector(A=ScalarParameter() {group=Distort, parameterName=Distort_PanSpeed_X}, B=ScalarParameter() {group=Distort, parameterName=Distort_PanSpeed_Y})
  n57 = Add(A=Constant2Vector(), B=@n56) {constB=0.3}
  n58 = Time()
  n59 = MaterialFunctionCall(Float2=@n39) {materialFunction=&BreakOutFloat2Components}
  n60 = Panner(Coordinate=AppendVector(A=@n59.R, B=Add(A=@n59.G) {constB=0.5}), Time=@n58, Speed=@n56)
  n61 = Add(
   A=TextureSample(UVs=Panner(Coordinate=@n39, Time=@n58, Speed=@n57) {speedX=0.2}) {samplerSource=SSM_Wrap_WorldGroupSettings, texture=&T_Inky_Smoke_Tile}.G
   B=LinearInterpolate(Alpha=TextureSample(UVs=@n60) {samplerSource=SSM_Wrap_WorldGroupSettings, texture=&T_Inky_Smoke_Tile}.G) {constB=3}
  )
  NamedRerouteDeclaration(None=FeatureLevelSwitch(Default=Add(A=@n55, B=@n61), ES2=@n54, ES3_1=@n54)) {name=Noise_ABCD, nodeColor=<rgba 0.125 1 0 1>}
  n62 = ReflectionVectorWS()
  n63 = ScalarParameter() {defaultValue=2, parameterName=ChromaReflectionVectorScale}
  n64 = Add(A=ComponentMask(None=Multiply(A=@n62, B=@n63)) {g=true, r=true}, B=@n39)
  NamedRerouteDeclaration(None=Multiply(
    A=TextureSample(UVs=@n64) {texture=/Game/Global/Textures/Default/Masks_and_Gradients/3pxRGBImage}.RGB
    B=VectorParameter() {parameterName=ChromaIntensity, defaultValue=<rgba 1.8 0.2 1.8 0>}.RGB
   )) {name=Chroma, nodeColor=<rgba 0.5234 0 1 1>}
  NamedRerouteUsage()
 ]
}
