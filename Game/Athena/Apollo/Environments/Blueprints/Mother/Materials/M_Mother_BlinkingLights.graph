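# Additive unlit material for blinking light flares in the Mother environment, drawn as camera-facing sprites through world position offset.
# Layered sines over time, phase-shifted by vertex red, pulse sprite size, which also fades near the camera.
# Vertex green varies rotation, flare texture blend and hue; vertex blue weights time-of-day size scaling.
# Shading-path and quality switches provide mobile brightness and size variants, plus distance color fade on High and Epic.
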Material /Game/Athena/Apollo/Environments/Blueprints/Mother/Materials/M_Mother_BlinkingLights
graph {
 properties {bApplyCloudFogging=true, blendMode=BLEND_Additive, pixelDepthOffsetMode=PDOM_Legacy, shadingModel=MSM_Unlit, usedShadingModels=Unlit}
 nodes [
  Constant()
  DynamicParameter() {paramNames [BlendFlareTextures, StarFlickIntense, StarFlickSpeed, Size]}
  n1 = ScalarParameter() {defaultValue=0.8, parameterName=SpriteSquareSize}
  n2 = VertexColor()
  n3 = Multiply(
   A=Add(A=Time(), B=Multiply(A=@n2.R) {constB=5})
   B=LinearInterpolate(
    B=ScalarParameter() {defaultValue=0.02, parameterName=StarFlickerInitialSpeed}
    Alpha=ScalarParameter() {defaultValue=1, parameterName=FlickerSpeed}
   )
  ) {constB=80}
  n4 = Multiply(A=Sine(None=@n3) {period=0.155}, B=Sine(None=@n3) {period=0.2})
  n5 = Sine(None=@n3) {period=0.43}
  n6 = LinearInterpolate(
   B=Multiply(A=Sine(None=@n3) {period=0.075}, B=Multiply(A=@n4, B=@n5))
   Alpha=ScalarParameter() {defaultValue=1, parameterName=FlickerIntense}
  ) {constA=1}
  LinearInterpolate(A=Multiply(A=@n1) {constB=0.5}, B=@n1, Alpha=@n6)
  LinearInterpolate(
   A=ScalarParameter() {defaultValue=1, parameterName=SpriteSize}
   B=ScalarParameter() {defaultValue=1, parameterName="SpritSize(OnlyWhenUseRandomOpacityEnabled)"}
   Alpha=ScalarParameter() {parameterName=UseRandomOpacity}
  )
  MaterialFunctionCall() {materialFunction=&Near_Camera_Fade}
  ShadingPathSwitch(Default=DepthFade(Opacity=Constant() {r=1}, FadeDistance=ScalarParameter() {parameterName=DepthFade}), Mobile=Constant() {r=1})
  n7 = ScalarParameter() {defaultValue=0.2, parameterName=HueShiftMult}
  n8 = VertexColor()
  n9 = MaterialFunctionCall(UVs=TextureCoordinate(), "Rotation Angle (0-1)"=VertexColor().G) {materialFunction=&CustomRotator}
  n10 = TextureSampleParameter2D(UVs=@n9.Rotated Values) {parameterName=FlareB, texture=/Game/Effects/Fort_Effects/Textures/Flares/T_Flare_Soft}
  n11 = Multiply(
   A=SphereMask(
    A=TextureCoordinate()
    B=Constant2Vector() {g=0.5, r=0.5}
    Radius=ScalarParameter() {defaultValue=0.5, parameterName=Radius}
    Hardness=ScalarParameter() {parameterName=Hardness}
   )
   B=ScalarParameter() {defaultValue=1, parameterName=ExtraRoundGlow}
  )
  n12 = Multiply(
   A=MaterialFunctionCall(Hue Shift Percentage=Multiply(A=@n7, B=@n8.G), Texture=VectorParameter() {parameterName=Color, defaultValue=<rgba 0.8866 0.434 2 1>}.RGB) {materialFunction=&HueShift}.Result
   B=LinearInterpolate(
    A=TextureSampleParameter2D(UVs=@n9.Rotated Values) {parameterName=FlareA, texture=/Game/Effects/Fort_Effects/Textures/Flares/T_Flare_Spikeball}.G
    B=Add(A=@n10.G, B=@n11)
    Alpha=VertexColor().G
   )
  )
  n13 = ScalarParameter() {defaultValue=100, parameterName=Brightness}
  n14 = ShadingPathSwitch(
   Default=MaterialFunctionCall(LDR Input=Clamp(None=@n12), HDR Tint=Multiply(A=@n12, B=@n13) {constB=100}) {materialFunction=&DeriveHDRfromLDR}.RemappedRange
   Mobile=Multiply(A=@n12, B=ScalarParameter() {defaultValue=2, parameterName=BrightnessMobile})
  )
  n15 = ScalarParameter() {defaultValue=1000, parameterName=Far_Fade_Length}
  n16 = ScalarParameter() {defaultValue=5000, parameterName=Far_Fade_Offset}
  n17 = Multiply(
   A=LinearInterpolate(
    A=ScalarParameter() {defaultValue=1, parameterName=FarFade_Color1}
    B=ScalarParameter() {parameterName=FarFade_Color2}
    Alpha=MaterialFunctionCall(Fade Length=@n15, Fade Offset=@n16) {materialFunction=&CameraDepthFade}.Result
   )
   B=@n14
  )
  ShadingPathSwitch(Default=QualitySwitch(Default=@n14, Low=@n14, High=@n17, Medium=@n14, Epic=@n17), Mobile=@n14)
  n18 = WorldPosition()
  n19 = ScalarParameter() {parameterName="Value_Morning (S)"}
  n20 = ScalarParameter() {parameterName="Value_Day (S)"}
  n21 = ScalarParameter() {defaultValue=1, parameterName="Value_Evening (S)"}
  n22 = ScalarParameter() {defaultValue=1, parameterName="Value_Night (S)"}
  n23 = LinearInterpolate(A=ScalarParameter() {parameterName=MinSizeMult}, B=ScalarParameter() {parameterName=MaxSizeMult}, Alpha=VertexColor().R)
  n24 = ShadingPathSwitch(Default=@n23, Mobile=Multiply(A=@n23, B=ScalarParameter() {defaultValue=2, parameterName=MaxSizeMult_Mobile}))
  n25 = ScalarParameter() {defaultValue=40000, parameterName=WPO_NearCameraFade}
  n26 = Multiply(
   A=LinearInterpolate(
    A=Constant() {r=1}
    B=MaterialFunctionCall(Value_Morning=@n19, Value_Day=@n20, Value_Evening=@n21, Value_Night=@n22) {materialFunction=&Scalar_Time_of_Day}.Result
    Alpha=VertexColor().B
   )
   B=LinearInterpolate(
    B=Multiply(A=@n24, B=@n6)
    Alpha=MaterialFunctionCall(Distance=@n25) {materialFunction=&Near_Camera_Fade}.Result
   )
  )
  n27 = ScalarParameter() {parameterName="Offset Amount (S)"}
  StaticSwitchParameter(
   True=Constant()
   False=Add(
    A=MaterialFunctionCall!(Center Location=@n18.XYZ, 0-1 UVs, XY scale=@n26, OverridePivot, "Normals (Optional)") {materialFunction=&Sprite}.World position offset
    B=MaterialFunctionCall(Offset Amount=@n27) {materialFunction=/Engine/Functions/Engine_MaterialFunctions02/WorldPositionOffset/CameraOffset}.Clamped Camera Offset
   )
  ) {parameterName=TurnOff_WPO}
 ]
}
