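# Additive unlit material that fakes flickering colored lights in the Mother environment by brightening the scene behind camera-facing, camera-offset sprites.
# Lighting comes from scene world normals and a sphere mask around the reconstructed position behind translucency, multiplied by scene color.
# Vertex color varies flicker phase, hue shift and sprite scale, while time of day drives brightness and scale.
# Non-deferred paths output black and collapse the mesh; High and Epic quality add camera depth fading.
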Material /Game/Athena/Apollo/Environments/Blueprints/Mother/Materials/M_Mother_FakeLighting
graph {
 properties {bApplyCloudFogging=true, blendMode=BLEND_Additive, pixelDepthOffsetMode=PDOM_Legacy, shadingModel=MSM_Unlit, usedShadingModels=Unlit}
 nodes [
  n1 = WorldPosition()
  n2 = MaterialFunctionCall() {materialFunction=&WorldPositionBehindTranslucency}
  n3 = DotProduct(
   A=Normalize(VectorInput=Subtract(A=@n1.XYZ, B=@n2.Result))
   B=ComponentMask(None=SceneTexture() {sceneTextureId=PPI_WorldNormal}.Color) {b=true, g=true, r=true}
  )
  n4 = SphereMask(A=TextureCoordinate(), B=Constant() {r=0.5}, Radius=ScalarParameter() {defaultValue=0.5}, Hardness=Constant() {r=0.1})
  n5 = ScalarParameter() {defaultValue=4, parameterName=RadiusFalloffExponent}
  n6 = SceneTexture() {sceneTextureId=PPI_StoredBaseColor}
  Multiply(
   A=Multiply(A=Saturate(None=@n3), B=Power(Base=@n4, Exp=@n5))
   B=LinearInterpolate(B=ComponentMask(None=@n6.Color) {b=true, g=true, r=true}) {constA=1, constAlpha=0.4}
  )
  n7 = WorldPosition()
  n8 = LinearInterpolate(
   A=ScalarParameter() {defaultValue=13000, parameterName=Scale}
   B=MaterialFunctionCall(
    Value_Morning=ScalarParameter() {parameterName="Value_Morning (S)"}
    Value_Day=ScalarParameter() {parameterName="Value_Day (S)"}
    Value_Evening=ScalarParameter() {defaultValue=1, parameterName="Value_Evening (S)"}
    Value_Night=ScalarParameter() {defaultValue=1, parameterName="Value_Night (S)"}
   ) {materialFunction=&Scalar_Time_of_Day}.Result
   Alpha=VertexColor().B
  )
  n9 = ScalarParameter() {defaultValue=15000, parameterName="Offset Amount (S)"}
  ShadingPathSwitch(
   Default=Subtract(A=ObjectPositionWS(), B=WorldPosition().XYZ)
   Deferred=Add(
    A=MaterialFunctionCall!(Center Location=@n7.XYZ, 0-1 UVs, XY scale=@n8, OverridePivot, "Normals (Optional)") {materialFunction=&Sprite}.World position offset
    B=MaterialFunctionCall(Offset Amount=@n9) {materialFunction=/Engine/Functions/Engine_MaterialFunctions02/WorldPositionOffset/CameraOffset}.Clamped Camera Offset
   )
  )
  n10 = VertexColor()
  n11 = Multiply(
   A=Add(A=Time(), B=Multiply(A=@n10.R) {constB=5})
   B=LinearInterpolate(
    B=ScalarParameter() {defaultValue=0.04, parameterName=StarFlickerInitialSpeed}
    Alpha=ScalarParameter() {defaultValue=0.6, parameterName=FlickerSpeed}
   )
  ) {constB=80}
  n12 = Sine(None=@n11) {period=0.075}
  n13 = Multiply(A=Multiply(A=Sine(None=@n11) {period=0.155}, B=Sine(None=@n11) {period=0.2}), B=Sine(None=@n11) {period=0.43})
  n14 = ScalarParameter() {defaultValue=0.2, parameterName=HueShiftMult}
  n15 = VertexColor()
  n16 = Multiply(
   A=LinearInterpolate(B=Multiply(A=@n12, B=@n13), Alpha=ScalarParameter() {defaultValue=0.45, parameterName=FlickerIntense}) {constA=1}
   B=MaterialFunctionCall(Hue Shift Percentage=Multiply(A=@n14, B=@n15.G), Texture=VectorParameter() {parameterName=Color, defaultValue=<rgba 1.5965 0.2658 2 1>}.RGB) {materialFunction=&HueShift}.Result
  )
  n17 = WorldPosition() {worldPositionShaderOffset=WPT_ExcludeAllShaderOffsets}
  n18 = MaterialFunctionCall() {materialFunction=&WorldPositionBehindTranslucency}
  n19 = DotProduct(
   A=Normalize(VectorInput=Subtract(A=@n17.XYZ, B=@n18.Result))
   B=ComponentMask(None=SceneTexture() {sceneTextureId=PPI_WorldNormal}.Color) {b=true, g=true, r=true}
  )
  n20 = SphereMask(
   A=@n17.XYZ
   B=@n18.Result
   Radius=ScalarParameter() {defaultValue=15000, parameterName=LightRadius}
   Hardness=ScalarParameter() {defaultValue=0.1, parameterName=FakeLightingMaskHardness}
  )
  n21 = ScalarParameter() {defaultValue=2, parameterName=LightFalloffExponent}
  n22 = Multiply(A=Multiply(A=Saturate(None=@n19), B=Power(Base=@n20, Exp=@n21)), B=SceneColor!(UVs).RGB)
  n23 = ScalarParameter() {defaultValue=0.3, parameterName=BrightnessDay}
  n24 = ScalarParameter() {defaultValue=1, parameterName=BrightnessNight}
  n25 = ShadingPathSwitch(
   Default=Constant()
   Deferred=Multiply(
    A=Multiply(A=@n16, B=@n22)
    B=MaterialFunctionCall(Value_Morning=@n23, Value_Day=@n23, Value_Evening=@n23, Value_Night=@n24) {materialFunction=&Scalar_Time_of_Day}.Result
   )
  )
  n26 = Constant() {r=100000}
  n27 = Constant() {r=100000}
  n28 = Multiply(
   A=LinearInterpolate(
    A=Constant() {r=1}
    B=Constant() {r=0.5}
    Alpha=MaterialFunctionCall(Fade Length=@n26, Fade Offset=@n27) {materialFunction=&CameraDepthFade}.Result
   )
   B=@n25
  )
  ShadingPathSwitch(Default=QualitySwitch(Default=@n25, High=@n28, Epic=@n28))
 ]
}
