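# M_Mother_Gradient is an unlit translucent material for the Apollo "Mother" environment blueprints that builds a tintable vertical gradient.
# It inverts the V texture coordinate and raises it to the Exp parameter to shape the falloff.
# That gradient is multiplied separately by Color and by Opacity, apparently for emissive and opacity.
# A Sine and a radial SphereMask with Radius and Hardness parameters sit apart, feeding no other node.
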Material /Game/Athena/Apollo/Environments/Blueprints/Mother/Materials/M_Mother_Gradient
graph {
 properties {blendMode=BLEND_Translucent, pixelDepthOffsetMode=PDOM_Legacy, shadingModel=MSM_Unlit, usedShadingModels=Unlit}
 nodes [
  n1 = ComponentMask(None=TextureCoordinate() {uTiling=0.5}) {g=true}
  n2 = Power(Base=OneMinus(None=@n1), Exp=ScalarParameter() {defaultValue=2, parameterName=Exp})
  Multiply(A=ComponentMask(None=VectorParameter() {parameterName=Color, defaultValue=<rgba 1 1 1 1>}.RGB) {b=true, g=true, r=true}, B=@n2)
  Multiply(A=@n2, B=ScalarParameter() {defaultValue=1, parameterName=Opacity})
  Sine(None=@n1)
  SphereMask(
   A=TextureCoordinate()
   B=Constant2Vector() {g=0.5, r=0.5}
   Radius=ScalarParameter() {defaultValue=0.5, parameterName=Radius}
   Hardness=ScalarParameter() {parameterName=Hardness}
  )
 ]
}
