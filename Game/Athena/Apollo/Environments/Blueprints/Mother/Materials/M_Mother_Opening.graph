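# Additive two-sided material for the Mother opening effect, drawing animated purple energy on meshes or mesh particles.
# Panned smoke and caustic textures, warped by distortion noise, form layered noise that also offsets a hue-shifted chroma texture lookup using the reflection vector.
# Base and sparkle color blends feed emissive with vertex alpha and an optional center energy gradient.
# Depth fade, fresnel, vertex alpha and optional center fade build a mask.
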Material /Game/Athena/Apollo/Environments/Blueprints/Mother/Materials/M_Mother_Opening
graph {
 properties {bUsedWithNiagaraMeshParticles=true, blendMode=BLEND_Additive, pixelDepthOffsetMode=PDOM_Legacy, twoSided=true, usedShadingModels=Default Lit}
 nodes [
  n1 = VectorParameter() {parameterName=ColorBaseA, defaultValue=<rgba 0.1184 0.0469 0.3038 0>}
  n2 = VectorParameter() {parameterName=ColorBaseB, defaultValue=<rgba 0.5646 0.0961 0.6163 0>}
  n3 = Multiply(A=NamedRerouteUsage(), B=NamedRerouteUsage())
  n4 = Multiply(
   A=NamedRerouteUsage()
   B=LinearInterpolate(
    A=LinearInterpolate(A=@n1.RGB, B=@n2.RGB, Alpha=@n3)
    B=VectorParameter() {parameterName=ColorSparkles, defaultValue=<rgba 0.7778 0.226 0.8 0>}.RGB
    Alpha=NamedRerouteUsage()
   )
  )
  n5 = VertexColor()
  n6 = Multiply(A=@n4, B=Multiply(A=ScalarParameter() {defaultValue=1, parameterName=EmmisiveBoost}, B=@n5.A))
  n7 = StaticSwitchParameter(True=Multiply(A=ParticleColor().RGB, B=@n6), False=@n6) {parameterName=UseAsMeshParticle?}
  n8 = MaterialFunctionCall() {materialFunction=&LinearGradient}
  n9 = Multiply(A=Multiply(A=Power(Base=@n8.VGradient) {constExponent=4}) {constB=200}, B=@n7)
  MaterialFunctionCall(Emissive=LinearInterpolate(A=@n7, B=Add(A=@n9, B=@n7), Alpha=ScalarParameter() {parameterName=Ues Center Energy})) {materialFunction=&MF_FX_EmissiveAdjustments}
  Multiply(A=@n4, B=@n5.A)
  n10 = ScalarParameter() {parameterName=ChromaHueShift}
  n11 = ReflectionVectorWS()
  n12 = ScalarParameter() {defaultValue=0.15, parameterName=ChromaReflectionVectorScale}
  n13 = TextureCoordinate()
  n14 = AppendVector(
   A=ScalarParameter() {group=Distort, parameterName=Distort_Main_PanSpeed_X}
   B=ScalarParameter() {group=Distort, parameterName=Distort_Main_PanSpeed_Y}
  )
  n15 = ComponentMask(None=TextureSampleParameter2D(UVs=Panner(Coordinate=@n13, Speed=@n14) {speedY=-0.4}) {parameterName=NoiseTexture, texture=&T_NoisePack_01}.RGB) {g=true, r=true}
  n16 = ScalarParameter() {defaultValue=8, group=Distort, parameterName=Distort_Amount}
  n17 = Add(
   A=ShadingPathSwitch(Default=Divide(A=@n15, B=@n16) {constB=8}, Mobile=Constant()) {desc=Migrated from FeatureLevelSwitch}
   B=Multiply(A=TextureCoordinate(), B=ScalarParameter() {defaultValue=1, group=Distort, parameterName=Distort_NoiseTile})
  )
  n18 = Add(A=ComponentMask(None=Multiply(A=@n11, B=@n12)) {g=true, r=true}, B=@n17)
  n19 = Divide(A=NamedRerouteUsage(), B=ScalarParameter() {defaultValue=1000, parameterName=WorldTileDivisorAB})
  n20 = VectorParameter() {parameterName=NoiseA_TilePan, defaultValue=<rgba 1 1 0 0.0194>}
  n21 = Add(A=Multiply(A=@n19, B=AppendVector(A=@n20.R, B=@n20.R)), B=@n17)
  n22 = AppendVector(A=@n20.B, B=@n20.B)
  n23 = VectorParameter() {parameterName=NoiseB_TilePan, defaultValue=<rgba -2 -2 0.014 0>}
  n24 = Add(A=Multiply(A=@n19, B=AppendVector(A=@n23.R, B=@n23.R)), B=@n17)
  n25 = AppendVector(A=@n23.B, B=@n23.B)
  n26 = Add(
   A=TextureSample(UVs=Panner(Coordinate=@n21, Speed=@n22) {bFractionalPart=true}) {texture=&T_Inky_Smoke_Tile}.G
   B=TextureSample(UVs=Panner(Coordinate=@n24, Speed=@n25) {bFractionalPart=true}) {texture=&T_Inky_Smoke_Tile}.G
  )
  n27 = ScalarParameter() {defaultValue=2, parameterName=NoiseBaseScale}
  n28 = Power(
   Base=Multiply(A=@n26, B=ScalarParameter() {defaultValue=4.5, parameterName=NoiseBrightScale})
   Exp=ScalarParameter() {defaultValue=5.5, parameterName=NoiseExponent}
  )
  n29 = VectorParameter() {parameterName=NoiseC_TilePan, defaultValue=<rgba 1 1 0.0875 -0.077>}
  n30 = Multiply(
   A=Divide(A=NamedRerouteUsage(), B=ScalarParameter() {defaultValue=3500, parameterName=WorldTileDivisorC})
   B=AppendVector(A=@n29.R, B=@n29.R)
  )
  n31 = TextureSample(UVs=Panner(Coordinate=Add(A=@n30, B=@n17), Speed=AppendVector(A=@n29.B, B=@n29.B)) {bFractionalPart=true}) {texture=&T_ShieldCaustics}
  n32 = VectorParameter() {parameterName=NoiseD_TilePan, defaultValue=<rgba -2 -2 0.075 0.0605>}
  n33 = Multiply(A=Divide(A=NamedRerouteUsage(), B=ScalarParameter() {parameterName=WorldTileDivisorC}), B=AppendVector(A=@n32.R, B=@n32.R))
  n34 = TextureSample(UVs=Panner(Coordinate=Add(A=@n33, B=@n17), Speed=AppendVector(A=@n32.B, B=@n32.B)) {bFractionalPart=true}) {texture=&T_ShieldCaustics}
  n35 = NamedRerouteDeclaration(None=Add(
    A=NamedRerouteDeclaration(None=@n31.G) {name=NoiseC, nodeColor=<rgba 0.7109 1 0 1>}
    B=NamedRerouteDeclaration(None=@n34.G) {name=NoiseD, nodeColor=<rgba 0.3359 1 0 1>}
   )) {name=NoiseC+D, nodeColor=<rgba 1 0 0.6562 1>}
  n36 = AppendVector(A=ScalarParameter() {group=Distort, parameterName=Distort_PanSpeed_X}, B=ScalarParameter() {group=Distort, parameterName=Distort_PanSpeed_Y})
  n37 = Time()
  n38 = Panner(Coordinate=@n17, Time=@n37, Speed=Add(A=Constant2Vector(), B=@n36) {constB=0.3}) {speedX=0.2}
  n39 = MaterialFunctionCall(Float2=@n17) {materialFunction=&BreakOutFloat2Components}
  n40 = Add(A=@n39.G) {constB=0.5}
  n41 = TextureSample(UVs=Panner(Coordinate=AppendVector(A=@n39.R, B=@n40), Time=@n37, Speed=@n36)) {texture=&T_Inky_Smoke_Tile}
  n42 = Add(
   A=Add(A=Multiply(A=@n26, B=@n27), B=Multiply(A=@n28, B=@n35))
   B=Add(
    A=TextureSample(UVs=@n38) {texture=&T_Inky_Smoke_Tile}.G
    B=LinearInterpolate(Alpha=@n41.G) {constB=3}
   )
  )
  n43 = Multiply(A=@n42, B=ScalarParameter() {parameterName=ChromaDistAmount})
  n44 = TextureSampleParameter2D(UVs=Panner(Coordinate=Add(A=@n18, B=@n43)) {speedX=0.2}) {parameterName=ChromaTexture, texture=/Game/Global/Textures/Default/Masks_and_Gradients/3pxRGBImage}
  NamedRerouteDeclaration(None=Multiply(
    A=MaterialFunctionCall(Hue Shift Percentage=@n10, Texture=@n44.RGB) {materialFunction=&HueShift}.Result
    B=VectorParameter() {parameterName=ChromaIntensity, defaultValue=<rgba 1.8 0.2 1.8 0>}.RGB
   )) {name=Chroma, nodeColor=<rgba 0.5 0 1 1>}
  n45 = ScalarParameter() {parameterName=FadeDetailDistance}
  NamedRerouteDeclaration(None=Clamp(None=MaterialFunctionCall(Distance=@n45) {materialFunction=&Near_Camera_Fade}.Result)) {name=FadeDetailDistance, nodeColor=<rgba 0.5 0.5 1 1>}
  NamedRerouteDeclaration(None=Multiply(A=TextureCoordinate(), B=ScalarParameter() {defaultValue=1000, parameterName=WorldTileMult})) {name=CheapWorldAlignedUVs, nodeColor=<rgba 0 0.7656 1 1>}
  NamedRerouteDeclaration(None=@n42) {name=Noise_ABCD, nodeColor=<rgba 0 1 0.2734 1>}
  n46 = ComponentMask(None=TextureCoordinate()) {g=true}
  n47 = Multiply(
   A=Power(Base=OneMinus(None=@n46), Exp=ScalarParameter() {defaultValue=1, parameterName=CenterFadePower})
   B=ScalarParameter() {defaultValue=1, parameterName=CenterFadeMult}
  )
  n48 = ScalarParameter() {parameterName=Depth_FadeDistance}
  n49 = ScalarParameter() {parameterName=FresnelExp}
  n50 = Multiply(A=Multiply(A=@n5.A, B=@n5.A), B=Multiply(A=DepthFade(FadeDistance=@n48), B=Fresnel(ExponentIn=@n49)))
  n51 = StaticSwitchParameter(True=Multiply(A=@n50, B=ParticleColor().A), False=@n50) {parameterName=UseAsMeshParticle?}
  StaticSwitchParameter(True=Multiply(A=Saturate(None=@n47), B=@n51), False=@n51) {parameterName=UseCenterFade?}
  n52 = CameraPositionWS()
  n53 = MaterialFunctionCall() {materialFunction=&ObjectPivotPoint}
  n54 = MaterialFunctionCall() {materialFunction=&ObjectScale}
  n55 = Max(A=@n54.Scale X, B=@n54.Scale X)
  n56 = Divide(
   A=Subtract(A=Distance(A=@n52, B=@n53.Object Pivot Location), B=Max(A=@n55, B=@n54.Scale Z))
   B=ScalarParameter() {defaultValue=1024, parameterName=DistanceDivisor}
  )
  VertexInterpolator(VS=OneMinus(None=Saturate(None=@n56)))
 ]
}
