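# Masked, lit rock material for Mother environment effects, usable on Niagara mesh particles, that can erode away.
# Erosion combines an object-bounds vertical ramp with UV1 noise and a crack offset, minus a dynamic Erode parameter.
# That result is multiplied by diffuse alpha, particle alpha and a channel-selectable opacity texture.
# The diffuse, lightened by a scalar and tinted by particle color and dynamic Emissive, feeds an emissive-adjustment function.
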Material /Game/Athena/Apollo/Environments/Blueprints/Mother/Materials/M_Mother_RockShader
graph {
 properties {bUsedWithNiagaraMeshParticles=true, blendMode=BLEND_Masked, pixelDepthOffsetMode=PDOM_Legacy, usedShadingModels=Default Lit}
 nodes [
  n1 = TextureSampleParameter2D() {
   group=Textures
   parameterName=SpecMasks
   samplerSource=SSM_Wrap_WorldGroupSettings
   samplerType=SAMPLERTYPE_Masks
   texture=&T_Fortnite_Default_S
  }
  LinearInterpolate(
   A=ScalarParameter() {group=Specular, parameterName=SpecRoughnessMin, sortPriority=1}
   B=ScalarParameter() {defaultValue=1, group=Specular, parameterName=SpecRoughnessMax, sortPriority=2}
   Alpha=@n1.B
  )
  n2 = TextureSampleParameter2D() {
   group=Textures
   parameterName=Diffuse
   samplerSource=SSM_Wrap_WorldGroupSettings
   sortPriority=1
   texture=/Engine/EngineMaterials/DefaultDiffuse
  }
  n3 = MaterialFunctionCall!(Blend=ScalarParameter() {parameterName=LightenBaseTexture}, Base=@n2.RGB) {materialFunction=/Engine/Functions/Engine_MaterialFunctions03/Blends/Blend_Lighten}
  n4 = ParticleColor()
  MaterialFunctionCall(Emissive=Multiply(
    A=Multiply(A=@n3.Result, B=@n4.RGB)
    B=DynamicParameter() {defaultValue=<rgba 0 1 1 1>, paramNames [Erode, LightenBaseTexture, NoiseUVScale, Emissive]}.Emissive
   )) {materialFunction=&MF_FX_EmissiveAdjustments}
  n5 = Constant() {r=0.5}
  n6 = Add(A=ObjectRadius()) {constB=0.001}
  n7 = Add(
   A=Subtract(A=WorldPosition() {worldPositionShaderOffset=WPT_ExcludeAllShaderOffsets}.XYZ, B=ObjectPositionWS()) {desc=Doesn't break with different locations}
   B=@n6
  )
  n8 = Multiply(A=@n6) {constB=2}
  n9 = Saturate(None=ComponentMask(None=Divide(A=@n7, B=@n8)) {b=true})
  n10 = LinearInterpolate(A=OneMinus(None=@n9), B=@n9, Alpha=ScalarParameter() {defaultValue=1, group=Erosion, parameterName=BoundRampDirection, sortPriority=5})
  n11 = ScalarParameter() {defaultValue=1, group=Erosion, parameterName=Use Bounds in Erosion, sortPriority=14}
  n12 = ComponentMask(None=Multiply(
    A=DynamicParameter() {defaultValue=<rgba 0 1 1 1>, paramNames [Erode, LightenBaseTexture, NoiseUVScale, Emissive]}.NoiseUVScale
    B=ScalarParameter() {defaultValue=1, group=Erosion, parameterName=CrackSize, sortPriority=10}
   )) {b=true}
  n13 = ScalarParameter() {defaultValue=0.5, group=Erosion, parameterName=CrackAffectMask, sortPriority=5}
  n14 = ComponentMask(None=VectorParameter() {group=Erosion, parameterName=Noise_UV_Scale, sortPriority=4, defaultValue=<rgba 2 2 0 0>}.RGB) {g=true, r=true}
  n15 = DynamicParameter() {defaultValue=<rgba 0 1 1 1>, paramNames [Erode, LightenBaseTexture, NoiseUVScale, Emissive]}
  n16 = TextureSampleParameter2D(UVs=Multiply(A=TextureCoordinate() {coordinateIndex=1}, B=Multiply(A=@n14, B=@n15.NoiseUVScale))) {
   group=Textures
   parameterName=Erosion Noise
   samplerType=SAMPLERTYPE_LinearColor
   sortPriority=3
   texture=/Game/Characters/Player/Male/Medium/Bodies/M_MED_SRDrift/Textures/T_ElectricNoise
  }
  n17 = Add(
   A=Subtract(A=LinearInterpolate(A=@n5, B=@n10, Alpha=@n11), B=Multiply(A=@n12, B=@n13) {constB=0.25})
   B=Multiply(A=@n16.R, B=ScalarParameter() {defaultValue=1, group=Erosion, parameterName=NoiseAffectMask, sortPriority=2})
  )
  n18 = DynamicParameter() {defaultValue=<rgba 0 1 1 1>, paramNames [Erode, LightenBaseTexture, NoiseUVScale, Emissive]}
  n19 = ParticleColor()
  n20 = VectorParameter() {group=Opacity, parameterName=Opacity Mask UV Scale, sortPriority=3, defaultValue=<rgba 1 1 0 0>}
  n21 = TextureSampleParameter2D(UVs=Multiply(A=TextureCoordinate(), B=ComponentMask(None=@n20.RGB) {g=true, r=true})) {
   group=Textures
   parameterName=Opacity_Mask_Texture
   samplerSource=SSM_Wrap_WorldGroupSettings
   sortPriority=1
   texture=&T_WhitesRGB
  }
  Multiply(
   A=Multiply(A=@n2.A, B=Subtract(A=@n17, B=@n18.Erode))
   B=Multiply(A=Multiply(A=@n19.A), B=ChannelMaskParameter(None=@n21.RGBA) {group=Opacity, parameterName=Opacity_Mask Channel Mask, sortPriority=2}.R)
  )
  ScalarParameter() {defaultValue=1, group=Erosion, parameterName=Use WorldAligned Crack Texture, sortPriority=8}
  TextureSampleParameter2D() {
   group=Textures
   parameterName=Normals
   samplerSource=SSM_Wrap_WorldGroupSettings
   samplerType=SAMPLERTYPE_Normal
   texture=/Engine/EngineMaterials/FlatNormal
  }
 ]
}
