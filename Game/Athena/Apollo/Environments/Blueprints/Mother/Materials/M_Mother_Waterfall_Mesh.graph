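# Masked, two-sided lit material for the Mother waterfall meshes, animating foam as falling water.
# Two scales of panning foam mask and normal textures are blended, with object world Y offsetting time to vary instances.
# The foam blend, shaped by a V gradient and soft side-edge mask, forms a cutout term, and also modulates time-of-day color and brightness.
# It also includes SplineThicken width controls, a flattened-normal Fresnel and a dynamic-parameter UV offset.
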Material /Game/Athena/Apollo/Environments/Blueprints/Mother/Materials/M_Mother_Waterfall_Mesh
graph {
 properties {blendMode=BLEND_Masked, pixelDepthOffsetMode=PDOM_Legacy, twoSided=true, usedShadingModels=Default Lit}
 nodes [
  n1 = DynamicParameter() {defaultValue=<rgba 1 0.5 100 0>, paramNames [PatternScale, PatterRepetition, GradientSoftness, UVOffset]}
  n2 = TextureCoordinate()
  n3 = Constant()
  n4 = VectorParameter() {parameterName=Tile, defaultValue=<rgba 1.25 0.4 0 0>}
  n5 = Multiply(A=AppendVector(A=@n4.R, B=@n4.R), B=TextureCoordinate())
  n6 = Add(A=Time(), B=ComponentMask(None=ObjectPositionWS()) {g=true})
  n7 = Panner(Coordinate=@n5, Time=@n6) {speedX=-0.05, speedY=0.3}
  n8 = TextureSampleParameter2D(UVs=@n7) {parameterName=Mask, samplerType=SAMPLERTYPE_LinearColor, texture=/Game/Athena/Apollo/Water/Textures/02_Rounded/02_WaterFoam_Veroni_base_H}
  n9 = Panner(Coordinate=Multiply(A=@n5) {constB=0.6}, Time=@n6) {speedX=0.05, speedY=0.16}
  n10 = TextureSampleParameter2D(UVs=@n9) {parameterName=Mask, samplerType=SAMPLERTYPE_LinearColor, texture=&02_WaterFoam_Veroni_base_H}
  n11 = ScalarParameter() {defaultValue=0.6, parameterName=Blend}
  n12 = LinearInterpolate(A=@n8.G, B=@n10.G, Alpha=@n11)
  Add(A=@n1.UVOffset, B=Add(A=AppendVector(A=@n2, B=@n3), B=Multiply(A=@n12) {constB=0.2}))
  Constant() {r=0.2}
  n13 = TextureSampleParameter2D(UVs=@n7) {parameterName=Normal, samplerType=SAMPLERTYPE_Normal, texture=/Game/Athena/Apollo/Water/Textures/02_Rounded/02_WaterFoam_Veroni_base_N}
  n14 = TextureSampleParameter2D(UVs=@n9) {parameterName=Normal, samplerType=SAMPLERTYPE_Normal, texture=&02_WaterFoam_Veroni_base_N}
  n15 = LinearInterpolate(A=@n13.RGB, B=@n14.RGB, Alpha=@n11)
  n16 = ScalarParameter() {defaultValue=15, group=WPO, parameterName=WidthBase_Min}
  n17 = ScalarParameter() {defaultValue=15, group=WPO, parameterName=WidthBase_Max}
  n18 = ComponentMask(None=TextureCoordinate()) {g=true}
  n19 = ScalarParameter() {defaultValue=250, group=WPO, parameterName=WidthTip_Min}
  n20 = ScalarParameter() {defaultValue=250, group=WPO, parameterName=WidthTip_Max}
  LinearInterpolate(
   A=@n15
   B=MaterialFunctionCall(WidthBase=LinearInterpolate(A=@n16, B=@n17, Alpha=@n18), WidthTip=LinearInterpolate(A=@n19, B=@n20, Alpha=@n18)) {materialFunction=&SplineThicken}.Normal
  )
  MaterialFunctionCall(
   Brightness=ScalarParameter() {defaultValue=1.3, parameterName=FresnelMult}
   Exponent=ScalarParameter() {defaultValue=2, parameterName=FresnelExp}
   Vertex / Pixel Normal=LinearInterpolate(A=@n15, B=Constant3Vector() {constant=<rgba 0 0 1 0>}, Alpha=ScalarParameter() {defaultValue=0.2, parameterName=Flatten Normal})
  ) {materialFunction=&MF_Fresnel_FX}
  n21 = ScalarParameter() {defaultValue=0.68}
  n22 = TextureCoordinate()
  n23 = Constant() {r=0.5}
  n24 = Multiply(
   A=ComponentMask(None=TextureCoordinate()) {g=true}
   B=SphereMask(A=ComponentMask(None=@n22) {r=true}, B=@n23, Radius=@n23, Hardness=ScalarParameter() {defaultValue=0.8, parameterName=EdgeHardness}) {hardnessPercent=1}
  )
  Multiply(A=@n21, B=Power(Base=Multiply(A=@n12, B=@n24), Exp=LinearInterpolate(Alpha=@n21) {constA=2, constB=-0.1}))
  n25 = VectorParameter() {group=Color, parameterName="Color_Morning (V3)", defaultValue=<rgba 1 1 1 0>}
  n26 = VectorParameter() {group=Color, parameterName="Color_Day (V3)", defaultValue=<rgba 1 1 1 0>}
  n27 = VectorParameter() {group=Color, parameterName="Color_Evening (V3)"}
  n28 = VectorParameter() {group=Color, parameterName="Color_Night (V3)"}
  n29 = LinearInterpolate(Alpha=Power(Base=@n12) {constExponent=0.6}) {constA=0.7}
  Multiply(
   A=Multiply(
    A=MaterialFunctionCall(Color_Morning=@n25.RGB, Color_Day=@n26.RGB, Color_Evening=@n27.RGB, Color_Night=@n28.RGB) {materialFunction=&Color_Time_of_Day}.Result
    B=Multiply(A=@n12, B=@n29)
   )
   B=MaterialFunctionCall(
    Value_Morning=ScalarParameter() {group=Color, parameterName="Value_Morning (S)"}
    Value_Day=ScalarParameter() {group=Color, parameterName="Value_Day (S)"}
    Value_Evening=ScalarParameter() {group=Color, parameterName="Value_Evening (S)"}
    Value_Night=ScalarParameter() {group=Color, parameterName="Value_Night (S)"}
   ) {materialFunction=&Scalar_Time_of_Day}.Result
  ) {constB=0.15}
  ScalarParameter() {defaultValue=0.15, parameterName=EmissiveAdd}
  Time() {bOverride_Period=true, period=1}
  VertexColor()
  VertexColor()
 ]
}
